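# Niagara system for the HappyHopper character body that draws two glowing purple eyes with sprites and particle lights.
# A burst spawns two particles, and each follows a socket through the socket reader, selected by its partition index.
# A waveform flicker multiplied by a time-of-day curve drives sprite alpha, peaking mid-cycle.
# User.bisHeadHidden collapses the sprites to zero size and disables the lights.
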NiagaraSystem /BRCosmetics/Characters/Player/Male/Medium/Bodies/M_MED_HappyHopper/FX/NS_HappyHopper_EyeGlow
dependencies {
 usedDataInterfaces [&NiagaraDataInterfaceCurve, &NiagaraDataInterfaceSocketReader]
 usedDynamicInputs []
 usedModules [
  &self:SystemSpawnScript.SystemScriptSource.SystemScriptGraph.NiagaraNodeAssignment_0.SetVariables_#1
  &SystemState
 ]
}
emitters {EyeGlow {
  emitterUpdateScript [
   &EmitterState(Life Cycle Mode=::NewEnumerator1, Loop Behavior=::NewEnumerator1, Loop Duration Mode=::NewEnumerator1, UseLoopDelay=false)
   &SpawnBurst_Instantaneous(Spawn Count=2, Use Spawn Probability=false, UseLoopCountLimit=false)
  ]
  particleSpawnScript [
   &InitializeParticle(
    Color=<rgba 1.3926 0.4945 3 1>
    Sprite Size Mode=::NewEnumerator3
    Uniform Sprite Size=11
    UsePositionOffset=false
    Write Material Random=true
   )
   &PartitionParticles()
  ]
  particleUpdateScript [
   &ParticleState(
    Kill Particles When Lifetime Has Elapsed=false
    Let Infinitely Lived Particles Die When Emitter Deactivates=true
    Loop Particles Lifetime=true
   )
   &SocketLocation(
    SamplePosOffset=true
    Sampled Position Offset=<0 -2 0>
    Socket Index=$Particles.PartitionParticles.Partition
    Socket Reader=$System.Socket Reader
    Socket Sampling Mode=::NewEnumerator3
    VelocitySpeedLimit=false
   )
   &CameraOffset(Camera Offset Amount=2)
   &ScaleColor(
    Scale Alpha=&Multiply_Float(
     A=&Multiply_Float(A=0.6, B=&Waveform("Waveform [1]"=::NewEnumerator0, "[1] Frequency"=3, "[1] Amplitude Min/Max"=<0.8 1>))
     B=&FloatFromCurve(
      FloatCurve=DataInterface() {propertyValues {Curve {keys [
          {interpMode=RCIM_Cubic, tangentMode=RCTM_User, time=0, value=0.7}
          {interpMode=RCIM_Cubic, tangentMode=RCTM_User, time=0.5, value=1}
          {tangentMode=RCTM_User, time=1, value=0.7}
         ]}}}
      CurveIndex=Linked() {linkedVariable {name=NPC.FortniteNPC.TimeOfDay_0-1, type=NiagaraFloat}}
     )
    )
    ScaleA=true
    ScaleRGB=false
   )
   &FN_Light_Attributes(
    Color=<rgba 0.3585 0 1 1>
    Create Color Attribute=true
    Create Position Attribute=true
    Enable Particle Light Enable=true
    Light Enabled:NiagaraBool=&Boolean_NotOperation(Bool=$User.bisHeadHidden)
    Position=&AddVectorToPosition(Position=$Output.SocketLocation.SampledPosition, Vector=<0 3 0>)
    Scale Color=6
   )
   &ScaleSpriteSize(
    Scale Sprite Size Mode=::NewEnumerator0
    Uniform Scale Factor=&MakeCustomFloatFromBool(Bool=$User.bisHeadHidden, True Float=0, False Float=1)
   )
  ]
  rendererClasses [&NiagaraSpriteRendererProperties, &NiagaraLightRendererProperties]
  renderers [
   {rendererClass=&NiagaraSpriteRendererProperties}
   {rendererClass=&NiagaraLightRendererProperties, rendererIndex=1}
  ]
 }}
summary {
 systemName=NS_HappyHopper_EyeGlow
 emitters [{emitterName=EyeGlow, rendererClasses [&NiagaraSpriteRendererProperties, &NiagaraLightRendererProperties]}]
 userVariables [User.bisHeadHidden {type=NiagaraBool}]
}
