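# Animation Blueprint for the M_MED_HappyHopper player body cosmetic, adding secondary motion to its belt strap, grenades, bullets and tail.
# It copies the attached parent's pose through the body's Control Rig, then runs a cached chain of Anim Dynamics nodes, with thigh-following spherical limits on the strap.
# Additive bone tweaks offset the strap and tail base.
# The event graph updates gravity, simulation space, emote, crouch and glider variables.
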AnimBlueprint /BRCosmetics/Characters/Player/Male/Medium/Bodies/M_MED_HappyHopper/Meshes/M_MED_HappyHopper_AnimBP
target_skeleton=&/M_MED_HappyHopper_Skeleton
graphs [
 AnimGraph {nodes [
   n1 = "Transform (Modify) Bone - Bone: base_Tail"!(
    ComponentPose=Local To Component(LocalPose=Use cached pose 'DynamicsCache'())
    bAlphaBoolEnabled:bool=True
    Rotation="0,10,0"
    Scale="0.8,1,1"
   ) -> (Pose) {
    alphaInputType=Bool
    rotationMode=BMM_Additive
    rotationSpace=BCS_BoneSpace
    scaleMode=BMM_Additive
    scaleSpace=BCS_BoneSpace
    alphaBoolBlend {blendOption=Linear}
    boneToModify {boneName=base_Tail}
    rotation {yaw=10}
    scale {x=0.8}
   }
   Output Pose(Result=Component To Local(ComponentPose=@n1)) {layerGroup=None}
   n2 = Control Rig!(
    Source=Copy Pose From Mesh() {bUseAttachedParent=true}
    Alpha
    "Is Crouching ":bool=Get bIsCrouching()
    Rban Enabled:bool=Get bRigidBodyEnabled()
    PlayingEmote:bool=false
    IsInFrontEnd:bool=false
   ) -> (Pose) {
    alphaBoolBlend {blendOption=Linear}
    controlRigAssetReference {blueprintRigClass=&/M_MED_HappyHopper_CtrlRig.M_MED_HappyHopper_CtrlRig_C}
   }
   Save cached pose 'CRCache_Input'(Pose=@n2) {cacheName=CRCache_Input}
   n3 = "Transform (Modify) Bone - Bone: dyn_BeltStrap_02"!(
    ComponentPose=Local To Component(LocalPose=Use cached pose 'CRCache_Input'())
    bAlphaBoolEnabled:bool=True
    Translation="0,0,0"
    Rotation="-20,0,0"
    Scale="1,1,1"
   ) -> (Pose) {
    alphaInputType=Bool
    rotationMode=BMM_Additive
    alphaBoolBlend {blendInTime=0.2, blendOption=Linear, blendOutTime=0.2}
    boneToModify {boneName=dyn_BeltStrap_02}
    rotation {pitch=-20}
   }
   n4 = Anim Dynamics - Bone: dyn_BeltStrap_01!(SimulationSpace:AnimPhysSimSpaceType Enum=Component, ComponentPose=@n3, Alpha=1, GravityOverride:Vector="0,0,-980") -> (Pose) {
    angularDampingOverride=0.9999
    angularSpringConstant=6000
    bAngularSpring=true
    bOverrideAngularDamping=true
    bUseGravityOverride=true
    bUseSphericalLimits=true
    gravityScale=0.35
    simulationSpace=RootRelative
    alphaBoolBlend {blendOption=Linear}
    boundBone {boneName=dyn_BeltStrap_01}
    componentAppliedLinearAccClamp=<10000 10000 10000>
    componentLinearAccScale=<1 1 1>
    componentLinearVelScale=<1 1 1>
    gravityOverride {z=-980}
    physicsBodyDefinitions [{
      boundBone {boneName=dyn_BeltStrap_01}
      boxExtents=<15 20 20>
      constraintSetup {angularLimitsMax=<3 3 2>, angularLimitsMin=<-3 -40 -2>, angularTarget=<1 0 0>}
      localJointOffset=<3 0 0>
     }]
    relativeSpaceBone {boneName=thigh_r}
    sphericalLimits [{limitRadius=11, drivingBone {boneName=thigh_r}, sphereLocalOffset=<2 -1.5 -2>}]
   }
   n5 = Anim Dynamics - Bone: dyn_BeltStrap_02!(SimulationSpace:AnimPhysSimSpaceType Enum, ComponentPose=@n4, Alpha=1, GravityOverride:Vector="0,0,-980") -> (Pose) {
    angularDampingOverride=0.9999
    angularSpringConstant=6000
    bAngularSpring=true
    bOverrideAngularDamping=true
    bUseGravityOverride=true
    bUseSphericalLimits=true
    gravityScale=0.35
    simulationSpace=RootRelative
    alphaBoolBlend {blendOption=Linear}
    boundBone {boneName=dyn_BeltStrap_02}
    componentAppliedLinearAccClamp=<10000 10000 10000>
    componentLinearAccScale=<1 1 1>
    componentLinearVelScale=<1 1 1>
    gravityOverride {z=-980}
    physicsBodyDefinitions [{
      boundBone {boneName=dyn_BeltStrap_02}
      boxExtents=<15 20 20>
      constraintSetup {angularLimitsMax=<2 3 2>, angularLimitsMin=<-2 -50 -2>, angularTarget=<1 0 0>}
      localJointOffset=<3 0 0>
     }]
    relativeSpaceBone {boneName=thigh_r}
    sphericalLimits [{limitRadius=12, drivingBone {boneName=thigh_r}, sphereLocalOffset=<8 0 -4>}]
   }
   n6 = Anim Dynamics - Bone: dyn_BeltStrap_03!(SimulationSpace:AnimPhysSimSpaceType Enum=World, ComponentPose=@n5, Alpha=1, GravityOverride:Vector="0,0,-980") -> (Pose) {
    angularDampingOverride=0.9999
    angularSpringConstant=6000
    bAngularSpring=true
    bOverrideAngularDamping=true
    bUseGravityOverride=true
    bUseSphericalLimits=true
    gravityScale=0.35
    simulationSpace=World
    alphaBoolBlend {blendOption=Linear}
    boundBone {boneName=dyn_BeltStrap_03}
    componentAppliedLinearAccClamp=<10000 10000 10000>
    componentLinearAccScale=<1 1 1>
    componentLinearVelScale=<1 1 1>
    gravityOverride {z=-980}
    physicsBodyDefinitions [{
      boundBone {boneName=dyn_BeltStrap_03}
      boxExtents=<15 20 20>
      constraintSetup {angularLimitsMax=<1 3 2>, angularLimitsMin=<-1 -40 -2>, angularTarget=<1 0 0>}
      localJointOffset=<3 0 0>
     }]
    relativeSpaceBone {boneName=thigh_r}
    sphericalLimits [
     {limitRadius=11.5, drivingBone {boneName=thigh_r}, sphereLocalOffset=<12.5 0 0>}
     {limitRadius=11.5, drivingBone {boneName=thigh_r}, sphereLocalOffset=<18 0 0>}
    ]
   }
   n7 = Anim Dynamics - Bone: dyn_BeltStrap_04!(SimulationSpace:AnimPhysSimSpaceType Enum=World, ComponentPose=@n6, Alpha=1, GravityOverride:Vector="0,0,-980") -> (Pose) {
    angularDampingOverride=0.9999
    angularSpringConstant=6000
    bAngularSpring=true
    bOverrideAngularDamping=true
    bUseGravityOverride=true
    bUseSphericalLimits=true
    gravityScale=0.35
    simulationSpace=World
    alphaBoolBlend {blendOption=Linear}
    boundBone {boneName=dyn_BeltStrap_04}
    componentAppliedLinearAccClamp=<10000 10000 10000>
    componentLinearAccScale=<1 1 1>
    componentLinearVelScale=<1 1 1>
    gravityOverride {z=-980}
    physicsBodyDefinitions [{
      boundBone {boneName=dyn_BeltStrap_04}
      boxExtents=<15 20 20>
      constraintSetup {angularLimitsMax=<0 2 2>, angularLimitsMin=<0 -40 -14>, angularTarget=<1 0 0>}
      localJointOffset=<3 0 0>
     }]
    relativeSpaceBone {boneName=thigh_r}
    sphericalLimits [
     {limitRadius=11.5, drivingBone {boneName=thigh_r}, sphereLocalOffset=<12.5 0 0>}
     {limitRadius=11.5, drivingBone {boneName=thigh_r}, sphereLocalOffset=<18 0 0>}
    ]
   }
   n8 = Anim Dynamics - Bone: dyn_R_Granade01!(ComponentPose=@n7, Alpha=1, GravityOverride:Vector="0,0,-980") -> (Pose) {
    angularDampingOverride=0.99
    angularSpringConstant=10000
    bAngularSpring=true
    bOverrideAngularDamping=true
    bUseGravityOverride=true
    gravityScale=0.35
    simulationSpace=BoneRelative
    alphaBoolBlend {blendOption=Linear}
    boundBone {boneName=dyn_R_Granade01}
    componentAppliedLinearAccClamp=<10000 10000 10000>
    componentLinearAccScale=<1 1 1>
    componentLinearVelScale=<1 1 1>
    gravityOverride {z=-980}
    physicsBodyDefinitions [{
      boundBone {boneName=dyn_R_Granade01}
      boxExtents=<26 46 50>
      constraintSetup {angularLimitsMax=<20 20 20>, angularLimitsMin=<-20 -50 -10>, angularTarget=<1 0 0>}
      localJointOffset=<1.5 0 0>
     }]
    relativeSpaceBone {boneName=pelvis}
   }
   n9 = Anim Dynamics - Bone: dyn_R_Granade02!(ComponentPose=@n8, Alpha=1, GravityOverride:Vector="0,0,-980") -> (Pose) {
    angularDampingOverride=0.99
    angularSpringConstant=10000
    bAngularSpring=true
    bOverrideAngularDamping=true
    bUseGravityOverride=true
    gravityScale=0.35
    simulationSpace=BoneRelative
    alphaBoolBlend {blendOption=Linear}
    boundBone {boneName=dyn_R_Granade02}
    componentAppliedLinearAccClamp=<10000 10000 10000>
    componentLinearAccScale=<1 1 1>
    componentLinearVelScale=<1 1 1>
    gravityOverride {z=-980}
    physicsBodyDefinitions [{
      boundBone {boneName=dyn_R_Granade02}
      boxExtents=<22 32 46>
      constraintSetup {angularLimitsMax=<20 20 10>, angularLimitsMin=<-20 -50 -20>, angularTarget=<1 0 0>}
      localJointOffset=<1.5 0 0>
     }]
    relativeSpaceBone {boneName=pelvis}
   }
   n10 = Anim Dynamics - Bone: dyn_R_Granade01_Ring!(ComponentPose=@n9, Alpha=1, GravityOverride:Vector="0,0,-980") -> (Pose) {
    angularDampingOverride=1
    angularSpringConstant=6500
    bAngularSpring=true
    bOverrideAngularDamping=true
    bUseGravityOverride=true
    gravityScale=0.35
    simulationSpace=RootRelative
    alphaBoolBlend {blendOption=Linear}
    boundBone {boneName=dyn_R_Granade01_Ring}
    componentAppliedLinearAccClamp=<10000 10000 10000>
    componentLinearAccScale=<1 1 1>
    componentLinearVelScale=<1 1 1>
    gravityOverride {z=-980}
    physicsBodyDefinitions [{
      boundBone {boneName=dyn_R_Granade01_Ring}
      boxExtents=<10 16 8>
      constraintSetup {angularLimitsMax=<5 90 60>, angularLimitsMin=<-5 0 -60>, angularTarget=<0 0 0>}
      localJointOffset=<2 0 0>
     }]
   }
   n11 = Anim Dynamics - Bone: dyn_R_Granade02_Ring!(ComponentPose=@n10, Alpha=1, GravityOverride:Vector="0,0,-980") -> (Pose) {
    angularDampingOverride=1
    angularSpringConstant=6500
    bAngularSpring=true
    bOverrideAngularDamping=true
    bUseGravityOverride=true
    gravityScale=0.35
    simulationSpace=RootRelative
    alphaBoolBlend {blendOption=Linear}
    boundBone {boneName=dyn_R_Granade02_Ring}
    componentAppliedLinearAccClamp=<10000 10000 10000>
    componentLinearAccScale=<1 1 1>
    componentLinearVelScale=<1 1 1>
    gravityOverride {z=-980}
    physicsBodyDefinitions [{
      boundBone {boneName=dyn_R_Granade02_Ring}
      boxExtents=<10 16 8>
      constraintSetup {angularLimitsMax=<5 90 60>, angularLimitsMin=<-5 0 -60>, angularTarget=<0 0 0>}
      localJointOffset=<2 0 0>
     }]
   }
   n12 = Anim Dynamics - Bone: dyn_Tail!(ComponentPose=@n11, Alpha=1, GravityOverride:Vector="0,0,-980") -> (Pose) {
    angularDampingOverride=1
    angularSpringConstant=6500
    bAngularSpring=true
    bOverrideAngularDamping=true
    bUseGravityOverride=true
    gravityScale=0.35
    simulationSpace=RootRelative
    alphaBoolBlend {blendOption=Linear}
    boundBone {boneName=dyn_Tail}
    componentAppliedLinearAccClamp=<10000 10000 10000>
    componentLinearAccScale=<1 1 1>
    componentLinearVelScale=<1 1 1>
    gravityOverride {z=-980}
    physicsBodyDefinitions [{
      boundBone {boneName=dyn_Tail}
      boxExtents=<23 33 33>
      constraintSetup {angularLimitsMax=<3 60 40>, angularLimitsMin=<-3 -60 -40>, angularTarget=<1 -0.15 0>}
      localJointOffset=<1.5 0 0>
     }]
   }
   n13 = Anim Dynamics - Bone: dyn_L_Bullet_01!(ComponentPose=@n12, Alpha=0.75, GravityOverride:Vector="0,0,-980") -> (Pose) {
    angularDampingOverride=0.99
    angularSpringConstant=12000
    bAngularSpring=true
    bOverrideAngularDamping=true
    bUseGravityOverride=true
    gravityScale=0.35
    simulationSpace=RootRelative
    alphaBoolBlend {blendOption=Linear}
    boundBone {boneName=dyn_L_Bullet_01}
    componentAppliedLinearAccClamp=<10000 10000 10000>
    componentLinearAccScale=<1 1 1>
    componentLinearVelScale=<1 1 1>
    gravityOverride {z=-980}
    physicsBodyDefinitions [{
      boundBone {boneName=dyn_L_Bullet_01}
      boxExtents=<17 47 27>
      constraintSetup {angularLimitsMax=<2 10 20>, angularLimitsMin=<-2 -20 -15>, angularTarget=<1 0 0>}
      localJointOffset=<2.8 0 0>
     }]
    relativeSpaceBone {boneName=pelvis}
   }
   n14 = Anim Dynamics - Bone: dyn_L_Bullet_02!(ComponentPose=@n13, Alpha=0.75, GravityOverride:Vector="0,0,-980") -> (Pose) {
    angularDampingOverride=0.99
    angularSpringConstant=12000
    bAngularSpring=true
    bOverrideAngularDamping=true
    bUseGravityOverride=true
    gravityScale=0.35
    simulationSpace=RootRelative
    alphaBoolBlend {blendOption=Linear}
    boundBone {boneName=dyn_L_Bullet_02}
    componentAppliedLinearAccClamp=<10000 10000 10000>
    componentLinearAccScale=<1 1 1>
    componentLinearVelScale=<1 1 1>
    gravityOverride {z=-980}
    physicsBodyDefinitions [{
      boundBone {boneName=dyn_L_Bullet_02}
      boxExtents=<15 55 25>
      constraintSetup {angularLimitsMax=<2 10 15>, angularLimitsMin=<-2 -20 -15>, angularTarget=<1 0 0>}
      localJointOffset=<3 0 0>
     }]
    relativeSpaceBone {boneName=pelvis}
   }
   n15 = Anim Dynamics - Bone: dyn_L_Bullet_03!(ComponentPose=@n14, Alpha=0.75, GravityOverride:Vector="0,0,-980") -> (Pose) {
    angularDampingOverride=0.99
    angularSpringConstant=12000
    bAngularSpring=true
    bOverrideAngularDamping=true
    bUseGravityOverride=true
    gravityScale=0.35
    simulationSpace=RootRelative
    alphaBoolBlend {blendOption=Linear}
    boundBone {boneName=dyn_L_Bullet_03}
    componentAppliedLinearAccClamp=<10000 10000 10000>
    componentLinearAccScale=<1 1 1>
    componentLinearVelScale=<1 1 1>
    gravityOverride {z=-980}
    physicsBodyDefinitions [{
      boundBone {boneName=dyn_L_Bullet_03}
      boxExtents=<17 57 27>
      constraintSetup {angularLimitsMax=<2 10 15>, angularLimitsMin=<-2 -20 -60>, angularTarget=<1 0 0>}
      localJointOffset=<2.5 0 0>
     }]
    relativeSpaceBone {boneName=pelvis}
   }
   Save cached pose 'DynamicsCache'(Pose=Component To Local(ComponentPose=@n15)) {cacheName=DynamicsCache}
  ]}
 EventGraph {nodes [
   Event BlueprintUpdateAnimation()
   > n1 = Gravity Override State Function DEPRECATED(CharacterPartAnimInstance=Self-Reference(), GameTimeInSeconds=GetGameTimeInSeconds(), GravityOverrideParams=Get Gravity Override Params())
   > Set GravityOverride(GravityOverride=@n1)
   n2 = GetAnimInstance(self=GetOwnerSkeletalMeshComponent())
   > Is Valid(InputObject=@n2)
   > Set Is Custom Emote Playing(Is Custom Emote Playing:bool=Montage_IsPlaying(self=@n2, Montage=/Game/Animation/Game/MainPlayer/Emotes/Emote_Celebration_M)) -> (Output_Get:bool)
   > n3 = Branch(Condition=Get bIsInFrontEnd())
   > Set AnimDyn SimSpace(AnimDyn SimSpace:AnimPhysSimSpaceType Enum=World) -> (Output_Get:AnimPhysSimSpaceType Enum)
   Set AnimDyn SimSpace(execute=@n3.else)
   Event OnParachuteOpened()
   > Switch on EGliderType(Selection=Get PlayerGliderType() -> (PlayerGliderType=HangGlider))
   > Set Is Using Hang Gliding in Parachute(Is Using Hang Gliding in Parachute:bool=true) -> (Output_Get:bool)
   Event OnParachuteClosed()
   > Set Is Using Hang Gliding in Parachute(Is Using Hang Gliding in Parachute=false)
   Event OnBeginCrouching()
   > Set fCrouching(fCrouching:double=1) -> (Output_Get:double)
   Event OnEndCrouching()
   > Set fCrouching(fCrouching=0)
   Event OnCharacterAndPickaxeCustomizationLoaded!() -> (
    OutputDelegate
    then
    CustomBaseMesh:Skeletal Mesh Component Object Reference
    CustomParts:Array of Custom Character Part Object References
    CustomMeshParts:Array of Skeletal Mesh Component Budgeted Object References
   ) {is=title K2Node_Event}
   n4 = CheckForBackbling()
   > Set isWearingBackblingLong(isWearingBackblingLong:bool=OR Boolean(A=@n4.bIsBackblingACape, B=@n4.bIsBackblingLong)) -> (Output_Get:bool)
  ]}
]
variables [
 AnimDyn SimSpace {type=AnimPhysSimSpaceType, default=RootRelative}
 fCrouching {type=number}
 Gravity Override Params {type=GravityOverrideParamsStruct}
 GravityOverride {type=Vector}
 Is Custom Emote Playing {type=boolean}
 Is Using Hang Gliding in Parachute {type=boolean}
 isWearingBackblingLong {type=boolean}
]
