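# Control Rig for the HappyHopper medium male face accessory, procedurally correcting its hood and ear bones as the character moves.
# Swing-twist head deltas and a right-clavicle pose reader become normalized, often curve-shaped weights.
# These weights blend offsets onto hood and ear bones, branching on crouch, skydiving, aiming and weapon states.
# While crouched, aim-constrained nulls and a quaternion lerp also steer the right ear.
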ControlRigBlueprint /BRCosmetics/Characters/Player/Male/Medium/Bodies/M_MED_HappyHopper/Meshes/Parts/M_MED_HappyHopper_FaceAcc_CtrlRig
graphs [
 Multiply_2_ContainedGraph {nodes [
   n1 = RigVMFunctionEntryNode()
   RigVMFunctionReturnNode(Result=&RigVMFunction_MathDoubleMul(B=@n1.C, A=&RigVMFunction_MathDoubleMul(A=@n1.A, B=@n1.B).Result).Result)
  ]}
 Multiply_3_ContainedGraph {nodes [
   n1 = RigVMFunctionEntryNode()
   RigVMFunctionReturnNode(Result=&RigVMFunction_MathDoubleMul(B=@n1.C, A=&RigVMFunction_MathDoubleMul(A=@n1.A, B=@n1.B).Result).Result)
  ]}
 New Function_ContainedGraph {nodes [
   n1 = RigVMFunctionEntryNode()
   RigVMFunction_MathBoolToInteger = &RigVMFunction_MathBoolToInteger(Value=@n1.onlyTwist)
   RigVMFunction_MathQuaternionSwingTwist = &RigVMFunction_MathQuaternionSwingTwist(
    TwistAxis=@n1.twistAxis
    Input=&RigUnit_GetRelativeTransformForItem(
     Child=(Type=Bone,Name="None")
     Parent=(Type=Bone,Name="None")
     bChildInitial=False
     bParentInitial=False
     Child.Name=@n1.child
     Parent.Name=@n1.parent
     Parent.Type=@n1.parentType
    ).RelativeTransform.Rotation
   )
   RigVMFunction_MathQuaternionToEuler = &RigVMFunction_MathQuaternionToEuler(
    Value=RigVMDispatchNode(
     Values=((),())
     Index=@RigVMFunction_MathBoolToInteger.Result
     Values.0=@RigVMFunction_MathQuaternionSwingTwist.Swing
     Values.1=@RigVMFunction_MathQuaternionSwingTwist.Twist
    ) {template="DISPATCH_RigVMDispatch_SelectInt32(in Index,in Values,out Result)"}.Result
    RotationOrder=@n1.Rotation Order
   )
   > DISPATCH_RigVMDispatch_Print = RigVMDispatchNode(ScreenColor=(R=1,G=1,B=1,A=1), ScreenDuration=1.5, Enabled=@n1.debugRotation, Value=@RigVMFunction_MathQuaternionToEuler.Result) {template="DISPATCH_RigVMDispatch_Print(in Prefix,in Value,in Enabled,in ScreenDuration,in ScreenColor)"}
   Clamp = &RigVMFunction_MathIntClamp(Value=@n1.primAxis)
   DISPATCH_RigVMDispatch_SelectInt32 = RigVMDispatchNode(
    Values=(0,0,0)
    Index=@Clamp.Result
    Values.0=@RigVMFunction_MathQuaternionToEuler.Result.X
    Values.1=@RigVMFunction_MathQuaternionToEuler.Result.Y
    Values.2=@RigVMFunction_MathQuaternionToEuler.Result.Z
   ) {template="DISPATCH_RigVMDispatch_SelectInt32(in Index,in Values,out Result)"}
   Remap = &RigVMFunction_MathDoubleRemap(
    TargetMaximum=1
    TargetMinimum=0
    bClamp=False
    Value=@DISPATCH_RigVMDispatch_SelectInt32.Result
    SourceMaximum=@n1.max
    SourceMinimum=@n1.min
   )
   Multiply = &RigVMFunction_MathDoubleMul(A=&RigVMFunction_MathDoubleClamp(Value=@Remap.Result).Result, B=@n1.premult)
   > RigVMFunctionReturnNode(outputWeight=@Multiply.Result)
  ]}
 New Function_ContainedGraph {nodes [
   n1 = RigVMFunctionEntryNode()
   RigVMFunction_MathBoolToInteger = &RigVMFunction_MathBoolToInteger(Value=@n1.onlyTwist)
   RigVMFunction_MathQuaternionSwingTwist = &RigVMFunction_MathQuaternionSwingTwist(
    TwistAxis=@n1.twistAxis
    Input=&RigUnit_GetRelativeTransformForItem(
     Child=(Type=Bone,Name="None")
     Parent=(Type=Bone,Name="None")
     bChildInitial=False
     bParentInitial=False
     Child.Name=@n1.child
     Parent.Name=@n1.parent
     Parent.Type=@n1.parentType
    ).RelativeTransform.Rotation
   )
   RigVMFunction_MathQuaternionToEuler = &RigVMFunction_MathQuaternionToEuler(
    Value=RigVMDispatchNode(
     Values=((),())
     Index=@RigVMFunction_MathBoolToInteger.Result
     Values.0=@RigVMFunction_MathQuaternionSwingTwist.Swing
     Values.1=@RigVMFunction_MathQuaternionSwingTwist.Twist
    ) {template="DISPATCH_RigVMDispatch_SelectInt32(in Index,in Values,out Result)"}.Result
    RotationOrder=@n1.Rotation Order
   )
   > DISPATCH_RigVMDispatch_Print = RigVMDispatchNode(ScreenColor=(R=1,G=1,B=1,A=1), ScreenDuration=1.5, Enabled=@n1.debugRotation, Value=@RigVMFunction_MathQuaternionToEuler.Result) {template="DISPATCH_RigVMDispatch_Print(in Prefix,in Value,in Enabled,in ScreenDuration,in ScreenColor)"}
   Clamp = &RigVMFunction_MathIntClamp(Value=@n1.primAxis)
   DISPATCH_RigVMDispatch_SelectInt32 = RigVMDispatchNode(
    Values=(0,0,0)
    Index=@Clamp.Result
    Values.0=@RigVMFunction_MathQuaternionToEuler.Result.X
    Values.1=@RigVMFunction_MathQuaternionToEuler.Result.Y
    Values.2=@RigVMFunction_MathQuaternionToEuler.Result.Z
   ) {template="DISPATCH_RigVMDispatch_SelectInt32(in Index,in Values,out Result)"}
   Remap = &RigVMFunction_MathDoubleRemap(
    TargetMaximum=1
    TargetMinimum=0
    bClamp=False
    Value=@DISPATCH_RigVMDispatch_SelectInt32.Result
    SourceMaximum=@n1.max
    SourceMinimum=@n1.min
   )
   Multiply = &RigVMFunction_MathDoubleMul(A=&RigVMFunction_MathDoubleClamp(Value=@Remap.Result).Result, B=@n1.premult)
   > RigVMFunctionReturnNode(outputWeight=@Multiply.Result)
  ]}
 New Function_ContainedGraph {nodes [
   n1 = RigVMFunctionEntryNode()
   RigVMFunction_MathBoolToInteger = &RigVMFunction_MathBoolToInteger(Value=@n1.onlyTwist)
   RigVMFunction_MathQuaternionSwingTwist = &RigVMFunction_MathQuaternionSwingTwist(
    TwistAxis=@n1.twistAxis
    Input=&RigUnit_GetRelativeTransformForItem(
     Child=(Type=Bone,Name="None")
     Parent=(Type=Bone,Name="None")
     bChildInitial=False
     bParentInitial=False
     Child.Name=@n1.child
     Parent.Name=@n1.parent
     Parent.Type=@n1.parentType
    ).RelativeTransform.Rotation
   )
   RigVMFunction_MathQuaternionToEuler = &RigVMFunction_MathQuaternionToEuler(
    Value=RigVMDispatchNode(
     Values=((),())
     Index=@RigVMFunction_MathBoolToInteger.Result
     Values.0=@RigVMFunction_MathQuaternionSwingTwist.Swing
     Values.1=@RigVMFunction_MathQuaternionSwingTwist.Twist
    ) {template="DISPATCH_RigVMDispatch_SelectInt32(in Index,in Values,out Result)"}.Result
    RotationOrder=@n1.Rotation Order
   )
   > DISPATCH_RigVMDispatch_Print = RigVMDispatchNode(ScreenColor=(R=1,G=1,B=1,A=1), ScreenDuration=1.5, Enabled=@n1.debugRotation, Value=@RigVMFunction_MathQuaternionToEuler.Result) {template="DISPATCH_RigVMDispatch_Print(in Prefix,in Value,in Enabled,in ScreenDuration,in ScreenColor)"}
   Clamp = &RigVMFunction_MathIntClamp(Value=@n1.primAxis)
   DISPATCH_RigVMDispatch_SelectInt32 = RigVMDispatchNode(
    Values=(0,0,0)
    Index=@Clamp.Result
    Values.0=@RigVMFunction_MathQuaternionToEuler.Result.X
    Values.1=@RigVMFunction_MathQuaternionToEuler.Result.Y
    Values.2=@RigVMFunction_MathQuaternionToEuler.Result.Z
   ) {template="DISPATCH_RigVMDispatch_SelectInt32(in Index,in Values,out Result)"}
   Remap = &RigVMFunction_MathDoubleRemap(
    TargetMaximum=1
    TargetMinimum=0
    bClamp=False
    Value=@DISPATCH_RigVMDispatch_SelectInt32.Result
    SourceMaximum=@n1.max
    SourceMinimum=@n1.min
   )
   Multiply = &RigVMFunction_MathDoubleMul(A=&RigVMFunction_MathDoubleClamp(Value=@Remap.Result).Result, B=@n1.premult)
   > RigVMFunctionReturnNode(outputWeight=@Multiply.Result)
  ]}
 New Function_ContainedGraph {nodes [
   n1 = RigVMFunctionEntryNode()
   RigVMFunction_ControlFlowBranch = &RigVMFunction_ControlFlowBranch(ExecuteContext=@n1.Execute, Condition=@n1.On)
   DISPATCH_RigVMDispatch_ArrayIterator = RigVMDispatchNode(
    Array=/Script/ControlRig.RigUnit_CollectionGetAll(TypeToSearch=Control) {is=struct RigVMUnitNode}.Items
    ExecuteContext=@RigVMFunction_ControlFlowBranch.True
   ) {template="DISPATCH_RigVMDispatch_ArrayIterator(in Array,out Element,out Index,out Count,out Ratio)"}
   Replace = /Script/RigVM.RigVMFunction_NameReplace(Old=_ctrl, Name=@DISPATCH_RigVMDispatch_ArrayIterator.Element.Name) {is=struct RigVMUnitNode, template="Replace::Execute(in Name,in Old,in New,out Result)"}
   > &RigUnit_SetTransform(
    Item=(Type=Bone,Name="spine_01")
    Weight=1
    bInitial=False
    Item.Name=@Replace.Result
    Value=&RigUnit_GetTransform(Item=(Type=Control,Name="spine_01_ctrl"), bInitial=False, Item.Name=@DISPATCH_RigVMDispatch_ArrayIterator.Element.Name).Transform
   )
   RigVMFunctionReturnNode(Execute=@RigVMFunction_ControlFlowBranch.Completed)
  ]}
 New Function_ContainedGraph {nodes [
   n1 = RigVMFunctionEntryNode()
   RigVMFunction_Sequence = RigVMAggregateNode(ExecuteContext=@n1.Execute)
   DeltaRotation = RigVMCollapseNode(
    Rotation Order=ZYX
    child=head
    debugRotation=false
    max=60
    min=-60
    onlyTwist=true
    parent=head_Neck01Delta_null
    parentType=Null
    premult=1
    twistAxis=(X=1,Y=0,Z=0)
    ExecuteContext=@RigVMFunction_Sequence.A
   )
   Remap = &RigVMFunction_MathDoubleRemap(SourceMaximum=0, SourceMinimum=0.5, TargetMaximum=1, TargetMinimum=0, Value=@DeltaRotation.outputWeight)
   n2 = &RigVMFunction_AnimEvalRichCurve(
    Curve="(EditorCurveData=(Keys=((InterpMode=RCIM_Cubic,TangentMode=RCTM_User,ArriveTangent=2.9885,LeaveTangent=2.9885),(InterpMode=RCIM_Cubic,TangentMode=RCTM_User,Time=0.2798,Value=0.7009,ArriveTangent=1.1721,LeaveTangent=1.1721),(InterpMode=RCIM_Cubic,Time=1,Val ...(truncated)"
    SourceMinimum=0
    TargetMinimum=0
    Value=@Remap.Result
   )
   > RigVMVariableNode(Value=@n2.Result) {variable=IsLookingLeft}
   n3 = &RigVMFunction_AnimEvalRichCurve(
    Curve=(EditorCurveData=(Keys=((InterpMode=RCIM_Cubic,TangentMode=RCTM_User,ArriveTangent=1.872,LeaveTangent=1.872),(InterpMode=RCIM_Cubic,Time=1,Value=1))))
    SourceMinimum=0
    TargetMinimum=0
    Value=&RigVMFunction_MathDoubleRemap(SourceMaximum=1, SourceMinimum=0.5, TargetMaximum=1, TargetMinimum=0, Value=@DeltaRotation.outputWeight).Result
   )
   > RigVMVariableNode(Value=@n3.Result) {variable=IsLookingRight}
   > DeltaRotation_1 = RigVMCollapseNode(
    Rotation Order=ZYX
    child=head
    debugRotation=false
    max=-30
    min=0
    onlyTwist=false
    parent=head_Neck01Delta_null
    parentType=Null
    premult=1
    primAxis=2
    twistAxis=(X=1,Y=0,Z=0)
   )
   > RigVMVariableNode(Value=@DeltaRotation_1.outputWeight) {variable=TiltingHeadUp}
   > DeltaRotation_2 = RigVMCollapseNode(
    Rotation Order=ZYX
    child=head
    debugRotation=false
    max=20
    min=0
    onlyTwist=false
    parent=head_Neck01Delta_null
    parentType=Null
    premult=1
    primAxis=2
    twistAxis=(X=1,Y=0,Z=0)
   )
   > RigVMVariableNode(Value=@DeltaRotation_2.outputWeight) {variable=TiltingHeadDown}
   &RigUnit_ModifyTransforms(
    ItemToModify=((Item=(Type=Bone,Name="base_hood_l"),Transform=(Rotation=(X=0,Y=0,Z=-0.0872,W=0.9962),Translation=(X=10,Y=4,Z=-1),Scale3D=(X=1,Y=1,Z=1))))
    WeightMinimum=0
    ExecutePin=@RigVMFunction_Sequence.B
    Weight=RigVMVariableNode() {variable=IsLookingLeft}.Value
   )
   > &RigUnit_ModifyTransforms(
    ItemToModify=((Item=(Type=Bone,Name="base_hood_r"),Transform=(Rotation=(X=0,Y=0,Z=-0.0436,W=0.999),Translation=(X=0,Y=0,Z=0),Scale3D=(X=1,Y=1,Z=1))))
    WeightMinimum=0
    Weight=RigVMVariableNode() {variable=IsLookingLeft}.Value
   )
   > &RigUnit_ModifyTransforms(
    ItemToModify=((Item=(Type=Bone,Name="base_hood_f"),Transform=(Rotation=(X=-0.0446,Y=-0.1227,Z=-0.3391,W=0.9317),Translation=(X=0,Y=1,Z=-0.5),Scale3D=(X=1,Y=1,Z=1))))
    WeightMinimum=0
    Weight=RigVMVariableNode() {variable=IsLookingLeft}.Value
   )
   > &RigUnit_ModifyTransforms(
    ItemToModify=((Item=(Type=Bone,Name="base_hood_l"),Transform=(Rotation=(X=0,Y=0,Z=-0.0654,W=0.9979),Translation=(X=0,Y=0,Z=0),Scale3D=(X=1,Y=1,Z=1))))
    WeightMinimum=0
    Weight=RigVMVariableNode() {variable=IsLookingRight}.Value
   )
   > &RigUnit_ModifyTransforms(
    ItemToModify=((Item=(Type=Bone,Name="base_hood_r"),Transform=(Rotation=(X=-0.087,Y=-0.0057,Z=0.0652,W=0.9941),Translation=(X=-10,Y=-4,Z=-1),Scale3D=(X=1,Y=1,Z=1))))
    WeightMinimum=0
    Weight=RigVMVariableNode() {variable=IsLookingRight}.Value
   )
   > &RigUnit_ModifyTransforms(
    ItemToModify=((Item=(Type=Bone,Name="base_hood_f"),Transform=(Rotation=(X=-0.1296,Y=0.0871,Z=0.3885,W=0.9081),Translation=(X=0,Y=1,Z=-1.75),Scale3D=(X=1,Y=1,Z=1))))
    WeightMinimum=0
    Weight=RigVMVariableNode() {variable=IsLookingRight}.Value
   )
   &RigUnit_ModifyTransforms(
    ItemToModify=((Item=(Type=Bone,Name="base_hood_l"),Transform=(Rotation=(X=-0.0871,Y=0.0038,Z=-0.0435,W=0.9952),Translation=(X=0,Y=-6,Z=0),Scale3D=(X=1,Y=1,Z=1))))
    WeightMinimum=0
    ExecutePin=@RigVMFunction_Sequence.C
    Weight=RigVMVariableNode() {variable=TiltingHeadUp}.Value
   )
   > &RigUnit_ModifyTransforms(
    ItemToModify=((Item=(Type=Bone,Name="base_hood_r"),Transform=(Rotation=(X=-0.0871,Y=-0.0038,Z=0.0435,W=0.9952),Translation=(X=0,Y=-6,Z=0),Scale3D=(X=1,Y=1,Z=1))))
    WeightMinimum=0
    Weight=RigVMVariableNode() {variable=TiltingHeadUp}.Value
   )
   > &RigUnit_ModifyTransforms(
    ItemToModify=((Item=(Type=Bone,Name="base_hood_f"),Transform=(Rotation=(X=0.1736,Y=0,Z=0,W=0.9848),Translation=(X=0,Y=3,Z=-1),Scale3D=(X=1,Y=1,Z=1))))
    WeightMinimum=0
    Weight=RigVMVariableNode() {variable=TiltingHeadUp}.Value
   )
   > &RigUnit_ModifyTransforms(
    ItemToModify=((Item=(Type=Bone,Name="base_hood_l"),Transform=(Rotation=(X=0,Y=0,Z=-0.0872,W=0.9962),Translation=(X=3,Y=10,Z=0),Scale3D=(X=1,Y=1,Z=1))))
    WeightMinimum=0
    Weight=RigVMVariableNode() {variable=TiltingHeadDown}.Value
   )
   > &RigUnit_ModifyTransforms(
    ItemToModify=((Item=(Type=Bone,Name="base_hood_r"),Transform=(Rotation=(X=0,Y=0,Z=0.0872,W=0.9962),Translation=(X=-3,Y=10,Z=0),Scale3D=(X=1,Y=1,Z=1))))
    WeightMinimum=0
    Weight=RigVMVariableNode() {variable=TiltingHeadDown}.Value
   )
   > &RigUnit_ModifyTransforms(
    ItemToModify=((Item=(Type=Bone,Name="base_hood_f"),Transform=(Rotation=(X=0.1736,Y=0,Z=0,W=0.9848),Translation=(X=0,Y=0,Z=0),Scale3D=(X=1,Y=1,Z=1))))
    WeightMinimum=0
    Weight=RigVMVariableNode() {variable=TiltingHeadDown}.Value
   )
   RigVMFunctionReturnNode(Execute=@RigVMFunction_Sequence.D)
  ]}
 New Function_ContainedGraph {nodes [
   n1 = RigVMFunctionEntryNode()
   RigVMFunction_MathBoolToInteger = &RigVMFunction_MathBoolToInteger(Value=@n1.onlyTwist)
   RigVMFunction_MathQuaternionSwingTwist = &RigVMFunction_MathQuaternionSwingTwist(
    TwistAxis=@n1.twistAxis
    Input=&RigUnit_GetRelativeTransformForItem(
     Child=(Type=Bone,Name="None")
     Parent=(Type=Bone,Name="None")
     bChildInitial=False
     bParentInitial=False
     Child.Name=@n1.child
     Parent.Name=@n1.parent
     Parent.Type=@n1.parentType
    ).RelativeTransform.Rotation
   )
   RigVMFunction_MathQuaternionToEuler = &RigVMFunction_MathQuaternionToEuler(
    Value=RigVMDispatchNode(
     Values=((),())
     Index=@RigVMFunction_MathBoolToInteger.Result
     Values.0=@RigVMFunction_MathQuaternionSwingTwist.Swing
     Values.1=@RigVMFunction_MathQuaternionSwingTwist.Twist
    ) {template="DISPATCH_RigVMDispatch_SelectInt32(in Index,in Values,out Result)"}.Result
    RotationOrder=@n1.Rotation Order
   )
   > DISPATCH_RigVMDispatch_Print = RigVMDispatchNode(ScreenColor=(R=1,G=1,B=1,A=1), ScreenDuration=1.5, Enabled=@n1.debugRotation, Value=@RigVMFunction_MathQuaternionToEuler.Result) {template="DISPATCH_RigVMDispatch_Print(in Prefix,in Value,in Enabled,in ScreenDuration,in ScreenColor)"}
   Clamp = &RigVMFunction_MathIntClamp(Value=@n1.primAxis)
   DISPATCH_RigVMDispatch_SelectInt32 = RigVMDispatchNode(
    Values=(0,0,0)
    Index=@Clamp.Result
    Values.0=@RigVMFunction_MathQuaternionToEuler.Result.X
    Values.1=@RigVMFunction_MathQuaternionToEuler.Result.Y
    Values.2=@RigVMFunction_MathQuaternionToEuler.Result.Z
   ) {template="DISPATCH_RigVMDispatch_SelectInt32(in Index,in Values,out Result)"}
   Remap = &RigVMFunction_MathDoubleRemap(
    TargetMaximum=1
    TargetMinimum=0
    bClamp=False
    Value=@DISPATCH_RigVMDispatch_SelectInt32.Result
    SourceMaximum=@n1.max
    SourceMinimum=@n1.min
   )
   Multiply = &RigVMFunction_MathDoubleMul(A=&RigVMFunction_MathDoubleClamp(Value=@Remap.Result).Result, B=@n1.premult)
   > RigVMFunctionReturnNode(outputWeight=@Multiply.Result)
  ]}
 New Function_ContainedGraph {nodes [
   n1 = RigVMFunctionEntryNode()
   RigVMFunction_MathBoolToInteger = &RigVMFunction_MathBoolToInteger(Value=@n1.onlyTwist)
   RigVMFunction_MathQuaternionSwingTwist = &RigVMFunction_MathQuaternionSwingTwist(
    TwistAxis=@n1.twistAxis
    Input=&RigUnit_GetRelativeTransformForItem(
     Child=(Type=Bone,Name="None")
     Parent=(Type=Bone,Name="None")
     bChildInitial=False
     bParentInitial=False
     Child.Name=@n1.child
     Parent.Name=@n1.parent
     Parent.Type=@n1.parentType
    ).RelativeTransform.Rotation
   )
   RigVMFunction_MathQuaternionToEuler = &RigVMFunction_MathQuaternionToEuler(
    Value=RigVMDispatchNode(
     Values=((),())
     Index=@RigVMFunction_MathBoolToInteger.Result
     Values.0=@RigVMFunction_MathQuaternionSwingTwist.Swing
     Values.1=@RigVMFunction_MathQuaternionSwingTwist.Twist
    ) {template="DISPATCH_RigVMDispatch_SelectInt32(in Index,in Values,out Result)"}.Result
    RotationOrder=@n1.Rotation Order
   )
   > DISPATCH_RigVMDispatch_Print = RigVMDispatchNode(ScreenColor=(R=1,G=1,B=1,A=1), ScreenDuration=1.5, Enabled=@n1.debugRotation, Value=@RigVMFunction_MathQuaternionToEuler.Result) {template="DISPATCH_RigVMDispatch_Print(in Prefix,in Value,in Enabled,in ScreenDuration,in ScreenColor)"}
   Clamp = &RigVMFunction_MathIntClamp(Value=@n1.primAxis)
   DISPATCH_RigVMDispatch_SelectInt32 = RigVMDispatchNode(
    Values=(0,0,0)
    Index=@Clamp.Result
    Values.0=@RigVMFunction_MathQuaternionToEuler.Result.X
    Values.1=@RigVMFunction_MathQuaternionToEuler.Result.Y
    Values.2=@RigVMFunction_MathQuaternionToEuler.Result.Z
   ) {template="DISPATCH_RigVMDispatch_SelectInt32(in Index,in Values,out Result)"}
   Remap = &RigVMFunction_MathDoubleRemap(
    TargetMaximum=1
    TargetMinimum=0
    bClamp=False
    Value=@DISPATCH_RigVMDispatch_SelectInt32.Result
    SourceMaximum=@n1.max
    SourceMinimum=@n1.min
   )
   Multiply = &RigVMFunction_MathDoubleMul(A=&RigVMFunction_MathDoubleClamp(Value=@Remap.Result).Result, B=@n1.premult)
   > RigVMFunctionReturnNode(outputWeight=@Multiply.Result)
  ]}
 New Function_ContainedGraph {nodes [
   n1 = RigVMFunctionEntryNode()
   RigVMFunction_MathBoolToInteger = &RigVMFunction_MathBoolToInteger(Value=@n1.onlyTwist)
   RigVMFunction_MathQuaternionSwingTwist = &RigVMFunction_MathQuaternionSwingTwist(
    TwistAxis=@n1.twistAxis
    Input=&RigUnit_GetRelativeTransformForItem(
     Child=(Type=Bone,Name="None")
     Parent=(Type=Bone,Name="None")
     bChildInitial=False
     bParentInitial=False
     Child.Name=@n1.child
     Parent.Name=@n1.parent
     Parent.Type=@n1.parentType
    ).RelativeTransform.Rotation
   )
   RigVMFunction_MathQuaternionToEuler = &RigVMFunction_MathQuaternionToEuler(
    Value=RigVMDispatchNode(
     Values=((),())
     Index=@RigVMFunction_MathBoolToInteger.Result
     Values.0=@RigVMFunction_MathQuaternionSwingTwist.Swing
     Values.1=@RigVMFunction_MathQuaternionSwingTwist.Twist
    ) {template="DISPATCH_RigVMDispatch_SelectInt32(in Index,in Values,out Result)"}.Result
    RotationOrder=@n1.Rotation Order
   )
   > DISPATCH_RigVMDispatch_Print = RigVMDispatchNode(ScreenColor=(R=1,G=1,B=1,A=1), ScreenDuration=1.5, Enabled=@n1.debugRotation, Value=@RigVMFunction_MathQuaternionToEuler.Result) {template="DISPATCH_RigVMDispatch_Print(in Prefix,in Value,in Enabled,in ScreenDuration,in ScreenColor)"}
   Clamp = &RigVMFunction_MathIntClamp(Value=@n1.primAxis)
   DISPATCH_RigVMDispatch_SelectInt32 = RigVMDispatchNode(
    Values=(0,0,0)
    Index=@Clamp.Result
    Values.0=@RigVMFunction_MathQuaternionToEuler.Result.X
    Values.1=@RigVMFunction_MathQuaternionToEuler.Result.Y
    Values.2=@RigVMFunction_MathQuaternionToEuler.Result.Z
   ) {template="DISPATCH_RigVMDispatch_SelectInt32(in Index,in Values,out Result)"}
   Remap = &RigVMFunction_MathDoubleRemap(
    TargetMaximum=1
    TargetMinimum=0
    bClamp=False
    Value=@DISPATCH_RigVMDispatch_SelectInt32.Result
    SourceMaximum=@n1.max
    SourceMinimum=@n1.min
   )
   Multiply = &RigVMFunction_MathDoubleMul(A=&RigVMFunction_MathDoubleClamp(Value=@Remap.Result).Result, B=@n1.premult)
   > RigVMFunctionReturnNode(outputWeight=@Multiply.Result)
  ]}
 New Function_ContainedGraph {nodes [
   n1 = RigVMFunctionEntryNode()
   RigVMFunction_MathQuaternionToEuler = &RigVMFunction_MathQuaternionToEuler(
    RotationOrder=ZYX
    Value=&RigUnit_GetRelativeTransformForItem(
     Child=(Type=Bone,Name="hand_l")
     Parent=(Type=Bone,Name="lowerarm_l")
     bChildInitial=False
     bParentInitial=False
     Child.Name=@n1.Child
     Parent.Name=@n1.Parent
    ).RelativeTransform.Rotation
   )
   Remap = &RigVMFunction_MathDoubleRemap(
    TargetMaximum=1
    TargetMinimum=0
    Value=@RigVMFunction_MathQuaternionToEuler.Result.Z
    SourceMaximum=@n1.Source Maximum
    SourceMinimum=@n1.Source Minimum
   )
   > RigVMFunctionReturnNode(Delta=@Remap.Result)
  ]}
 New Function_ContainedGraph {nodes [
   n1 = RigVMFunctionEntryNode()
   RigVMFunction_Sequence = &RigVMFunction_Sequence(ExecuteContext=@n1.ExecuteContext)
   RigVMFunction_ControlFlowBranch = &RigVMFunction_ControlFlowBranch(ExecuteContext=@RigVMFunction_Sequence.A, Condition=RigVMVariableNode() {variable=bIsCrouching}.Value)
   &RigUnit_ModifyTransforms(
    ItemToModify=((Item=(Type=Bone,Name="dyn_R_Ear_02"),Transform=(Rotation=(X=0,Y=0,Z=0,W=1),Translation=(X=0,Y=-20,Z=0),Scale3D=(X=1,Y=1,Z=1))))
    WeightMinimum=0
    ExecutePin=@RigVMFunction_ControlFlowBranch.Completed
    Weight=RigVMVariableNode() {variable=IsJumping}.Value
   )
   &RigUnit_ModifyTransforms(
    ItemToModify=((Item=(Type=Bone,Name="base_R_Ear_01"),Transform=(Rotation=(X=0,Y=0,Z=-0.1736,W=0.9848),Translation=(X=0,Y=0,Z=0),Scale3D=(X=1,Y=1,Z=1))))
    WeightMinimum=0
    ExecutePin=@RigVMFunction_ControlFlowBranch.False
    Weight=RigVMVariableNode() {variable=IsHeadLookingUp}.Value
   )
   > &RigUnit_ModifyTransforms(
    ItemToModify=((Item=(Type=Bone,Name="base_L_Ear_01"),Transform=(Rotation=(X=0,Y=0,Z=0.1736,W=0.9848),Translation=(X=0,Y=0,Z=0),Scale3D=(X=1,Y=1,Z=1))))
    WeightMinimum=0
    Weight=RigVMVariableNode() {variable=IsHeadLookingDown}.Value
   )
   n2 = &RigVMFunction_Sequence(ExecuteContext=@RigVMFunction_ControlFlowBranch.True)
   RigVMFunction_MathBoolOr = &RigVMFunction_MathBoolOr(A=RigVMVariableNode() {variable=bisPickaxeOut}.Value, B=RigVMVariableNode() {variable=IsPistolOut}.Value)
   RigVMFunction_MathBoolToFloat = &RigVMFunction_MathBoolToFloat(Value=&RigVMFunction_MathBoolNot(Value=@RigVMFunction_MathBoolOr.Result).Result)
   &RigUnit_AimConstraintLocalSpaceOffset(
    Child=(Type=Null,Name="dyn_R_Ear_02_ClavAim_null")
    Filter=(bX=True,bY=true,bZ=True)
    Parents=((Item=(Type=Null,Name="R_ArmAim_Null"),Weight=1))
    UpAxis=(X=0,Y=1,Z=0)
    Weight=1
    WorldUp=(Target=(X=0,Y=0,Z=1),Kind=Location,Space=(Type=None,Name="None"))
    bMaintainOffset=true
    WorldUp.Target=&RigUnit_GetTransform(Item=(Type=Bone,Name="spine_05"), bInitial=False).Transform.Translation
    Parents.0.Weight=@RigVMFunction_MathBoolToFloat.Result
    ExecutePin=@n2.A
   )
   RigVMFunction_MathBoolOr_2 = RigVMAggregateNode(
    C=RigVMVariableNode() {variable=IsAimingDownSights}.Value
    B=RigVMVariableNode() {variable=IsPistolOut}.Value
    A=RigVMVariableNode() {variable=bisPickaxeOut}.Value
   )
   Multiply_2 = RigVMAggregateNode(
    C=1
    B=&RigVMFunction_MathBoolToFloat(Value=&RigVMFunction_MathBoolNot(Value=@RigVMFunction_MathBoolOr_2.Result).Result).Result
    A=RigVMVariableNode() {variable=IsCrouchWalking}.Value
   )
   > &RigUnit_AimConstraintLocalSpaceOffset(
    Child=(Type=Null,Name="dyn_R_Ear_02_SpineAim_null")
    Filter=(bX=True,bY=True,bZ=True)
    Parents=((Item=(Type=Null,Name="R_SpineAim_null"),Weight=1))
    UpAxis=(X=0,Y=1,Z=0)
    Weight=1
    WorldUp=(Target=(X=0,Y=0,Z=1),Kind=Location,Space=(Type=None,Name="None"))
    bMaintainOffset=true
    Parents.0.Weight=@Multiply_2.Result
    WorldUp.Target=&RigUnit_GetTransform(Item=(Type=Bone,Name="spine_05"), bInitial=False).Transform.Translation
   )
   RigVMFunction_AnimEvalRichCurve = &RigVMFunction_AnimEvalRichCurve(
    Curve=(EditorCurveData=(Keys=((InterpMode=RCIM_Cubic),(InterpMode=RCIM_Cubic,Time=0.5633,Value=0.309,ArriveTangent=1,LeaveTangent=1),(InterpMode=RCIM_Cubic,TangentMode=RCTM_User,Time=1,Value=1,ArriveTangent=1.7143,LeaveTangent=1.7143))))
    SourceMinimum=0
    TargetMinimum=0
    Value=RigVMVariableNode() {variable=IsHeadLookingUp}.Value
   )
   > &RigUnit_ModifyTransforms(
    ItemToModify=((Item=(Type=Bone,Name="base_L_Ear_01"),Transform=(Rotation=(X=-0.0687,Y=0.1402,Z=-0.1176,W=0.9807),Translation=(X=0,Y=0,Z=0),Scale3D=(X=1,Y=1,Z=1))))
    WeightMinimum=0
    ItemToModify.0.Transform.Rotation=RigVMDispatchNode(
     False=(X=0.0544,Y=0.1261,Z=-0.092,W=0.9862)
     True=(X=0.092,Y=0.1261,Z=-0.0544,W=0.9862)
     Condition=RigVMVariableNode() {variable=bisPickaxeOut}.Value
    ) {template="DISPATCH_RigVMDispatch_If(in Condition,in True,in False,out Result)"}.Result
    Weight=@RigVMFunction_AnimEvalRichCurve.Result
   )
   > &RigUnit_ModifyTransforms(
    ItemToModify=((Item=(Type=Bone,Name="dyn_L_Ear_02"),Transform=(Rotation=(X=-0.2578,Y=0.0226,Z=-0.0842,W=0.9623),Translation=(X=0,Y=0,Z=0),Scale3D=(X=1,Y=1,Z=1))))
    WeightMinimum=0
    ItemToModify.0.Transform.Rotation=RigVMDispatchNode(
     False=(X=-0.2578,Y=0.0226,Z=-0.0842,W=0.9623)
     True=(X=0,Y=0,Z=-0.0872,W=0.9962)
     Condition=RigVMVariableNode() {variable=bisPickaxeOut}.Value
    ) {template="DISPATCH_RigVMDispatch_If(in Condition,in True,in False,out Result)"}.Result
    Weight=RigVMVariableNode() {variable=IsLookingUpExtreme}.Value
   )
   > n3 = &RigUnit_ModifyTransforms(
    ItemToModify=((Item=(Type=Bone,Name="base_R_Ear_01"),Transform=(Rotation=(X=-0.1798,Y=-0.0704,Z=-0.1006,W=0.976),Translation=(X=0,Y=0,Z=0),Scale3D=(X=1,Y=1,Z=1))))
    WeightMinimum=0
    Weight=RigVMVariableNode() {variable=IsHeadLookingUp}.Value
   )
   n4 = &RigVMFunction_ControlFlowBranch(
    ExecuteContext=@n3.ExecutePin
    Condition=&RigVMFunction_MathBoolOr(B=RigVMVariableNode() {variable=IsPistolOut}.Value, A=RigVMVariableNode() {variable=bisPickaxeOut}.Value).Result
   )
   DeltaRotation_2 = RigVMCollapseNode(
    Rotation Order=ZYX
    child=head
    debugRotation=false
    max=-90
    min=-45
    onlyTwist=false
    parent=head_Spine05Delta_null
    parentType=Null
    premult=1
    primAxis=2
    twistAxis=(X=1,Y=0,Z=0)
    ExecuteContext=@n4.Completed
   )
   Multiply = &RigVMFunction_MathDoubleMul(A=@DeltaRotation_2.outputWeight, B=RigVMVariableNode() {variable=IsCrouchWalking}.Value)
   n5 = &RigVMFunction_AnimEvalRichCurve(
    Curve="(EditorCurveData=(Keys=((InterpMode=RCIM_Cubic,TangentMode=RCTM_User,ArriveTangent=0.1796,LeaveTangent=0.1796),(InterpMode=RCIM_Cubic,TangentMode=RCTM_User,Time=0.5869,Value=0.3839,ArriveTangent=1.1768,LeaveTangent=1.1768),(InterpMode=RCIM_Cubic,TangentMod ...(truncated)"
    SourceMinimum=0
    TargetMinimum=0
    Value=@Multiply.Result
   )
   > &RigUnit_ModifyTransforms(
    ItemToModify=((Item=(Type=Bone,Name="base_R_Ear_01"),Transform=(Rotation=(X=-0.0038,Y=-0.0435,Z=-0.0871,W=0.9952),Translation=(X=0,Y=0,Z=0),Scale3D=(X=1,Y=1,Z=1))))
    WeightMinimum=0
    Weight=@n5.Result
   )
   > &RigUnit_ModifyTransforms(
    ItemToModify=((Item=(Type=Bone,Name="dyn_R_Ear_02"),Transform=(Rotation=(X=0,Y=0,Z=-0.2588,W=0.9659),Translation=(X=0,Y=0,Z=0),Scale3D=(X=1,Y=1,Z=1))))
    WeightMinimum=0
    Weight=@n5.Result
   )
   > &RigUnit_ModifyTransforms(
    ItemToModify=((Item=(Type=Bone,Name="base_L_Ear_01"),Transform=(Rotation=(X=-0.0151,Y=-0.0858,Z=-0.173,W=0.9811),Translation=(X=0,Y=0,Z=0),Scale3D=(X=1,Y=1,Z=1))))
    WeightMinimum=0
    Weight=@n5.Result
   )
   > &RigUnit_ModifyTransforms(
    ItemToModify=((Item=(Type=Bone,Name="dyn_L_Ear_02"),Transform=(Rotation=(X=0,Y=0,Z=0,W=1),Translation=(X=0,Y=0,Z=0),Scale3D=(X=1,Y=1,Z=1))))
    WeightMinimum=0
    Weight=@n5.Result
   )
   RigVMFunction_MathBoolNot = &RigVMFunction_MathBoolNot(Value=&RigVMFunction_MathBoolOr(A=RigVMVariableNode() {variable=IsPistolOut}.Value, B=RigVMVariableNode() {variable=bisPickaxeOut}.Value).Result)
   > &RigUnit_ModifyTransforms(
    ItemToModify=((Item=(Type=Bone,Name="dyn_R_Ear_02"),Transform=(Rotation=(X=0.0019,Y=-0.0436,Z=0.0436,W=0.9981),Translation=(X=0,Y=0,Z=0),Scale3D=(X=1,Y=1,Z=1))))
    WeightMinimum=0
    Weight=&RigVMFunction_MathDoubleMul(
     B=&RigVMFunction_MathBoolToFloat(Value=@RigVMFunction_MathBoolNot.Result).Result
     A=RigVMVariableNode() {variable=IsHeadLookingUp}.Value
    ).Result
   )
   RigUnit_GetTransform = &RigUnit_GetTransform(Item=(Type=Null,Name="dyn_R_Ear_02_ClavAim_null"), bInitial=False)
   AccumulateLerp = /Script/RigVM.RigVMFunction_AccumulateQuatLerp(
    bIntegrateDeltaTime=False
    Blend=&RigVMFunction_MathDoubleMul(B=0.75, A=RigVMVariableNode() {variable=IsCrouchWalking}.Value).Result
    InitialValue=@RigUnit_GetTransform.Transform.Rotation
    TargetValue=&RigUnit_GetTransform(Item=(Type=Null,Name="dyn_R_Ear_02_SpineAim_null"), bInitial=False).Transform.Rotation
   ) {is=struct RigVMUnitNode, template="AccumulateLerp::Execute(in TargetValue,in InitialValue,in Blend,in bIntegrateDeltaTime,out Result)"}
   Set Transform = &RigUnit_SetRotation(Item=(Type=Bone,Name="dyn_R_Ear_02"), Space=GlobalSpace, Weight=1, bInitial=False, Value=@AccumulateLerp.Result, ExecutePin=@n4.False)
   &RigUnit_ModifyTransforms(
    ItemToModify=((Item=(Type=Bone,Name="base_R_Ear_01"),Transform=(Rotation=(X=-0.0114,Y=-0.13,Z=-0.0864,W=0.9877),Translation=(X=0,Y=0,Z=0),Scale3D=(X=1,Y=1,Z=1))))
    WeightMinimum=0
    ExecutePin=@n2.B
    Weight=RigVMVariableNode() {variable=IsCrouchWalking}.Value
   )
   RigVMFunctionReturnNode(ExecuteContext=@RigVMFunction_Sequence.B)
   RigVMDispatchNode(
    False=(X=-0.0687,Y=0.1402,Z=-0.1176,W=0.9807)
    True=(X=-0.0747,Y=0.0974,Z=-0.122,W=0.9849)
    Condition=RigVMVariableNode() {variable=bisPickaxeOut}.Value
   ) {template="DISPATCH_RigVMDispatch_If(in Condition,in True,in False,out Result)"}
  ]}
 New Function_ContainedGraph {nodes [
   n1 = RigVMFunctionEntryNode()
   RigVMFunction_MathBoolToInteger = &RigVMFunction_MathBoolToInteger(Value=@n1.onlyTwist)
   RigVMFunction_MathQuaternionSwingTwist = &RigVMFunction_MathQuaternionSwingTwist(
    TwistAxis=@n1.twistAxis
    Input=&RigUnit_GetRelativeTransformForItem(
     Child=(Type=Bone,Name="None")
     Parent=(Type=Bone,Name="None")
     bChildInitial=False
     bParentInitial=False
     Child.Name=@n1.child
     Parent.Name=@n1.parent
     Parent.Type=@n1.parentType
    ).RelativeTransform.Rotation
   )
   RigVMFunction_MathQuaternionToEuler = &RigVMFunction_MathQuaternionToEuler(
    Value=RigVMDispatchNode(
     Values=((),())
     Index=@RigVMFunction_MathBoolToInteger.Result
     Values.0=@RigVMFunction_MathQuaternionSwingTwist.Swing
     Values.1=@RigVMFunction_MathQuaternionSwingTwist.Twist
    ) {template="DISPATCH_RigVMDispatch_SelectInt32(in Index,in Values,out Result)"}.Result
    RotationOrder=@n1.Rotation Order
   )
   > DISPATCH_RigVMDispatch_Print = RigVMDispatchNode(ScreenColor=(R=1,G=1,B=1,A=1), ScreenDuration=1.5, Enabled=@n1.debugRotation, Value=@RigVMFunction_MathQuaternionToEuler.Result) {template="DISPATCH_RigVMDispatch_Print(in Prefix,in Value,in Enabled,in ScreenDuration,in ScreenColor)"}
   Clamp = &RigVMFunction_MathIntClamp(Value=@n1.primAxis)
   DISPATCH_RigVMDispatch_SelectInt32 = RigVMDispatchNode(
    Values=(0,0,0)
    Index=@Clamp.Result
    Values.0=@RigVMFunction_MathQuaternionToEuler.Result.X
    Values.1=@RigVMFunction_MathQuaternionToEuler.Result.Y
    Values.2=@RigVMFunction_MathQuaternionToEuler.Result.Z
   ) {template="DISPATCH_RigVMDispatch_SelectInt32(in Index,in Values,out Result)"}
   Remap = &RigVMFunction_MathDoubleRemap(
    TargetMaximum=1
    TargetMinimum=0
    bClamp=False
    Value=@DISPATCH_RigVMDispatch_SelectInt32.Result
    SourceMaximum=@n1.max
    SourceMinimum=@n1.min
   )
   Multiply = &RigVMFunction_MathDoubleMul(A=&RigVMFunction_MathDoubleClamp(Value=@Remap.Result).Result, B=@n1.premult)
   > RigVMFunctionReturnNode(outputWeight=@Multiply.Result)
  ]}
 New Function_ContainedGraph {nodes [
   RigVMFunctionEntryNode()
   > RigVMFunction_Sequence = RigVMAggregateNode()
   n1 = &RigVMFunction_ControlFlowBranch(ExecuteContext=@RigVMFunction_Sequence.B, Condition=RigVMVariableNode() {variable=isSkyDiving}.Value)
   n2 = &RigVMFunction_Sequence(ExecuteContext=@n1.False)
   DeltaRotation_1 = RigVMCollapseNode(
    Rotation Order=ZYX
    child=head
    debugRotation=false
    max=-24
    min=0
    onlyTwist=false
    parent=head_Neck01Delta_null
    parentType=Null
    premult=1
    primAxis=2
    twistAxis=(X=1,Y=0,Z=0)
    ExecuteContext=@n2.A
   )
   RigVMFunction_AnimEvalRichCurve = &RigVMFunction_AnimEvalRichCurve(
    Curve=(EditorCurveData=(Keys=((InterpMode=RCIM_Cubic),(InterpMode=RCIM_Cubic,Time=0.6232,Value=0.3032,ArriveTangent=1,LeaveTangent=1),(InterpMode=RCIM_Cubic,TangentMode=RCTM_User,Time=1,Value=1,ArriveTangent=2.8434,LeaveTangent=2.8434))))
    SourceMinimum=0
    TargetMinimum=0
    Value=@DeltaRotation_1.outputWeight
   )
   > RigVMVariableNode(Value=@RigVMFunction_AnimEvalRichCurve.Result) {variable=LocalHeadLookingUp}
   > DeltaRotation_2 = RigVMCollapseNode(
    Rotation Order=ZYX
    child=head
    debugRotation=false
    max=-38
    min=-24
    onlyTwist=false
    parent=head_Neck01Delta_null
    parentType=Null
    premult=1
    primAxis=2
    twistAxis=(X=1,Y=0,Z=0)
   )
   n3 = &RigVMFunction_AnimEvalRichCurve(
    Curve=(EditorCurveData=(Keys=((InterpMode=RCIM_Cubic),(InterpMode=RCIM_Cubic,Time=0.6232,Value=0.3032,ArriveTangent=1,LeaveTangent=1),(InterpMode=RCIM_Cubic,TangentMode=RCTM_User,Time=1,Value=1,ArriveTangent=2.8434,LeaveTangent=2.8434))))
    SourceMinimum=0
    TargetMinimum=0
    Value=@DeltaRotation_2.outputWeight
   )
   > RigVMVariableNode(Value=@n3.Result) {variable=LocalHeadLookingUpExtreme}
   > n4 = &RigUnit_SphericalPoseReader(
    ActiveRegionScaleFactors=(PositiveWidth=1,NegativeWidth=1,PositiveHeight=0.75,NegativeHeight=0.4)
    ActiveRegionSize=0.1
    Debug=(bDrawDebug=false,bDraw2D=False,bDrawLocalAxes=False,DebugScale=25,DebugSegments=20,DebugThickness=0.25)
    DriverAxis=(X=-1,Y=0,Z=0)
    DriverItem=(Type=Bone,Name="clavicle_r")
    FalloffRegionScaleFactors=(PositiveWidth=1,NegativeWidth=0.6,PositiveHeight=1,NegativeHeight=0.6)
    FalloffSize=0.175
    FlipHeightScaling=False
    FlipWidthScaling=False
    OptionalParentItem=(Type=Bone,Name="None")
    RotationOffset=(X=115,Y=-30,Z=65)
   )
   > RigVMVariableNode(Value=@n4.OutputParam) {variable=R_ClavicleBack}
   RigVMFunction_MathBoolNot = &RigVMFunction_MathBoolNot(Value=RigVMVariableNode() {variable=IsPistolOut}.Value)
   BoolToFloat = RigVMFunctionReferenceNode(Bool=@RigVMFunction_MathBoolNot.Result)
   RigVMFunction_MathBoolToInteger = &RigVMFunction_MathBoolToInteger(Value=RigVMVariableNode() {variable=bIsCrouching}.Value)
   Subtract = &RigVMFunction_MathDoubleSub(
    A=1
    B=RigVMDispatchNode(Values=(0,0), Index=@RigVMFunction_MathBoolToInteger.Result, Values.0=RigVMVariableNode() {variable=LocalHeadLookingUp}.Value) {template="DISPATCH_RigVMDispatch_SelectInt32(in Index,in Values,out Result)"}.Result
   )
   Multiply_1 = RigVMAggregateNode(B=@BoolToFloat.Float, C=@Subtract.Result, A=RigVMVariableNode() {variable=R_ClavicleBack}.Value)
   &RigUnit_ModifyTransforms(
    ItemToModify=((Item=(Type=Bone,Name="dyn_R_Ear_04_c_jnt"),Transform=(Rotation=(X=0,Y=0,Z=0,W=1),Translation=(X=0,Y=-1,Z=0),Scale3D=(X=1,Y=1,Z=1))))
    WeightMinimum=0
    Weight=@Multiply_1.Result
    ExecutePin=@n2.B
   )
   > &RigUnit_ModifyTransforms(
    ItemToModify=((Item=(Type=Bone,Name="dyn_R_Ear_05_c_jnt"),Transform=(Rotation=(X=0,Y=0,Z=0,W=1),Translation=(X=0,Y=-2.5,Z=0),Scale3D=(X=1,Y=1,Z=1))))
    WeightMinimum=0
    Weight=@Multiply_1.Result
   )
   > &RigUnit_ModifyTransforms(
    ItemToModify=((Item=(Type=Bone,Name="dyn_R_Ear_06_c_jnt"),Transform=(Rotation=(X=0,Y=0,Z=0,W=1),Translation=(X=0,Y=-2,Z=0),Scale3D=(X=1,Y=1,Z=1))))
    WeightMinimum=0
    Weight=@Multiply_1.Result
   )
   > ArmsBackEarsInteraction = RigVMFunctionReferenceNode()
   RigVMFunction_ControlFlowBranch = &RigVMFunction_ControlFlowBranch(ExecuteContext=@ArmsBackEarsInteraction.ExecuteContext, Condition=RigVMVariableNode() {variable=bIsCrouching}.Value)
   n5 = &RigVMFunction_ControlFlowBranch(ExecuteContext=@RigVMFunction_ControlFlowBranch.False, Condition=RigVMVariableNode() {variable=IsAimingDownSights}.Value)
   &RigUnit_ModifyTransforms(
    ItemToModify=((Item=(Type=Bone,Name="dyn_R_Ear_03_c_jnt"),Transform=(Rotation=(X=0,Y=0,Z=0,W=1),Translation=(X=-2,Y=-3,Z=0),Scale3D=(X=1,Y=1,Z=1))))
    WeightMinimum=0
    ExecutePin=@n5.False
    Weight=RigVMVariableNode() {variable=LocalHeadLookingUp}.Value
   )
   > &RigUnit_ModifyTransforms(
    ItemToModify=((Item=(Type=Bone,Name="dyn_R_Ear_04_c_jnt"),Transform=(Rotation=(X=0,Y=0,Z=0,W=1),Translation=(X=0,Y=-5,Z=0),Scale3D=(X=1,Y=1,Z=1))))
    WeightMinimum=0
    Weight=RigVMVariableNode() {variable=LocalHeadLookingUp}.Value
   )
   > &RigUnit_ModifyTransforms(
    ItemToModify=((Item=(Type=Bone,Name="dyn_R_Ear_06_c_jnt"),Transform=(Rotation=(X=0,Y=0,Z=0,W=1),Translation=(X=2,Y=0,Z=0),Scale3D=(X=1,Y=1,Z=1))))
    WeightMinimum=0
    Weight=RigVMVariableNode() {variable=LocalHeadLookingUp}.Value
   )
   > &RigUnit_ModifyTransforms(
    ItemToModify=((Item=(Type=Bone,Name="dyn_L_Ear_04_c_jnt"),Transform=(Rotation=(X=0,Y=0,Z=0,W=1),Translation=(X=-2,Y=-3,Z=0),Scale3D=(X=1,Y=1,Z=1))))
    WeightMinimum=0
    Weight=RigVMVariableNode() {variable=LocalHeadLookingUp}.Value
   )
   n6 = &RigVMFunction_ControlFlowBranch(ExecuteContext=@n5.True, Condition=RigVMVariableNode() {variable=IsPistolOut}.Value)
   &RigUnit_ModifyTransforms(
    ItemToModify=((Item=(Type=Bone,Name="dyn_R_Ear_03_c_jnt"),Transform=(Rotation=(X=0,Y=0,Z=0,W=1),Translation=(X=-2,Y=-5,Z=0),Scale3D=(X=1,Y=1,Z=1))))
    WeightMinimum=0
    ExecutePin=@n6.False
    Weight=RigVMVariableNode() {variable=LocalHeadLookingUp}.Value
   )
   > &RigUnit_ModifyTransforms(
    ItemToModify=((Item=(Type=Bone,Name="dyn_R_Ear_04_c_jnt"),Transform=(Rotation=(X=0,Y=0,Z=0,W=1),Translation=(X=0,Y=-6,Z=-4),Scale3D=(X=1,Y=1,Z=1))))
    WeightMinimum=0
    Weight=RigVMVariableNode() {variable=LocalHeadLookingUp}.Value
   )
   > &RigUnit_ModifyTransforms(
    ItemToModify=((Item=(Type=Bone,Name="dyn_R_Ear_05_c_jnt"),Transform=(Rotation=(X=0,Y=0,Z=0,W=1),Translation=(X=9,Y=1,Z=0),Scale3D=(X=1,Y=1,Z=1))))
    WeightMinimum=0
    Weight=RigVMVariableNode() {variable=LocalHeadLookingUp}.Value
   )
   > &RigUnit_ModifyTransforms(
    ItemToModify=((Item=(Type=Bone,Name="dyn_R_Ear_06_c_jnt"),Transform=(Rotation=(X=0,Y=0,Z=-0.0436,W=0.999),Translation=(X=12,Y=2,Z=0),Scale3D=(X=1,Y=1,Z=1))))
    WeightMinimum=0
    Weight=RigVMVariableNode() {variable=LocalHeadLookingUp}.Value
   )
   &RigUnit_ModifyTransforms(
    ItemToModify=((Item=(Type=Bone,Name="dyn_R_Ear_03_c_jnt"),Transform=(Rotation=(X=0,Y=0,Z=0,W=1),Translation=(X=-2,Y=-3,Z=0),Scale3D=(X=1,Y=1,Z=1))))
    WeightMinimum=0
    ExecutePin=@n6.True
    Weight=RigVMVariableNode() {variable=LocalHeadLookingUp}.Value
   )
   > &RigUnit_ModifyTransforms(
    ItemToModify=((Item=(Type=Bone,Name="dyn_R_Ear_04_c_jnt"),Transform=(Rotation=(X=0,Y=0,Z=0,W=1),Translation=(X=0,Y=-3,Z=0),Scale3D=(X=1,Y=1,Z=1))))
    WeightMinimum=0
    Weight=RigVMVariableNode() {variable=LocalHeadLookingUp}.Value
   )
   > &RigUnit_ModifyTransforms(
    ItemToModify=((Item=(Type=Bone,Name="dyn_R_Ear_05_c_jnt"),Transform=(Rotation=(X=0,Y=0,Z=0,W=1),Translation=(X=7,Y=2,Z=0),Scale3D=(X=1,Y=1,Z=1))))
    WeightMinimum=0
    Weight=RigVMVariableNode() {variable=LocalHeadLookingUp}.Value
   )
   > &RigUnit_ModifyTransforms(
    ItemToModify=((Item=(Type=Bone,Name="dyn_R_Ear_06_c_jnt"),Transform=(Rotation=(X=0,Y=0,Z=0.0436,W=0.999),Translation=(X=10,Y=4,Z=0),Scale3D=(X=1,Y=1,Z=1))))
    WeightMinimum=0
    Weight=RigVMVariableNode() {variable=LocalHeadLookingUp}.Value
   )
   &RigUnit_ModifyTransforms(
    ItemToModify=((Item=(Type=Bone,Name="dyn_R_Ear_06_c_jnt"),Transform=(Rotation=(X=0,Y=0,Z=0.0872,W=0.9962),Translation=(X=5,Y=2,Z=0),Scale3D=(X=1,Y=1,Z=1))))
    WeightMinimum=0
    ExecutePin=@RigVMFunction_ControlFlowBranch.True
    Weight=RigVMVariableNode() {variable=LocalHeadLookingUp}.Value
   )
   > n7 = &RigUnit_ModifyTransforms(
    ItemToModify=((Item=(Type=Bone,Name="dyn_L_Ear_04_c_jnt"),Transform=(Rotation=(X=0,Y=0,Z=0,W=1),Translation=(X=4,Y=0,Z=0),Scale3D=(X=1,Y=1,Z=1))))
    WeightMinimum=0
    Weight=RigVMVariableNode() {variable=LocalHeadLookingUp}.Value
   )
   n8 = &RigVMFunction_ControlFlowBranch(ExecuteContext=@n7.ExecutePin, Condition=RigVMVariableNode() {variable=IsAimingDownSights}.Value)
   &RigUnit_ModifyTransforms(
    ItemToModify=((Item=(Type=Bone,Name="dyn_R_Ear_03_c_jnt"),Transform=(Rotation=(X=0,Y=0,Z=0,W=1),Translation=(X=-1,Y=-3,Z=0),Scale3D=(X=1,Y=1,Z=1))))
    WeightMinimum=0
    ExecutePin=@n8.False
    Weight=RigVMVariableNode() {variable=LocalHeadLookingUpExtreme}.Value
   )
   > &RigUnit_ModifyTransforms(
    ItemToModify=((Item=(Type=Bone,Name="dyn_R_Ear_04_c_jnt"),Transform=(Rotation=(X=0,Y=0,Z=0.0872,W=0.9962),Translation=(X=4,Y=-5,Z=0),Scale3D=(X=1,Y=1,Z=1))))
    WeightMinimum=0
    Weight=RigVMVariableNode() {variable=LocalHeadLookingUpExtreme}.Value
   )
   > &RigUnit_ModifyTransforms(
    ItemToModify=((Item=(Type=Bone,Name="dyn_R_Ear_05_c_jnt"),Transform=(Rotation=(X=0,Y=0,Z=0,W=1),Translation=(X=4,Y=-2,Z=0),Scale3D=(X=1,Y=1,Z=1))))
    WeightMinimum=0
    Weight=RigVMVariableNode() {variable=LocalHeadLookingUpExtreme}.Value
   )
   > &RigUnit_ModifyTransforms(
    ItemToModify=((Item=(Type=Bone,Name="dyn_L_Ear_04_c_jnt"),Transform=(Rotation=(X=0,Y=0,Z=0,W=1),Translation=(X=-2,Y=-3,Z=0),Scale3D=(X=1,Y=1,Z=1))))
    WeightMinimum=0
    Weight=RigVMVariableNode() {variable=LocalHeadLookingUpExtreme}.Value
   )
   Multiply = &RigVMFunction_MathDoubleMul(B=RigVMVariableNode() {variable=LocalHeadLookingUpExtreme}.Value, A=RigVMVariableNode() {variable=IsCrouchWalking}.Value)
   > &RigUnit_ModifyTransforms(
    ItemToModify=((Item=(Type=Bone,Name="dyn_R_Ear_06_c_jnt"),Transform=(Rotation=(X=0,Y=0,Z=0,W=1),Translation=(X=5,Y=-3,Z=0),Scale3D=(X=1,Y=1,Z=1))))
    WeightMinimum=0
    Weight=@Multiply.Result
   )
   > &RigUnit_ModifyTransforms(
    ItemToModify=((Item=(Type=Bone,Name="dyn_R_Ear_05_c_jnt"),Transform=(Rotation=(X=0,Y=0,Z=0,W=1),Translation=(X=3,Y=-2,Z=0),Scale3D=(X=1,Y=1,Z=1))))
    WeightMinimum=0
    Weight=@Multiply.Result
   )
   > n9 = &RigUnit_ModifyTransforms(
    ItemToModify=((Item=(Type=Bone,Name="dyn_R_Ear_03_c_jnt"),Transform=(Rotation=(X=0,Y=0,Z=0,W=1),Translation=(X=0,Y=-2,Z=0),Scale3D=(X=1,Y=1,Z=1))))
    WeightMinimum=0
    Weight=@Multiply.Result
   )
   n10 = &RigVMFunction_ControlFlowBranch(ExecuteContext=@n9.ExecutePin, Condition=RigVMVariableNode() {variable=IsPistolOut}.Value)
   &RigUnit_ModifyTransforms(
    ItemToModify=((Item=(Type=Bone,Name="dyn_L_Ear_06_c_jnt"),Transform=(Rotation=(X=-0.0038,Y=0.0435,Z=0.0871,W=0.9952),Translation=(X=8,Y=-2,Z=-2),Scale3D=(X=1,Y=1,Z=1))))
    WeightMinimum=0
    ExecutePin=@n10.False
    Weight=@Multiply.Result
   )
   > &RigUnit_ModifyTransforms(
    ItemToModify=((Item=(Type=Bone,Name="dyn_L_Ear_05_c_jnt"),Transform=(Rotation=(X=0,Y=0,Z=0,W=1),Translation=(X=5,Y=-2,Z=-1),Scale3D=(X=1,Y=1,Z=1))))
    WeightMinimum=0
    Weight=@Multiply.Result
   )
   > &RigUnit_ModifyTransforms(
    ItemToModify=((Item=(Type=Bone,Name="dyn_L_Ear_04_c_jnt"),Transform=(Rotation=(X=0,Y=0,Z=0,W=1),Translation=(X=5,Y=-1,Z=-0.5),Scale3D=(X=1,Y=1,Z=1))))
    WeightMinimum=0
    Weight=@Multiply.Result
   )
   &RigUnit_ModifyTransforms(
    ItemToModify=((Item=(Type=Bone,Name="dyn_L_Ear_06_c_jnt"),Transform=(Rotation=(X=0,Y=0,Z=0.1736,W=0.9848),Translation=(X=10,Y=1.5,Z=0),Scale3D=(X=1,Y=1,Z=1))))
    WeightMinimum=0
    ExecutePin=@n10.True
    Weight=@Multiply.Result
   )
   > &RigUnit_ModifyTransforms(
    ItemToModify=((Item=(Type=Bone,Name="dyn_L_Ear_05_c_jnt"),Transform=(Rotation=(X=0,Y=0,Z=0,W=1),Translation=(X=5,Y=-1,Z=-2),Scale3D=(X=1,Y=1,Z=1))))
    WeightMinimum=0
    Weight=@Multiply.Result
   )
   > &RigUnit_ModifyTransforms(
    ItemToModify=((Item=(Type=Bone,Name="dyn_L_Ear_04_c_jnt"),Transform=(Rotation=(X=0,Y=0,Z=0,W=1),Translation=(X=5,Y=-1,Z=0),Scale3D=(X=1,Y=1,Z=1))))
    WeightMinimum=0
    Weight=@Multiply.Result
   )
   n11 = &RigVMFunction_ControlFlowBranch(ExecuteContext=@n8.True, Condition=RigVMVariableNode() {variable=IsPistolOut}.Value)
   &RigUnit_ModifyTransforms(
    ItemToModify=((Item=(Type=Bone,Name="dyn_R_Ear_03_c_jnt"),Transform=(Rotation=(X=-0.3391,Y=0.0446,Z=-0.1227,W=0.9317),Translation=(X=-3.5,Y=-7,Z=0),Scale3D=(X=1,Y=1,Z=1))))
    WeightMinimum=0
    ExecutePin=@n11.False
    Weight=RigVMVariableNode() {variable=LocalHeadLookingUp}.Value
   )
   > &RigUnit_ModifyTransforms(
    ItemToModify=((Item=(Type=Bone,Name="dyn_R_Ear_04_c_jnt"),Transform=(Rotation=(X=-0.1305,Y=0,Z=0,W=0.9914),Translation=(X=3,Y=-7,Z=-3),Scale3D=(X=1,Y=1,Z=1))))
    WeightMinimum=0
    Weight=RigVMVariableNode() {variable=LocalHeadLookingUp}.Value
   )
   > &RigUnit_ModifyTransforms(
    ItemToModify=((Item=(Type=Bone,Name="dyn_R_Ear_05_c_jnt"),Transform=(Rotation=(X=0,Y=0,Z=0.1736,W=0.9848),Translation=(X=12,Y=3,Z=0),Scale3D=(X=1,Y=1,Z=1))))
    WeightMinimum=0
    Weight=RigVMVariableNode() {variable=LocalHeadLookingUp}.Value
   )
   > &RigUnit_ModifyTransforms(
    ItemToModify=((Item=(Type=Bone,Name="dyn_R_Ear_06_c_jnt"),Transform=(Rotation=(X=0,Y=0,Z=0,W=1),Translation=(X=12,Y=1.5,Z=0),Scale3D=(X=1,Y=1,Z=1))))
    WeightMinimum=0
    Weight=RigVMVariableNode() {variable=LocalHeadLookingUp}.Value
   )
   > DeltaRotation_3 = RigVMCollapseNode(
    Rotation Order=ZYX
    child=head
    debugRotation=false
    max=-32
    min=-16
    onlyTwist=false
    parent=head_Neck01Delta_null
    parentType=Null
    premult=1
    primAxis=2
    twistAxis=(X=1,Y=0,Z=0)
   )
   n12 = &RigVMFunction_AnimEvalRichCurve(Curve=(EditorCurveData=(Keys=((),(Time=0.4,Value=1),(Time=1)))), SourceMinimum=0, TargetMinimum=0, Value=@DeltaRotation_3.outputWeight)
   > &RigUnit_ModifyTransforms(
    ItemToModify=((Item=(Type=Bone,Name="dyn_R_Ear_03_c_jnt"),Transform=(Rotation=(X=0,Y=0,Z=0,W=1),Translation=(X=0,Y=4.5,Z=0),Scale3D=(X=1,Y=1,Z=1))))
    WeightMinimum=0
    Weight=&RigVMFunction_MathDoubleMul(A=@n12.Result, B=RigVMVariableNode() {variable=IsCrouchWalking}.Value).Result
   )
   n13 = &RigVMFunction_MathDoubleMul(B=RigVMVariableNode() {variable=LocalHeadLookingUpExtreme}.Value, A=RigVMVariableNode() {variable=IsCrouchWalking}.Value)
   > &RigUnit_ModifyTransforms(
    ItemToModify=((Item=(Type=Bone,Name="dyn_R_Ear_03_c_jnt"),Transform=(Rotation=(X=0,Y=0,Z=0,W=1),Translation=(X=0,Y=-1,Z=0),Scale3D=(X=1,Y=1,Z=1))))
    WeightMinimum=0
    Weight=@n13.Result
   )
   > &RigUnit_ModifyTransforms(
    ItemToModify=((Item=(Type=Bone,Name="dyn_R_Ear_05_c_jnt"),Transform=(Rotation=(X=0,Y=0,Z=0,W=1),Translation=(X=3,Y=-2,Z=0),Scale3D=(X=1,Y=1,Z=1))))
    WeightMinimum=0
    Weight=@n13.Result
   )
   > &RigUnit_ModifyTransforms(
    ItemToModify=((Item=(Type=Bone,Name="dyn_R_Ear_06_c_jnt"),Transform=(Rotation=(X=0,Y=0,Z=0,W=1),Translation=(X=5,Y=0,Z=0),Scale3D=(X=1,Y=1,Z=1))))
    WeightMinimum=0
    Weight=@n13.Result
   )
   > &RigUnit_ModifyTransforms(
    ItemToModify=((Item=(Type=Bone,Name="dyn_L_Ear_03_c_jnt"),Transform=(Rotation=(X=0,Y=0,Z=0,W=1),Translation=(X=0,Y=-1,Z=0),Scale3D=(X=1,Y=1,Z=1))))
    WeightMinimum=0
    Weight=@n13.Result
   )
   > &RigUnit_ModifyTransforms(
    ItemToModify=((Item=(Type=Bone,Name="dyn_L_Ear_04_c_jnt"),Transform=(Rotation=(X=0,Y=0,Z=0,W=1),Translation=(X=6,Y=-5,Z=0),Scale3D=(X=1,Y=1,Z=1))))
    WeightMinimum=0
    Weight=@n13.Result
   )
   > &RigUnit_ModifyTransforms(
    ItemToModify=((Item=(Type=Bone,Name="dyn_L_Ear_05_c_jnt"),Transform=(Rotation=(X=0,Y=0,Z=0,W=1),Translation=(X=14,Y=0,Z=-3),Scale3D=(X=1,Y=1,Z=1))))
    WeightMinimum=0
    Weight=@n13.Result
   )
   > &RigUnit_ModifyTransforms(
    ItemToModify=((Item=(Type=Bone,Name="dyn_L_Ear_06_c_jnt"),Transform=(Rotation=(X=0,Y=0,Z=0.1736,W=0.9848),Translation=(X=16,Y=3,Z=-4),Scale3D=(X=1,Y=1,Z=1))))
    WeightMinimum=0
    Weight=@n13.Result
   )
   &RigUnit_ModifyTransforms(
    ItemToModify=((Item=(Type=Bone,Name="dyn_R_Ear_03_c_jnt"),Transform=(Rotation=(X=0,Y=0,Z=-0.1736,W=0.9848),Translation=(X=-3,Y=-6,Z=0),Scale3D=(X=1,Y=1,Z=1))))
    WeightMinimum=0
    ExecutePin=@n11.True
    Weight=RigVMVariableNode() {variable=LocalHeadLookingUpExtreme}.Value
   )
   > &RigUnit_ModifyTransforms(
    ItemToModify=((Item=(Type=Bone,Name="dyn_R_Ear_04_c_jnt"),Transform=(Rotation=(X=0,Y=0,Z=0,W=1),Translation=(X=2,Y=-9,Z=0),Scale3D=(X=1,Y=1,Z=1))))
    WeightMinimum=0
    Weight=RigVMVariableNode() {variable=LocalHeadLookingUpExtreme}.Value
   )
   > &RigUnit_ModifyTransforms(
    ItemToModify=((Item=(Type=Bone,Name="dyn_R_Ear_05_c_jnt"),Transform=(Rotation=(X=0,Y=0,Z=0,W=1),Translation=(X=8,Y=0,Z=0),Scale3D=(X=1,Y=1,Z=1))))
    WeightMinimum=0
    Weight=RigVMVariableNode() {variable=LocalHeadLookingUpExtreme}.Value
   )
   > &RigUnit_ModifyTransforms(
    ItemToModify=((Item=(Type=Bone,Name="dyn_R_Ear_06_c_jnt"),Transform=(Rotation=(X=0,Y=0,Z=0.0436,W=0.999),Translation=(X=10,Y=2,Z=0),Scale3D=(X=1,Y=1,Z=1))))
    WeightMinimum=0
    Weight=RigVMVariableNode() {variable=LocalHeadLookingUpExtreme}.Value
   )
   n14 = &RigVMFunction_MathDoubleMul(B=RigVMVariableNode() {variable=LocalHeadLookingUpExtreme}.Value, A=RigVMVariableNode() {variable=IsCrouchWalking}.Value)
   > &RigUnit_ModifyTransforms(
    ItemToModify=((Item=(Type=Bone,Name="dyn_L_Ear_03_c_jnt"),Transform=(Rotation=(X=0,Y=0,Z=0,W=1),Translation=(X=0,Y=-2,Z=0),Scale3D=(X=1,Y=1,Z=1))))
    WeightMinimum=0
    Weight=@n14.Result
   )
   > &RigUnit_ModifyTransforms(
    ItemToModify=((Item=(Type=Bone,Name="dyn_L_Ear_04_c_jnt"),Transform=(Rotation=(X=0,Y=0,Z=0,W=1),Translation=(X=6,Y=-6,Z=-5),Scale3D=(X=1,Y=1,Z=1))))
    WeightMinimum=0
    Weight=@n14.Result
   )
   > &RigUnit_ModifyTransforms(
    ItemToModify=((Item=(Type=Bone,Name="dyn_L_Ear_05_c_jnt"),Transform=(Rotation=(X=0,Y=0,Z=0,W=1),Translation=(X=14,Y=0,Z=-3),Scale3D=(X=1,Y=1,Z=1))))
    WeightMinimum=0
    Weight=@n14.Result
   )
   > &RigUnit_ModifyTransforms(
    ItemToModify=((Item=(Type=Bone,Name="dyn_L_Ear_06_c_jnt"),Transform=(Rotation=(X=0,Y=0,Z=0.1736,W=0.9848),Translation=(X=16,Y=3,Z=-4),Scale3D=(X=1,Y=1,Z=1))))
    WeightMinimum=0
    Weight=@n14.Result
   )
   &RigUnit_ModifyTransforms(
    ItemToModify=((Item=(Type=Bone,Name="dyn_R_Ear_03_c_jnt"),Transform=(Rotation=(X=0,Y=0,Z=-0.1736,W=0.9848),Translation=(X=0,Y=-4,Z=1),Scale3D=(X=1,Y=1,Z=1))))
    Weight=1
    WeightMinimum=0
    ExecutePin=@n1.True
   )
   > &RigUnit_ModifyTransforms(
    ItemToModify=((Item=(Type=Bone,Name="dyn_R_Ear_04_c_jnt"),Transform=(Rotation=(X=-0.342,Y=0,Z=0,W=0.9397),Translation=(X=3,Y=-10,Z=2),Scale3D=(X=1,Y=1,Z=1))))
    Weight=1
    WeightMinimum=0
   )
   > &RigUnit_ModifyTransforms(
    ItemToModify=((Item=(Type=Bone,Name="dyn_R_Ear_05_c_jnt"),Transform=(Rotation=(X=0,Y=0,Z=0,W=1),Translation=(X=10,Y=-4,Z=3),Scale3D=(X=1,Y=1,Z=1))))
    Weight=1
    WeightMinimum=0
   )
   > &RigUnit_ModifyTransforms(
    ItemToModify=((Item=(Type=Bone,Name="dyn_R_Ear_06_c_jnt"),Transform=(Rotation=(X=0,Y=0,Z=0,W=1),Translation=(X=15,Y=0,Z=4),Scale3D=(X=1,Y=1,Z=1))))
    Weight=1
    WeightMinimum=0
   )
   > &RigUnit_ModifyTransforms(
    ItemToModify=((Item=(Type=Bone,Name="dyn_L_Ear_03_c_jnt"),Transform=(Rotation=(X=0,Y=0,Z=-0.1736,W=0.9848),Translation=(X=0,Y=-2,Z=-1),Scale3D=(X=1,Y=1,Z=1))))
    Weight=1
    WeightMinimum=0
   )
   > &RigUnit_ModifyTransforms(
    ItemToModify=((Item=(Type=Bone,Name="dyn_L_Ear_04_c_jnt"),Transform=(Rotation=(X=0.342,Y=0,Z=0,W=0.9397),Translation=(X=3,Y=-6,Z=-2),Scale3D=(X=1,Y=1,Z=1))))
    Weight=1
    WeightMinimum=0
   )
   > &RigUnit_ModifyTransforms(
    ItemToModify=((Item=(Type=Bone,Name="dyn_L_Ear_05_c_jnt"),Transform=(Rotation=(X=0,Y=0,Z=0,W=1),Translation=(X=10,Y=-2,Z=-3),Scale3D=(X=1,Y=1,Z=1))))
    Weight=1
    WeightMinimum=0
   )
   > &RigUnit_ModifyTransforms(
    ItemToModify=((Item=(Type=Bone,Name="dyn_L_Ear_06_c_jnt"),Transform=(Rotation=(X=0,Y=0,Z=0,W=1),Translation=(X=15,Y=0,Z=-4),Scale3D=(X=1,Y=1,Z=1))))
    Weight=1
    WeightMinimum=0
   )
   > &RigUnit_ModifyTransforms(
    ItemToModify=((Item=(Type=Bone,Name="dyn_R_Ear_03_c_jnt"),Transform=(Rotation=(X=0,Y=0,Z=0,W=1),Translation=(X=0,Y=0,Z=2),Scale3D=(X=1,Y=1,Z=1))))
    WeightMinimum=0
    Weight=RigVMVariableNode() {variable=SkydiveDiveMode}.Value
   )
   > &RigUnit_ModifyTransforms(
    ItemToModify=((Item=(Type=Bone,Name="dyn_R_Ear_04_c_jnt"),Transform=(Rotation=(X=0,Y=0,Z=0,W=1),Translation=(X=0,Y=0,Z=5),Scale3D=(X=1,Y=1,Z=1))))
    WeightMinimum=0
    Weight=RigVMVariableNode() {variable=SkydiveDiveMode}.Value
   )
   > &RigUnit_ModifyTransforms(
    ItemToModify=((Item=(Type=Bone,Name="dyn_R_Ear_05_c_jnt"),Transform=(Rotation=(X=0,Y=0,Z=0,W=1),Translation=(X=0,Y=-1,Z=7),Scale3D=(X=1,Y=1,Z=1))))
    WeightMinimum=0
    Weight=RigVMVariableNode() {variable=SkydiveDiveMode}.Value
   )
   > &RigUnit_ModifyTransforms(
    ItemToModify=((Item=(Type=Bone,Name="dyn_R_Ear_06_c_jnt"),Transform=(Rotation=(X=0,Y=0,Z=0,W=1),Translation=(X=0,Y=-2,Z=5),Scale3D=(X=1,Y=1,Z=1))))
    WeightMinimum=0
    Weight=RigVMVariableNode() {variable=SkydiveDiveMode}.Value
   )
   > &RigUnit_ModifyTransforms(
    ItemToModify=((Item=(Type=Bone,Name="dyn_L_Ear_03_c_jnt"),Transform=(Rotation=(X=0,Y=0,Z=0,W=1),Translation=(X=0,Y=0,Z=-2),Scale3D=(X=1,Y=1,Z=1))))
    WeightMinimum=0
    Weight=RigVMVariableNode() {variable=SkydiveDiveMode}.Value
   )
   > &RigUnit_ModifyTransforms(
    ItemToModify=((Item=(Type=Bone,Name="dyn_L_Ear_04_c_jnt"),Transform=(Rotation=(X=0,Y=0,Z=0,W=1),Translation=(X=0,Y=0,Z=-5),Scale3D=(X=1,Y=1,Z=1))))
    WeightMinimum=0
    Weight=RigVMVariableNode() {variable=SkydiveDiveMode}.Value
   )
   > &RigUnit_ModifyTransforms(
    ItemToModify=((Item=(Type=Bone,Name="dyn_L_Ear_05_c_jnt"),Transform=(Rotation=(X=0,Y=0,Z=0,W=1),Translation=(X=0,Y=-1,Z=-7),Scale3D=(X=1,Y=1,Z=1))))
    WeightMinimum=0
    Weight=RigVMVariableNode() {variable=SkydiveDiveMode}.Value
   )
   > &RigUnit_ModifyTransforms(
    ItemToModify=((Item=(Type=Bone,Name="dyn_L_Ear_06_c_jnt"),Transform=(Rotation=(X=0,Y=0,Z=0,W=1),Translation=(X=0,Y=-2,Z=-5),Scale3D=(X=1,Y=1,Z=1))))
    WeightMinimum=0
    Weight=RigVMVariableNode() {variable=SkydiveDiveMode}.Value
   )
   RigUnit_ItemArray = &RigUnit_ItemArray(Items=((Type=Bone,Name="dyn_R_Ear_02"),(Type=Bone,Name="dyn_R_Ear_03"),(Type=Bone,Name="dyn_R_Ear_04"),(Type=Bone,Name="dyn_R_Ear_05"),(Type=Bone,Name="dyn_R_Ear_06")))
   n15 = &RigUnit_ItemArray(Items=((Type=Bone,Name="dyn_R_Ear_02_c_jnt"),(Type=Bone,Name="dyn_R_Ear_03_c_jnt"),(Type=Bone,Name="dyn_R_Ear_04_c_jnt"),(Type=Bone,Name="dyn_R_Ear_05_c_jnt"),(Type=Bone,Name="dyn_R_Ear_06_c_jnt")))
   SplineIK = RigVMFunctionReferenceNode(
    Primary Axis=(X=1,Y=0,Z=0)
    Samples Per Segment=16
    Secondary Spline Direction=(X=0,Y=5,Z=0)
    Spline Mode=BSpline
    Stretch=false
    Up Axis=(X=0,Y=1,Z=0)
    debug=true
    Bones=@RigUnit_ItemArray.Items
    Controls=@n15.Items
    ExecuteContext=@RigVMFunction_Sequence.C
   )
   n16 = &RigUnit_ItemArray(Items=((Type=Bone,Name="dyn_L_Ear_02"),(Type=Bone,Name="dyn_L_Ear_03"),(Type=Bone,Name="dyn_L_Ear_04"),(Type=Bone,Name="dyn_L_Ear_05"),(Type=Bone,Name="dyn_L_Ear_06")))
   n17 = &RigUnit_ItemArray(Items=((Type=Bone,Name="dyn_L_Ear_02_c_jnt"),(Type=Bone,Name="dyn_L_Ear_03_c_jnt"),(Type=Bone,Name="dyn_L_Ear_04_c_jnt"),(Type=Bone,Name="dyn_L_Ear_05_c_jnt"),(Type=Bone,Name="dyn_L_Ear_06_c_jnt")))
   > SplineIK_1 = RigVMFunctionReferenceNode(
    Primary Axis=(X=1,Y=0,Z=0)
    Samples Per Segment=16
    Secondary Spline Direction=(X=0,Y=5,Z=0)
    Spline Mode=BSpline
    Up Axis=(X=0,Y=1,Z=0)
    Bones=@n16.Items
    Controls=@n17.Items
   )
   RigVMFunctionReturnNode(ExecuteContext=@RigVMFunction_Sequence.D)
   &RigUnit_ParentConstraint(
    AdvancedSettings=(InterpolationType=Average,RotationOrderForFilter=XZY)
    Child=(Type=Bone,Name="dyn_R_Ear_02_c_jnt")
    Parents=((Item=(Type=Bone,Name="head"),Weight=1))
    Weight=1
    ExecutePin=@RigVMFunction_Sequence.A
   )
   > &RigUnit_ParentConstraint(
    AdvancedSettings=(RotationOrderForFilter=XZY)
    Child=(Type=Bone,Name="dyn_R_Ear_03_c_jnt")
    Parents="((Item=(Type=Bone,Name=\"clavicle_l\"),Weight=0.0197),(Item=(Type=Bone,Name=\"dyn_hood_c\"),Weight=0.0297),(Item=(Type=Bone,Name=\"neck_01\"),Weight=0.0326),(Item=(Type=Bone,Name=\"neck_02\"),Weight=0.0599),(Item=(Type=Bone,Name=\"spine_05\"),Weight=0.0978),(Item=(T ...(truncated)"
    Weight=1
   )
   > &RigUnit_ParentConstraint(
    AdvancedSettings=(RotationOrderForFilter=XZY)
    Child=(Type=Bone,Name="dyn_R_Ear_04_c_jnt")
    Parents="((Item=(Type=Bone,Name=\"dyn_hood_c\"),Weight=0.0244),(Item=(Type=Bone,Name=\"neck_01\"),Weight=0.0939),(Item=(Type=Bone,Name=\"neck_02\"),Weight=0.0966),(Item=(Type=Bone,Name=\"clavicle_l\"),Weight=0.1316),(Item=(Type=Bone,Name=\"head\"),Weight=0.2668),(Item=(Type= ...(truncated)"
    Weight=1
   )
   > &RigUnit_ParentConstraint(
    AdvancedSettings=(RotationOrderForFilter=XZY)
    Child=(Type=Bone,Name="dyn_R_Ear_05_c_jnt")
    Parents="((Item=(Type=Bone,Name=\"deform_upperarm_bk_r\"),Weight=0.0125),(Item=(Type=Bone,Name=\"upperarm_twist_01_r\"),Weight=0.0189),(Item=(Type=Bone,Name=\"spine_04\"),Weight=0.0998),(Item=(Type=Bone,Name=\"spine_05\"),Weight=0.2056),(Item=(Type=Bone,Name=\"spine_03\"),We ...(truncated)"
    Weight=1
   )
   > &RigUnit_ParentConstraint(
    AdvancedSettings=(RotationOrderForFilter=XZY)
    Child=(Type=Bone,Name="dyn_R_Ear_06_c_jnt")
    Parents="((Item=(Type=Bone,Name=\"deform_upperarm_bk_r\"),Weight=0.013),(Item=(Type=Bone,Name=\"spine_02\"),Weight=0.0344),(Item=(Type=Bone,Name=\"spine_05\"),Weight=0.1134),(Item=(Type=Bone,Name=\"spine_04\"),Weight=0.2052),(Item=(Type=Bone,Name=\"clavicle_r\"),Weight=0.303 ...(truncated)"
    Weight=1
   )
   > &RigUnit_ParentConstraint(
    AdvancedSettings=(InterpolationType=Average,RotationOrderForFilter=XZY)
    Child=(Type=Bone,Name="dyn_L_Ear_02_c_jnt")
    Parents=((Item=(Type=Bone,Name="head"),Weight=1))
    Weight=1
   )
   > &RigUnit_ParentConstraint(
    AdvancedSettings=(RotationOrderForFilter=XZY)
    Child=(Type=Bone,Name="dyn_L_Ear_03_c_jnt")
    Parents="((Item=(Type=Bone,Name=\"clavicle_r\"),Weight=0.0272),(Item=(Type=Bone,Name=\"neck_01\"),Weight=0.0429),(Item=(Type=Bone,Name=\"dyn_hood_c\"),Weight=0.0607),(Item=(Type=Bone,Name=\"neck_02\"),Weight=0.0736),(Item=(Type=Bone,Name=\"spine_05\"),Weight=0.1302),(Item=(T ...(truncated)"
    Weight=1
   )
   > n18 = &RigUnit_ParentConstraint(
    AdvancedSettings=(RotationOrderForFilter=XZY)
    Child=(Type=Bone,Name="dyn_L_Ear_04_c_jnt")
    Parents="((Item=(Type=Bone,Name=\"dyn_hood_c\"),Weight=0.0192),(Item=(Type=Bone,Name=\"neck_02\"),Weight=0.0444),(Item=(Type=Bone,Name=\"head\"),Weight=0.0561),(Item=(Type=Bone,Name=\"neck_01\"),Weight=0.0803),(Item=(Type=Bone,Name=\"clavicle_r\"),Weight=0.2547),(Item=(Type= ...(truncated)"
    Weight=1
   )
   n19 = &RigVMFunction_ControlFlowBranch(ExecuteContext=@n18.ExecutePin, Condition=RigVMVariableNode() {variable=IsPistolOut}.Value)
   &RigUnit_ParentConstraint(
    AdvancedSettings=(RotationOrderForFilter=XZY)
    Child=(Type=Bone,Name="dyn_L_Ear_05_c_jnt")
    Parents="((Item=(Type=Bone,Name=\"deform_upperarm_bk_l\"),Weight=0.0117),(Item=(Type=Bone,Name=\"upperarm_twist_01_l\"),Weight=0.0176),(Item=(Type=Bone,Name=\"spine_04\"),Weight=0.1002),(Item=(Type=Bone,Name=\"spine_05\"),Weight=0.2051),(Item=(Type=Bone,Name=\"spine_03\"),We ...(truncated)"
    Weight=1
    ExecutePin=@n19.False
   )
   > &RigUnit_ParentConstraint(
    AdvancedSettings=(RotationOrderForFilter=XZY)
    Child=(Type=Bone,Name="dyn_L_Ear_06_c_jnt")
    Parents=((Item=(Type=Bone,Name="spine_02"),Weight=0.0361),(Item=(Type=Bone,Name="spine_05"),Weight=0.1076),(Item=(Type=Bone,Name="spine_04"),Weight=0.2117),(Item=(Type=Bone,Name="clavicle_l"),Weight=0.3107),(Item=(Type=Bone,Name="spine_03"),Weight=0.334))
    Weight=1
   )
   &RigUnit_ParentConstraint(
    AdvancedSettings=(RotationOrderForFilter=XZY)
    Child=(Type=Bone,Name="dyn_L_Ear_05_c_jnt")
    Parents=((Item=(Type=Bone,Name="upperarm_twist_01_r"),Weight=0),(Item=(Type=Bone,Name="spine_04"),Weight=0.036),(Item=(Type=Bone,Name="spine_05"),Weight=0.476),(Item=(Type=Bone,Name="spine_03"),Weight=0.117),(Item=(Type=Bone,Name="clavicle_r"),Weight=0.37))
    Weight=1
    ExecutePin=@n19.True
   )
   > &RigUnit_ParentConstraint(
    AdvancedSettings=(RotationOrderForFilter=XZY)
    Child=(Type=Bone,Name="dyn_L_Ear_06_c_jnt")
    Parents=((Item=(Type=Bone,Name="spine_05"),Weight=0.294),(Item=(Type=Bone,Name="spine_02"),Weight=0.034),(Item=(Type=Bone,Name="spine_03"),Weight=0.255),(Item=(Type=Bone,Name="spine_04"),Weight=0.167),(Item=(Type=Bone,Name="clavicle_r"),Weight=0.25))
    Weight=1
   )
   &RigVMFunction_MathDoubleMul(
    B=&RigVMFunction_MathDoubleSub(A=1, B=RigVMVariableNode() {variable=isTargetingF}.Value).Result
    A=RigVMVariableNode() {variable=LocalHeadLookingUp}.Value
   )
  ]}
 New Function_ContainedGraph {nodes [
   n1 = RigVMFunctionEntryNode()
   RigVMFunction_MathBoolToInteger = &RigVMFunction_MathBoolToInteger(Value=@n1.onlyTwist)
   RigVMFunction_MathQuaternionSwingTwist = &RigVMFunction_MathQuaternionSwingTwist(
    TwistAxis=@n1.twistAxis
    Input=&RigUnit_GetRelativeTransformForItem(
     Child=(Type=Bone,Name="None")
     Parent=(Type=Bone,Name="None")
     bChildInitial=False
     bParentInitial=False
     Child.Name=@n1.child
     Parent.Name=@n1.parent
     Parent.Type=@n1.parentType
    ).RelativeTransform.Rotation
   )
   RigVMFunction_MathQuaternionToEuler = &RigVMFunction_MathQuaternionToEuler(
    Value=RigVMDispatchNode(
     Values=((),())
     Index=@RigVMFunction_MathBoolToInteger.Result
     Values.0=@RigVMFunction_MathQuaternionSwingTwist.Swing
     Values.1=@RigVMFunction_MathQuaternionSwingTwist.Twist
    ) {template="DISPATCH_RigVMDispatch_SelectInt32(in Index,in Values,out Result)"}.Result
    RotationOrder=@n1.Rotation Order
   )
   > DISPATCH_RigVMDispatch_Print = RigVMDispatchNode(ScreenColor=(R=1,G=1,B=1,A=1), ScreenDuration=1.5, Enabled=@n1.debugRotation, Value=@RigVMFunction_MathQuaternionToEuler.Result) {template="DISPATCH_RigVMDispatch_Print(in Prefix,in Value,in Enabled,in ScreenDuration,in ScreenColor)"}
   Clamp = &RigVMFunction_MathIntClamp(Value=@n1.primAxis)
   DISPATCH_RigVMDispatch_SelectInt32 = RigVMDispatchNode(
    Values=(0,0,0)
    Index=@Clamp.Result
    Values.0=@RigVMFunction_MathQuaternionToEuler.Result.X
    Values.1=@RigVMFunction_MathQuaternionToEuler.Result.Y
    Values.2=@RigVMFunction_MathQuaternionToEuler.Result.Z
   ) {template="DISPATCH_RigVMDispatch_SelectInt32(in Index,in Values,out Result)"}
   Remap = &RigVMFunction_MathDoubleRemap(
    TargetMaximum=1
    TargetMinimum=0
    bClamp=False
    Value=@DISPATCH_RigVMDispatch_SelectInt32.Result
    SourceMaximum=@n1.max
    SourceMinimum=@n1.min
   )
   Multiply = &RigVMFunction_MathDoubleMul(A=&RigVMFunction_MathDoubleClamp(Value=@Remap.Result).Result, B=@n1.premult)
   > RigVMFunctionReturnNode(outputWeight=@Multiply.Result)
  ]}
 New Function_ContainedGraph {nodes [
   n1 = RigVMFunctionEntryNode()
   RigVMFunction_MathBoolToInteger = &RigVMFunction_MathBoolToInteger(Value=@n1.onlyTwist)
   RigVMFunction_MathQuaternionSwingTwist = &RigVMFunction_MathQuaternionSwingTwist(
    TwistAxis=@n1.twistAxis
    Input=&RigUnit_GetRelativeTransformForItem(
     Child=(Type=Bone,Name="None")
     Parent=(Type=Bone,Name="None")
     bChildInitial=False
     bParentInitial=False
     Child.Name=@n1.child
     Parent.Name=@n1.parent
     Parent.Type=@n1.parentType
    ).RelativeTransform.Rotation
   )
   RigVMFunction_MathQuaternionToEuler = &RigVMFunction_MathQuaternionToEuler(
    Value=RigVMDispatchNode(
     Values=((),())
     Index=@RigVMFunction_MathBoolToInteger.Result
     Values.0=@RigVMFunction_MathQuaternionSwingTwist.Swing
     Values.1=@RigVMFunction_MathQuaternionSwingTwist.Twist
    ) {template="DISPATCH_RigVMDispatch_SelectInt32(in Index,in Values,out Result)"}.Result
    RotationOrder=@n1.Rotation Order
   )
   > DISPATCH_RigVMDispatch_Print = RigVMDispatchNode(ScreenColor=(R=1,G=1,B=1,A=1), ScreenDuration=1.5, Enabled=@n1.debugRotation, Value=@RigVMFunction_MathQuaternionToEuler.Result) {template="DISPATCH_RigVMDispatch_Print(in Prefix,in Value,in Enabled,in ScreenDuration,in ScreenColor)"}
   Clamp = &RigVMFunction_MathIntClamp(Value=@n1.primAxis)
   DISPATCH_RigVMDispatch_SelectInt32 = RigVMDispatchNode(
    Values=(0,0,0)
    Index=@Clamp.Result
    Values.0=@RigVMFunction_MathQuaternionToEuler.Result.X
    Values.1=@RigVMFunction_MathQuaternionToEuler.Result.Y
    Values.2=@RigVMFunction_MathQuaternionToEuler.Result.Z
   ) {template="DISPATCH_RigVMDispatch_SelectInt32(in Index,in Values,out Result)"}
   Remap = &RigVMFunction_MathDoubleRemap(
    TargetMaximum=1
    TargetMinimum=0
    bClamp=False
    Value=@DISPATCH_RigVMDispatch_SelectInt32.Result
    SourceMaximum=@n1.max
    SourceMinimum=@n1.min
   )
   Multiply = &RigVMFunction_MathDoubleMul(A=&RigVMFunction_MathDoubleClamp(Value=@Remap.Result).Result, B=@n1.premult)
   > RigVMFunctionReturnNode(outputWeight=@Multiply.Result)
  ]}
 New Function_ContainedGraph {nodes [
   n1 = RigVMFunctionEntryNode()
   RigVMFunction_MathBoolToInteger = &RigVMFunction_MathBoolToInteger(Value=@n1.onlyTwist)
   RigVMFunction_MathQuaternionSwingTwist = &RigVMFunction_MathQuaternionSwingTwist(
    TwistAxis=@n1.twistAxis
    Input=&RigUnit_GetRelativeTransformForItem(
     Child=(Type=Bone,Name="None")
     Parent=(Type=Bone,Name="None")
     bChildInitial=False
     bParentInitial=False
     Child.Name=@n1.child
     Parent.Name=@n1.parent
     Parent.Type=@n1.parentType
    ).RelativeTransform.Rotation
   )
   RigVMFunction_MathQuaternionToEuler = &RigVMFunction_MathQuaternionToEuler(
    Value=RigVMDispatchNode(
     Values=((),())
     Index=@RigVMFunction_MathBoolToInteger.Result
     Values.0=@RigVMFunction_MathQuaternionSwingTwist.Swing
     Values.1=@RigVMFunction_MathQuaternionSwingTwist.Twist
    ) {template="DISPATCH_RigVMDispatch_SelectInt32(in Index,in Values,out Result)"}.Result
    RotationOrder=@n1.Rotation Order
   )
   > DISPATCH_RigVMDispatch_Print = RigVMDispatchNode(ScreenColor=(R=1,G=1,B=1,A=1), ScreenDuration=1.5, Enabled=@n1.debugRotation, Value=@RigVMFunction_MathQuaternionToEuler.Result) {template="DISPATCH_RigVMDispatch_Print(in Prefix,in Value,in Enabled,in ScreenDuration,in ScreenColor)"}
   Clamp = &RigVMFunction_MathIntClamp(Value=@n1.primAxis)
   DISPATCH_RigVMDispatch_SelectInt32 = RigVMDispatchNode(
    Values=(0,0,0)
    Index=@Clamp.Result
    Values.0=@RigVMFunction_MathQuaternionToEuler.Result.X
    Values.1=@RigVMFunction_MathQuaternionToEuler.Result.Y
    Values.2=@RigVMFunction_MathQuaternionToEuler.Result.Z
   ) {template="DISPATCH_RigVMDispatch_SelectInt32(in Index,in Values,out Result)"}
   Remap = &RigVMFunction_MathDoubleRemap(
    TargetMaximum=1
    TargetMinimum=0
    bClamp=False
    Value=@DISPATCH_RigVMDispatch_SelectInt32.Result
    SourceMaximum=@n1.max
    SourceMinimum=@n1.min
   )
   Multiply = &RigVMFunction_MathDoubleMul(A=&RigVMFunction_MathDoubleClamp(Value=@Remap.Result).Result, B=@n1.premult)
   > RigVMFunctionReturnNode(outputWeight=@Multiply.Result)
  ]}
 New Function_ContainedGraph {nodes [
   n1 = RigVMFunctionEntryNode()
   > RigVMFunctionReturnNode(Float=&RigVMFunction_MathBoolToFloat(Value=@n1.Bool).Result)
  ]}
 New Function_ContainedGraph {nodes [
   RigVMFunctionEntryNode()
   > n1 = &RigUnit_SphericalPoseReader(
    ActiveRegionScaleFactors=(PositiveWidth=1.5,NegativeWidth=0.5,PositiveHeight=0.75,NegativeHeight=0.75)
    ActiveRegionSize=0.075
    Debug=(bDrawDebug=true,DebugScale=25,DebugSegments=20,DebugThickness=0.25)
    DriverAxis=(X=1,Y=0,Z=0)
    DriverItem=(Type=Bone,Name="upperarm_l")
    FalloffRegionScaleFactors=(PositiveWidth=1,NegativeWidth=1,PositiveHeight=1,NegativeHeight=1)
    FalloffSize=0.25
    OptionalParentItem=(Type=Bone,Name="spine_05")
    RotationOffset=(X=60,Y=130,Z=-150)
   )
   n2 = &RigVMFunction_MathDoubleMul(A=@n1.OutputParam, B=&RigVMFunction_MathDoubleSub(A=1, B=RigVMVariableNode() {variable=isTargetingF}.Value).Result)
   > &RigUnit_ModifyTransforms(
    ItemToModify=((Item=(Type=Bone,Name="dyn_L_Ear_04_c_jnt"),Transform=(Rotation=(X=0.1736,Y=0,Z=0,W=0.9848),Translation=(X=0,Y=0,Z=-3),Scale3D=(X=1,Y=1,Z=1))))
    WeightMinimum=0
    Weight=@n2.Result
   )
   > &RigUnit_ModifyTransforms(
    ItemToModify=((Item=(Type=Bone,Name="dyn_L_Ear_05_c_jnt"),Transform=(Rotation=(X=0.4226,Y=0,Z=0,W=0.9063),Translation=(X=0,Y=0,Z=-6),Scale3D=(X=1,Y=1,Z=1))))
    WeightMinimum=0
    Weight=@n2.Result
   )
   > &RigUnit_ModifyTransforms(
    ItemToModify=((Item=(Type=Bone,Name="dyn_L_Ear_06_c_jnt"),Transform=(Rotation=(X=0.1736,Y=0,Z=0,W=0.9848),Translation=(X=2,Y=0,Z=-8),Scale3D=(X=1,Y=1,Z=1))))
    WeightMinimum=0
    Weight=@n2.Result
   )
   > DeltaRotation_4 = RigVMCollapseNode(
    Rotation Order=ZYX
    child=clavicle_l
    debugRotation=false
    max=-16
    min=0
    onlyTwist=false
    parent=clavicle_l_SpineDelta_null
    parentType=Null
    premult=1
    primAxis=2
    twistAxis=(X=1,Y=0,Z=0)
   )
   n3 = &RigVMFunction_MathDoubleMul(
    A=@DeltaRotation_4.outputWeight
    B=&RigVMFunction_MathDoubleSub(A=1, B=RigVMVariableNode() {variable=isTargetingF}.Value).Result
   )
   > &RigUnit_ModifyTransforms(
    ItemToModify=((Item=(Type=Bone,Name="dyn_L_Ear_05_c_jnt"),Transform=(Rotation=(X=0.1736,Y=0,Z=0,W=0.9848),Translation=(X=0,Y=-1,Z=-1),Scale3D=(X=1,Y=1,Z=1))))
    WeightMinimum=0
    Weight=@n3.Result
   )
   > &RigUnit_ModifyTransforms(
    ItemToModify=((Item=(Type=Bone,Name="dyn_L_Ear_06_c_jnt"),Transform=(Rotation=(X=0,Y=0,Z=0,W=1),Translation=(X=2,Y=-1,Z=0),Scale3D=(X=1,Y=1,Z=1))))
    WeightMinimum=0
    Weight=@n3.Result
   )
   Subtract = &RigVMFunction_MathDoubleSub(A=1, B=@n3.Result)
   Multiply = &RigVMFunction_MathDoubleMul(B=@Subtract.Result, A=RigVMVariableNode() {variable=isTacticalJumpingF}.Value)
   > &RigUnit_ModifyTransforms(
    ItemToModify=((Item=(Type=Bone,Name="dyn_L_Ear_04_c_jnt"),Transform=(Rotation=(X=0,Y=0,Z=0,W=1),Translation=(X=0,Y=-1,Z=-1),Scale3D=(X=1,Y=1,Z=1))))
    WeightMinimum=0
    Weight=@Multiply.Result
   )
   > &RigUnit_ModifyTransforms(
    ItemToModify=((Item=(Type=Bone,Name="dyn_L_Ear_05_c_jnt"),Transform=(Rotation=(X=0,Y=0,Z=0,W=1),Translation=(X=0,Y=-1,Z=-5),Scale3D=(X=1,Y=1,Z=1))))
    WeightMinimum=0
    Weight=@Multiply.Result
   )
   > &RigUnit_ModifyTransforms(
    ItemToModify=((Item=(Type=Bone,Name="dyn_L_Ear_06_c_jnt"),Transform=(Rotation=(X=0,Y=0,Z=0,W=1),Translation=(X=2,Y=0,Z=-6),Scale3D=(X=1,Y=1,Z=1))))
    WeightMinimum=0
    Weight=@Multiply.Result
   )
   > n4 = &RigUnit_SphericalPoseReader(
    ActiveRegionScaleFactors=(PositiveWidth=0.75,NegativeWidth=0.75,PositiveHeight=1.5,NegativeHeight=0.5)
    ActiveRegionSize=0.075
    Debug=(bDrawDebug=true,DebugScale=25,DebugSegments=20,DebugThickness=0.25)
    DriverAxis=(X=-1,Y=0,Z=0)
    DriverItem=(Type=Bone,Name="upperarm_r")
    FalloffRegionScaleFactors=(PositiveWidth=1,NegativeWidth=1,PositiveHeight=1,NegativeHeight=1)
    FalloffSize=0.25
    OptionalParentItem=(Type=Bone,Name="spine_05")
    RotationOffset=(X=70,Y=-40,Z=70)
   )
   > &RigUnit_ModifyTransforms(
    ItemToModify=((Item=(Type=Bone,Name="dyn_R_Ear_04_c_jnt"),Transform=(Rotation=(X=-0.1736,Y=0,Z=0,W=0.9848),Translation=(X=0,Y=0,Z=3),Scale3D=(X=1,Y=1,Z=1))))
    Weight=1
    WeightMinimum=0
   )
   n5 = &RigVMFunction_MathDoubleMul(A=@n4.OutputParam, B=&RigVMFunction_MathDoubleSub(A=1, B=RigVMVariableNode() {variable=isTargetingF}.Value).Result)
   > &RigUnit_ModifyTransforms(
    ItemToModify=((Item=(Type=Bone,Name="dyn_R_Ear_05_c_jnt"),Transform=(Rotation=(X=-0.4226,Y=0,Z=0,W=0.9063),Translation=(X=0,Y=0,Z=6),Scale3D=(X=1,Y=1,Z=1))))
    WeightMinimum=0
    Weight=@n5.Result
   )
   > &RigUnit_ModifyTransforms(
    ItemToModify=((Item=(Type=Bone,Name="dyn_R_Ear_06_c_jnt"),Transform=(Rotation=(X=-0.1736,Y=0,Z=0,W=0.9848),Translation=(X=2,Y=0,Z=8),Scale3D=(X=1,Y=1,Z=1))))
    WeightMinimum=0
    Weight=@n5.Result
   )
   > DeltaRotation_5 = RigVMCollapseNode(
    Rotation Order=ZYX
    child=clavicle_r
    debugRotation=false
    max=-16
    min=0
    onlyTwist=false
    parent=clavicle_r_SpineDelta_null
    parentType=Null
    premult=1
    primAxis=2
    twistAxis=(X=1,Y=0,Z=0)
   )
   n6 = &RigVMFunction_MathDoubleMul(
    A=@DeltaRotation_5.outputWeight
    B=&RigVMFunction_MathDoubleSub(A=1, B=RigVMVariableNode() {variable=isTargetingF}.Value).Result
   )
   > &RigUnit_ModifyTransforms(
    ItemToModify=((Item=(Type=Bone,Name="dyn_R_Ear_05_c_jnt"),Transform=(Rotation=(X=-0.1736,Y=0,Z=0,W=0.9848),Translation=(X=0,Y=-1,Z=1),Scale3D=(X=1,Y=1,Z=1))))
    WeightMinimum=0
    Weight=@n6.Result
   )
   > &RigUnit_ModifyTransforms(
    ItemToModify=((Item=(Type=Bone,Name="dyn_R_Ear_06_c_jnt"),Transform=(Rotation=(X=0,Y=0,Z=0,W=1),Translation=(X=2,Y=-1,Z=0),Scale3D=(X=1,Y=1,Z=1))))
    WeightMinimum=0
    Weight=@n6.Result
   )
   n7 = &RigVMFunction_MathDoubleMul(B=&RigVMFunction_MathDoubleSub(A=1, B=@n6.Result).Result, A=RigVMVariableNode() {variable=isTacticalJumpingF}.Value)
   > &RigUnit_ModifyTransforms(
    ItemToModify=((Item=(Type=Bone,Name="dyn_R_Ear_04_c_jnt"),Transform=(Rotation=(X=0,Y=0,Z=0,W=1),Translation=(X=0,Y=-1,Z=1),Scale3D=(X=1,Y=1,Z=1))))
    WeightMinimum=0
    Weight=@n7.Result
   )
   > &RigUnit_ModifyTransforms(
    ItemToModify=((Item=(Type=Bone,Name="dyn_R_Ear_05_c_jnt"),Transform=(Rotation=(X=0,Y=0,Z=0,W=1),Translation=(X=0,Y=-1,Z=5),Scale3D=(X=1,Y=1,Z=1))))
    WeightMinimum=0
    Weight=@n7.Result
   )
   > &RigUnit_ModifyTransforms(
    ItemToModify=((Item=(Type=Bone,Name="dyn_R_Ear_06_c_jnt"),Transform=(Rotation=(X=0,Y=0,Z=0,W=1),Translation=(X=2,Y=0,Z=6),Scale3D=(X=1,Y=1,Z=1))))
    WeightMinimum=0
    Weight=@n7.Result
   )
   > RigVMFunctionReturnNode()
  ]}
 New Function_ContainedGraph {nodes [
   n1 = RigVMFunctionEntryNode()
   RigVMFunction_MathBoolToInteger = &RigVMFunction_MathBoolToInteger(Value=@n1.onlyTwist)
   RigVMFunction_MathQuaternionSwingTwist = &RigVMFunction_MathQuaternionSwingTwist(
    TwistAxis=@n1.twistAxis
    Input=&RigUnit_GetRelativeTransformForItem(
     Child=(Type=Bone,Name="None")
     Parent=(Type=Bone,Name="None")
     bChildInitial=False
     bParentInitial=False
     Child.Name=@n1.child
     Parent.Name=@n1.parent
     Parent.Type=@n1.parentType
    ).RelativeTransform.Rotation
   )
   RigVMFunction_MathQuaternionToEuler = &RigVMFunction_MathQuaternionToEuler(
    Value=RigVMDispatchNode(
     Values=((),())
     Index=@RigVMFunction_MathBoolToInteger.Result
     Values.0=@RigVMFunction_MathQuaternionSwingTwist.Swing
     Values.1=@RigVMFunction_MathQuaternionSwingTwist.Twist
    ) {template="DISPATCH_RigVMDispatch_SelectInt32(in Index,in Values,out Result)"}.Result
    RotationOrder=@n1.Rotation Order
   )
   > DISPATCH_RigVMDispatch_Print = RigVMDispatchNode(ScreenColor=(R=1,G=1,B=1,A=1), ScreenDuration=1.5, Enabled=@n1.debugRotation, Value=@RigVMFunction_MathQuaternionToEuler.Result) {template="DISPATCH_RigVMDispatch_Print(in Prefix,in Value,in Enabled,in ScreenDuration,in ScreenColor)"}
   Clamp = &RigVMFunction_MathIntClamp(Value=@n1.primAxis)
   DISPATCH_RigVMDispatch_SelectInt32 = RigVMDispatchNode(
    Values=(0,0,0)
    Index=@Clamp.Result
    Values.0=@RigVMFunction_MathQuaternionToEuler.Result.X
    Values.1=@RigVMFunction_MathQuaternionToEuler.Result.Y
    Values.2=@RigVMFunction_MathQuaternionToEuler.Result.Z
   ) {template="DISPATCH_RigVMDispatch_SelectInt32(in Index,in Values,out Result)"}
   Remap = &RigVMFunction_MathDoubleRemap(
    TargetMaximum=1
    TargetMinimum=0
    bClamp=False
    Value=@DISPATCH_RigVMDispatch_SelectInt32.Result
    SourceMaximum=@n1.max
    SourceMinimum=@n1.min
   )
   Multiply = &RigVMFunction_MathDoubleMul(A=&RigVMFunction_MathDoubleClamp(Value=@Remap.Result).Result, B=@n1.premult)
   > RigVMFunctionReturnNode(outputWeight=@Multiply.Result)
  ]}
 New Function_ContainedGraph {nodes [
   n1 = RigVMFunctionEntryNode()
   RigVMFunction_MathBoolToInteger = &RigVMFunction_MathBoolToInteger(Value=@n1.onlyTwist)
   RigVMFunction_MathQuaternionSwingTwist = &RigVMFunction_MathQuaternionSwingTwist(
    TwistAxis=@n1.twistAxis
    Input=&RigUnit_GetRelativeTransformForItem(
     Child=(Type=Bone,Name="None")
     Parent=(Type=Bone,Name="None")
     bChildInitial=False
     bParentInitial=False
     Child.Name=@n1.child
     Parent.Name=@n1.parent
     Parent.Type=@n1.parentType
    ).RelativeTransform.Rotation
   )
   RigVMFunction_MathQuaternionToEuler = &RigVMFunction_MathQuaternionToEuler(
    Value=RigVMDispatchNode(
     Values=((),())
     Index=@RigVMFunction_MathBoolToInteger.Result
     Values.0=@RigVMFunction_MathQuaternionSwingTwist.Swing
     Values.1=@RigVMFunction_MathQuaternionSwingTwist.Twist
    ) {template="DISPATCH_RigVMDispatch_SelectInt32(in Index,in Values,out Result)"}.Result
    RotationOrder=@n1.Rotation Order
   )
   > DISPATCH_RigVMDispatch_Print = RigVMDispatchNode(ScreenColor=(R=1,G=1,B=1,A=1), ScreenDuration=1.5, Enabled=@n1.debugRotation, Value=@RigVMFunction_MathQuaternionToEuler.Result) {template="DISPATCH_RigVMDispatch_Print(in Prefix,in Value,in Enabled,in ScreenDuration,in ScreenColor)"}
   Clamp = &RigVMFunction_MathIntClamp(Value=@n1.primAxis)
   DISPATCH_RigVMDispatch_SelectInt32 = RigVMDispatchNode(
    Values=(0,0,0)
    Index=@Clamp.Result
    Values.0=@RigVMFunction_MathQuaternionToEuler.Result.X
    Values.1=@RigVMFunction_MathQuaternionToEuler.Result.Y
    Values.2=@RigVMFunction_MathQuaternionToEuler.Result.Z
   ) {template="DISPATCH_RigVMDispatch_SelectInt32(in Index,in Values,out Result)"}
   Remap = &RigVMFunction_MathDoubleRemap(
    TargetMaximum=1
    TargetMinimum=0
    bClamp=False
    Value=@DISPATCH_RigVMDispatch_SelectInt32.Result
    SourceMaximum=@n1.max
    SourceMinimum=@n1.min
   )
   Multiply = &RigVMFunction_MathDoubleMul(A=&RigVMFunction_MathDoubleClamp(Value=@Remap.Result).Result, B=@n1.premult)
   > RigVMFunctionReturnNode(outputWeight=@Multiply.Result)
  ]}
 RigVMFunctionLibrary {nodes [
   ArmsBackEarsInteraction = RigVMCollapseNode(ExecuteContext=())
   BackBlingIkSetup = RigVMCollapseNode(ExecuteContext=())
   BoolToFloat = RigVMCollapseNode(Bool=False)
   Connect FKPreviewControls = RigVMCollapseNode(Execute=(), On=False)
   DeltaRotation = RigVMCollapseNode(ExecuteContext=())
   Ears_RigidBodies_DEF = RigVMCollapseNode(ExecuteContext=())
   HoodHeadTwistAdjust = RigVMCollapseNode(Execute=())
  ]}
 RigVMFunction_MathBoolOr_2_ContainedGraph {nodes [
   n1 = RigVMFunctionEntryNode()
   RigVMFunction_MathBoolOr = &RigVMFunction_MathBoolOr(A=@n1.A, B=@n1.B)
   RigVMFunctionReturnNode(Result=&RigVMFunction_MathBoolOr(B=@n1.C, A=@RigVMFunction_MathBoolOr.Result).Result)
  ]}
 RigVMFunction_MathBoolOr_2_ContainedGraph {nodes [
   n1 = RigVMFunctionEntryNode()
   RigVMFunctionReturnNode(Result=&RigVMFunction_MathBoolOr(B=@n1.C, A=&RigVMFunction_MathBoolOr(A=@n1.A, B=@n1.B).Result).Result)
  ]}
 RigVMFunction_MathBoolOr_3_ContainedGraph {nodes [
   n1 = RigVMFunctionEntryNode()
   RigVMFunctionReturnNode(Result=&RigVMFunction_MathBoolOr(B=@n1.C, A=&RigVMFunction_MathBoolOr(A=@n1.A, B=@n1.B).Result).Result)
  ]}
 RigVMFunction_Sequence_1_ContainedGraph {nodes [
   n1 = RigVMFunctionEntryNode()
   RigVMFunction_Sequence = &RigVMFunction_Sequence(ExecuteContext=@n1.ExecuteContext)
   n2 = &RigVMFunction_Sequence(ExecuteContext=@RigVMFunction_Sequence.B)
   RigVMFunctionReturnNode(A=@RigVMFunction_Sequence.A, B=@n2.A, C=@n2.B)
  ]}
 RigVMFunction_Sequence_1_ContainedGraph {nodes [
   n1 = RigVMFunctionEntryNode()
   RigVMFunction_Sequence = &RigVMFunction_Sequence(ExecuteContext=@n1.ExecuteContext)
   n2 = &RigVMFunction_Sequence(ExecuteContext=@RigVMFunction_Sequence.B)
   n3 = &RigVMFunction_Sequence(ExecuteContext=@n2.B)
   n4 = &RigVMFunction_Sequence(ExecuteContext=@n3.B)
   n5 = &RigVMFunction_Sequence(ExecuteContext=@n4.B)
   RigVMFunctionReturnNode(A=@RigVMFunction_Sequence.A, B=@n2.A, C=@n3.A, D=@n4.A, E=@n5.A, F=@n5.B)
  ]}
 RigVMFunction_Sequence_1_ContainedGraph {nodes [
   n1 = RigVMFunctionEntryNode()
   RigVMFunction_Sequence = &RigVMFunction_Sequence(ExecuteContext=@n1.ExecuteContext)
   n2 = &RigVMFunction_Sequence(ExecuteContext=@RigVMFunction_Sequence.B)
   n3 = &RigVMFunction_Sequence(ExecuteContext=@n2.B)
   RigVMFunctionReturnNode(A=@RigVMFunction_Sequence.A, B=@n2.A, C=@n3.A, D=@n3.B)
  ]}
 RigVMModel {nodes [
   RigVMFunction_MathBoolMake = /Script/RigVM.RigVMFunction_MathBoolMake(Value=false) {is=struct RigVMUnitNode}
   Connect FKPreviewControls = RigVMFunctionReferenceNode(Execute=@n5.ExecutePin, On=@RigVMFunction_MathBoolMake.Value)
   HoodHeadTwistAdjust = RigVMFunctionReferenceNode(Execute=@"Connect FKPreviewControls".Execute)
   RigVMFunction_MathBoolNot = &RigVMFunction_MathBoolNot(Value=RigVMVariableNode() {variable=isRbanEnabled}.Value)
   RigVMFunction_MathBoolOr = RigVMAggregateNode(
    B=@RigVMFunction_MathBoolNot.Result
    C=RigVMVariableNode() {variable=isDBNO}.Value
    A=RigVMVariableNode() {variable=IsWearingBackBling}.Value
   )
   RigVMFunction_ControlFlowBranch = &RigVMFunction_ControlFlowBranch(ExecuteContext=@HoodHeadTwistAdjust.Execute, Condition=@RigVMFunction_MathBoolOr.Result)
   &RigUnit_ParentConstraint(
    AdvancedSettings=(InterpolationType=Average,RotationOrderForFilter=XZY)
    Child=(Type=Bone,Name="base_ZipperMain")
    Parents=((Item=(Type=Bone,Name="spine_04"),Weight=0.066),(Item=(Type=Bone,Name="spine_05"),Weight=0.154),(Item=(Type=Bone,Name="neck_02"),Weight=0.019),(Item=(Type=Bone,Name="spine_03"),Weight=0.011),(Item=(Type=Bone,Name="dyn_hood_f"),Weight=0.744))
    Weight=1
    ExecutePin=@RigVMFunction_ControlFlowBranch.Completed
   )
   > &RigUnit_ParentConstraint(
    AdvancedSettings=(InterpolationType=Average,RotationOrderForFilter=XZY)
    Child=(Type=Null,Name="dyn_ZipperMain_aim_null")
    Parents=((Item=(Type=Bone,Name="spine_05"),Weight=0.128),(Item=(Type=Bone,Name="spine_04"),Weight=0.045),(Item=(Type=Bone,Name="dyn_hood_f"),Weight=0.82),(Item=(Type=Bone,Name="None"),Weight=1))
    Weight=1
   )
   RigUnit_GetTransform = &RigUnit_GetTransform(Item=(Type=Bone,Name="neck_02"), bInitial=False)
   > &RigUnit_AimConstraintLocalSpaceOffset(
    AimAxis=(X=0,Y=1,Z=0)
    Child=(Type=Bone,Name="base_ZipperMain")
    Filter=(bX=True,bY=true,bZ=true)
    Parents=((Item=(Type=Null,Name="dyn_ZipperMain_aim_null"),Weight=1))
    UpAxis=(X=-1,Y=0,Z=0)
    Weight=1
    WorldUp=(Target=(X=0,Y=0,Z=1),Kind=Location,Space=(Type=Bone,Name="neck_02"))
    WorldUp.Target=@RigUnit_GetTransform.Transform.Translation
   )
   Ears_RigidBodies_DEF = RigVMFunctionReferenceNode(ExecuteContext=@RigVMFunction_ControlFlowBranch.False)
   BackBlingIkSetup = RigVMFunctionReferenceNode(ExecuteContext=@RigVMFunction_ControlFlowBranch.True)
   RigUnit_BeginExecution = &RigUnit_BeginExecution()
   RigVMFunction_Sequence = RigVMAggregateNode(ExecuteContext=@RigUnit_BeginExecution.ExecutePin)
   FN_DeformRig_FNC = RigVMFunctionReferenceNode(
    DeformRig_ Curve="(deformRig_l_upperarm_up_alpha_72_#1=False,deformRig_r_upperarm_up_alpha_73_#2=False,deformRig_l_upperarm_fr_alpha_74_#3=true,deformRig_r_upperarm_fr_alpha_75_CB51B80 ...(truncated)"
    DeformRig_Bool="(upperarm_1_#4=True,elbow_3_#5=True,wrist_7_#6=True,groin_8_#7=True,glute_17_#8=True,knee_9_40C408324CA261 ...(truncated)"
    DeformRig_Debug=(upperarm_debug_18_#4=False,wrist_debug_20_#6=False,lat_debug_26_#9=False,groinGlute_debug_27_#7=False)
    DeformRig_Value="(upperarm_up_18_#10=(Rotation=(X=0.0,Y=-0.0,Z=0.0,W=1.0),Translation=(X=0.0,Y=0.0,Z=8.0),Scale3D=(X=0.70,Y=1.0,Z=1.0)),upperarm_fwd_fr_19_#11=(Rotation=(X=0.0,Y=-0.0,Z=0.0,W=1.0),Translation=(X=0.0, ...(truncated)"
    ExecuteContext=@RigVMFunction_Sequence.A
   )
   > &RigUnit_ParentConstraint(
    AdvancedSettings=(RotationOrderForFilter=XZY)
    Child=(Type=Null,Name="R_ArmAim_Null")
    Parents="((Item=(Type=Bone,Name=\"upperarm_twist_02_r\"),Weight=0.0126),(Item=(Type=Bone,Name=\"spine_04\"),Weight=0.0448),(Item=(Type=Bone,Name=\"spine_05\"),Weight=0.0941),(Item=(Type=Bone,Name=\"deform_upperarm_up_r\"),Weight=0.0992),(Item=(Type=Bone,Name=\"deform_uppera ...(truncated)"
    Weight=1
   )
   > &RigUnit_ParentConstraint(
    AdvancedSettings=(RotationOrderForFilter=XZY)
    Child=(Type=Null,Name="R_SpineAim_null")
    Parents="((Item=(Type=Bone,Name=\"deform_upperarm_bk_r\"),Weight=0.0125),(Item=(Type=Bone,Name=\"upperarm_twist_01_r\"),Weight=0.0189),(Item=(Type=Bone,Name=\"spine_04\"),Weight=0.0998),(Item=(Type=Bone,Name=\"spine_05\"),Weight=0.2056),(Item=(Type=Bone,Name=\"spine_03\"),We ...(truncated)"
    Weight=1
   )
   DeltaRotation_1 = RigVMCollapseNode(
    Rotation Order=ZYX
    child=head
    debugRotation=false
    max=-25
    min=0
    onlyTwist=false
    parent=head_Spine05Delta_null
    parentType=Null
    premult=1
    primAxis=2
    twistAxis=(X=1,Y=0,Z=0)
    ExecuteContext=@RigVMFunction_Sequence.B
   )
   RigVMFunction_AnimEvalRichCurve = &RigVMFunction_AnimEvalRichCurve(
    Curve=(EditorCurveData=(Keys=((InterpMode=RCIM_Cubic),(InterpMode=RCIM_Cubic,Time=0.5726,Value=0.3579,ArriveTangent=1,LeaveTangent=1),(InterpMode=RCIM_Cubic,TangentMode=RCTM_User,Time=1,Value=1,ArriveTangent=1.2795,LeaveTangent=1.2795))))
    SourceMinimum=0
    TargetMinimum=0
    Value=@DeltaRotation_1.outputWeight
   )
   > RigVMVariableNode(Value=@RigVMFunction_AnimEvalRichCurve.Result) {variable=IsHeadLookingUp}
   n1 = &RigVMFunction_AnimEvalRichCurve(
    Curve=(EditorCurveData=(Keys=((),(Time=1,Value=1)),DefaultValue=340282346638528859811704183484516925440,PreInfinityExtrap=RCCE_Constant,PostInfinityExtrap=RCCE_Constant),ExternalCurve=None)
    SourceMinimum=0
    TargetMinimum=0
    Value=&RigVMFunction_MathDoubleRemap(
     SourceMaximum=85
     SourceMinimum=44
     TargetMaximum=0
     TargetMinimum=1
     bClamp=false
     Value=&RigUnit_GetRelativeTransformForItem(Child=(Type=Bone,Name="pelvis"), Parent=(Type=Bone,Name="root"), bChildInitial=False, bParentInitial=False).RelativeTransform.Translation.Z
    ).Result
   )
   > RigVMVariableNode(Value=@n1.Result) {variable=IsCrouching}
   n2 = &RigVMFunction_AnimEvalRichCurve(
    Curve=(EditorCurveData=(Keys=((InterpMode=RCIM_Cubic),(InterpMode=RCIM_Cubic,Time=0.5,Value=1),(Time=0.8))))
    SourceMaximum=2
    SourceMinimum=0
    TargetMinimum=0
    Value=&RigVMFunction_MathDoubleRemap(
     SourceMaximum=75
     SourceMinimum=45
     TargetMaximum=1
     TargetMinimum=0
     bClamp=false
     Value=&RigUnit_GetRelativeTransformForItem(Child=(Type=Bone,Name="pelvis"), Parent=(Type=Bone,Name="root"), bChildInitial=False, bParentInitial=False).RelativeTransform.Translation.Z
    ).Result
   )
   > RigVMVariableNode(Value=@n2.Result) {variable=IsCrouchWalking}
   > DeltaRotation_2 = RigVMCollapseNode(
    Rotation Order=ZYX
    child=head
    debugRotation=false
    max=-60
    min=-30
    onlyTwist=false
    parent=head_Spine05Delta_null
    parentType=Null
    premult=1
    primAxis=2
    twistAxis=(X=1,Y=0,Z=0)
   )
   n3 = &RigVMFunction_AnimEvalRichCurve(
    Curve="(EditorCurveData=(Keys=((InterpMode=RCIM_Cubic),(InterpMode=RCIM_Cubic,TangentMode=RCTM_User,Time=0.6379,Value=0.3022,ArriveTangent=1.0148,LeaveTangent=1.0148),(InterpMode=RCIM_Cubic,TangentMode=RCTM_User,Time=1,Value=1,ArriveTangent=2.5251,LeaveTangent=2. ...(truncated)"
    SourceMinimum=0
    TargetMinimum=0
    Value=@DeltaRotation_2.outputWeight
   )
   > RigVMVariableNode(Value=@n3.Result) {variable=IsLookingUpExtreme}
   > DeltaRotation_3 = RigVMCollapseNode(
    Rotation Order=ZYX
    child=head
    debugRotation=false
    max=12
    min=-2
    onlyTwist=false
    parent=head_Neck01Delta_null
    parentType=Null
    premult=1
    primAxis=2
    twistAxis=(X=1,Y=0,Z=0)
   )
   n4 = &RigVMFunction_AnimEvalRichCurve(
    Curve=(EditorCurveData=(Keys=((InterpMode=RCIM_Cubic),(InterpMode=RCIM_Cubic,Time=0.5726,Value=0.3579,ArriveTangent=1,LeaveTangent=1),(InterpMode=RCIM_Cubic,TangentMode=RCTM_User,Time=1,Value=1,ArriveTangent=1.2795,LeaveTangent=1.2795))))
    SourceMinimum=0
    TargetMinimum=0
    Value=@DeltaRotation_3.outputWeight
   )
   > RigVMVariableNode(Value=@n4.Result) {variable=IsHeadLookingDown}
   Subtract = &RigVMFunction_MathDoubleSub(A=1, B=&RigVMFunction_MathBoolToFloat(Value=@RigVMFunction_MathBoolMake.Value).Result)
   &RigUnit_ModifyTransforms(
    ItemToModify="((Item=(Type=Bone,Name=\"base_R_Ear_01\"),Transform=(Rotation=(X=0,Y=0,Z=-0.061,W=0.9981),Translation=(X=0,Y=0,Z=0),Scale3D=(X=1,Y=1,Z=1))),(Item=(Type=Bone,Name=\"base_L_Ear_01\"),Transform=(Rotation=(X=0,Y=0,Z=-0.061,W=0.9981),Translation=(X=0,Y=0,Z=0),Scale ...(truncated)"
    WeightMinimum=0
    ExecutePin=@RigVMFunction_Sequence.C
    Weight=@Subtract.Result
   )
   RigVMFunction_MathBoolOr_1 = RigVMAggregateNode(
    B=&RigVMFunction_MathBoolNot(Value=RigVMVariableNode() {variable=isRbanEnabled}.Value).Result
    C=RigVMVariableNode() {variable=isDBNO}.Value
    A=RigVMVariableNode() {variable=IsWearingBackBling}.Value
   )
   Multiply = &RigVMFunction_MathDoubleMul(
    B=&RigVMFunction_MathDoubleSub(A=1, B=&RigVMFunction_MathBoolToFloat(Value=@RigVMFunction_MathBoolOr_1.Result).Result).Result
    A=RigVMVariableNode() {variable=IsLookingUpExtreme}.Value
   )
   > n5 = &RigUnit_ModifyTransforms(
    ItemToModify="((Item=(Type=Bone,Name=\"base_R_Ear_01\"),Transform=(Rotation=(X=0,Y=0,Z=-0.1736,W=0.9848),Translation=(X=0,Y=0,Z=0),Scale3D=(X=1,Y=1,Z=1))),(Item=(Type=Bone,Name=\"base_L_Ear_01\"),Transform=(Rotation=(X=0,Y=0.0872,Z=0,W=0.9962),Translation=(X=0,Y=0,Z=0),Scal ...(truncated)"
    WeightMinimum=0
    Weight=@Multiply.Result
   )
   BoolToFloat = RigVMFunctionReferenceNode(Bool=False)
  ]}
]
hierarchy [
 root {loc=<0 0 0>, rot=<0 0 0 1>}
 pelvis {parent=root, loc=<0 -0.7071 93.1191>, rot=<0.0127 0.707 -0.0127 -0.707>}
 spine_01 {parent=pelvis, loc=<6.7494 -0.6388 0>, rot=<0 0 0.034 -0.9994>}
 spine_02 {parent=spine_01, loc=<7.2518 -0.035 0>, rot=<0 0 0.0238 -0.9997>}
 spine_03 {parent=spine_02, loc=<6.0104 0 0>, rot=<0 0 0.0149 -0.9999>}
 spine_04 {parent=spine_03, loc=<8.1024 0 0>, rot=<0 0 -0.0132 -0.9999>}
 spine_05 {parent=spine_04, loc=<10.0146 0 0>, rot=<0 0 -0.0015 -1>}
 neck_01 {parent=spine_05, loc=<11.3138 0.1529 0>, rot=<0 0 -0.2003 -0.9797>}
 neck_02 {parent=neck_01, loc=<4.5627 0 0>, rot=<0 0 0.0504 -0.9987>}
 head {parent=neck_02, loc=<5.3914 0 0>, rot=<0 0 0.115 -0.9934>}
 base_R_Ear_01 {parent=head, loc=<14.2156 -4.5127 5.3595>, rot=<0.4401 -0.8779 0.1227 0.143>}
 dyn_R_Ear_01 {parent=base_R_Ear_01, loc=<0 0 0>, rot=<0 -0.0789 0 -0.9969>}
 dyn_R_Ear_02 {parent=dyn_R_Ear_01, loc=<3.768 -0.3625 -0.23>, rot=<0.0987 0.1195 -0.2863 -0.9455>}
 dyn_R_Ear_03 {parent=dyn_R_Ear_02, loc=<9.8402 0 0>, rot=<0 0 0 1>}
 dyn_R_Ear_04 {parent=dyn_R_Ear_03, loc=<8.2842 0.098 0.0014>, rot=<0.008 0.0012 -0.0553 -0.9984>}
 dyn_R_Ear_05 {parent=dyn_R_Ear_04, loc=<9.6543 -0.3471 0.2544>, rot=<-0.0027 0.0241 -0.0484 -0.9985>}
 dyn_R_Ear_06 {parent=dyn_R_Ear_05, loc=<8.4757 0 -0.2486>, rot=<0 0.0658 0 -0.9978>}
 dyn_R_Ear_02_ClavAim_null {parent=dyn_R_Ear_01, type=Null, loc=<3.6972 -0.8089 -0.2413>, rot=<0.0966 0.1161 -0.2276 -0.962>}
 dyn_R_Ear_02_SpineAim_null {parent=dyn_R_Ear_01, type=Null, loc=<3.6972 -0.8089 -0.2413>, rot=<0.0966 0.1161 -0.2276 -0.962>}
 base_L_Ear_01 {parent=head, loc=<14.216 -4.5127 -5.3595>, rot=<0.4401 -0.8779 -0.1227 -0.143>}
 dyn_L_Ear_01 {parent=base_L_Ear_01, loc=<0 0 0>, rot=<0 0.0789 0 -0.9969>}
 dyn_L_Ear_02 {parent=dyn_L_Ear_01, loc=<3.7665 -0.3606 0.2275>, rot=<-0.1182 -0.1199 -0.2818 -0.9446>}
 dyn_L_Ear_03 {parent=dyn_L_Ear_02, loc=<9.8399 0 0>, rot=<0 0 0 1>}
 dyn_L_Ear_04 {parent=dyn_L_Ear_03, loc=<8.2854 0.0988 -0.0202>, rot=<0.0035 -0.0112 -0.0558 -0.9984>}
 dyn_L_Ear_05 {parent=dyn_L_Ear_04, loc=<9.6547 -0.3432 -0.2619>, rot=<-0.003 -0.023 -0.0483 -0.9986>}
 dyn_L_Ear_06 {parent=dyn_L_Ear_05, loc=<8.4852 0 0.3101>, rot=<0 -0.0822 0 -0.9966>}
 dyn_L_Ear_02_ClavAim_null {parent=dyn_L_Ear_01, type=Null, loc=<3.7038 -0.7703 0.2375>, rot=<-0.1163 -0.1157 -0.2281 -0.9597>}
 dyn_L_Ear_02_SpineAim_null {parent=dyn_L_Ear_01, type=Null, loc=<3.7038 -0.7703 0.2375>, rot=<-0.1163 -0.1157 -0.2281 -0.9597>}
 head_null {parent=head, type=Null, loc=<0 0 0>, rot=<0 0 0 1>}
 base_hood_c {parent=neck_02, loc=<-1.1595 -9.3247 0>, rot=<-0.0784 -0.7027 0.0784 -0.7027>}
 dyn_hood_c {parent=base_hood_c, loc=<0 0 0>, rot=<0 0 0 1>}
 base_hood_r {parent=neck_02, loc=<-2.5209 -9.1759 1>, rot=<-0.0784 -0.7027 0.0784 -0.7027>}
 dyn_hood_r {parent=base_hood_r, loc=<0 0 0>, rot=<0 0 0 1>}
 base_hood_l {parent=neck_02, loc=<-2.5209 -9.1759 -1>, rot=<-0.0784 -0.7027 0.0784 -0.7027>}
 dyn_hood_l {parent=base_hood_l, loc=<0 0 0>, rot=<0 0 0 1>}
 base_ZipperMain {parent=neck_02, loc=<-3.1051 13.9756 0>, rot=<-0.6593 -0.7519 0 0>}
 dyn_ZipperMain {parent=base_ZipperMain, loc=<0 0 0>, rot=<0 0 0 1>}
 dyn_Zipper {parent=dyn_ZipperMain, loc=<0.4098 -0.2805 0>, rot=<0 0 0.6015 -0.7989>}
 dyn_Zipper_End {parent=dyn_Zipper, loc=<2.6559 0 0>, rot=<0 0 0 1>}
 base_hood_f {parent=neck_02, loc=<-0.7359 4.4155 0>, rot=<0.702 0.0845 0.702 -0.0845>}
 dyn_hood_f {parent=base_hood_f, loc=<0 0 0>, rot=<0 0 0 1>}
 dyn_ZipperMain_aim_null {parent=neck_02, type=Null, loc=<-1.82 14.01 0>, rot=<0.6593 0.7519 0 0>}
 head_Neck01Delta_null {parent=neck_01, type=Null, loc=<9.9266 -0.5431 0>, rot=<0 0 -0.1649 0.9863>}
 clavicle_l {parent=spine_05, loc=<4.3575 -0.1466 -2.3223>, rot=<0.0465 -0.7289 -0.009 -0.6829>}
 upperarm_l {parent=clavicle_l, loc=<15.5111 0 0>, rot=<0.0226 -0.3234 -0.0088 -0.9459>}
 lowerarm_l {parent=upperarm_l, loc=<26.6514 0 0>, rot=<0.0391 -0.0099 -0.2453 -0.9686>}
 hand_l {parent=lowerarm_l, loc=<22.3491 0 0>, rot=<-0.7081 -0.0507 0.0975 -0.6975>}
 weapon_l {parent=hand_l, loc=<-0.1974 -3.4111 0.3811>, rot=<0 0 0 1>}
 thumb_01_l {parent=hand_l, loc=<1.3118 -0.3191 -2.459>, rot=<0.6546 -0.4042 0.0664 -0.6354>}
 thumb_02_l {parent=thumb_01_l, loc=<5.9336 -0.0448 0.0401>, rot=<-0.0037 0.0213 0.298 -0.9543>}
 thumb_03_l {parent=thumb_02_l, loc=<2.2223 0.3682 -0.0907>, rot=<0.0118 -0.013 0.0066 -0.9998>}
 index_metacarpal_l {parent=hand_l, loc=<1.6535 0.0183 -2.96>, rot=<0.0158 -0.0124 -0.0173 -0.9996>}
 index_01_l {parent=index_metacarpal_l, loc=<8.3008 0.0357 -0.0524>, rot=<-0.0044 0.0485 0.0924 -0.9945>}
 index_02_l {parent=index_01_l, loc=<3.3587 -0.3004 0.0452>, rot=<-0.002 0.0027 0.3225 -0.9466>}
 index_03_l {parent=index_02_l, loc=<2.5578 -0.123 -0.0252>, rot=<-0.002 0.0067 0.2558 -0.9667>}
 middle_metacarpal_l {parent=hand_l, loc=<2.1856 -0.1725 -0.5984>, rot=<0.006 0.0168 -0.0329 -0.9993>}
 middle_01_l {parent=middle_metacarpal_l, loc=<7.3275 -0.2447 0.0129>, rot=<-0.0208 0.0042 0.0858 -0.9961>}
 middle_02_l {parent=middle_01_l, loc=<4.0123 0.1401 0.1175>, rot=<-0.0218 0.0258 0.3512 -0.9357>}
 middle_03_l {parent=middle_02_l, loc=<2.6077 0.0534 0.1373>, rot=<0.0082 0.0188 0.1708 -0.9851>}
 ring_metacarpal_l {parent=hand_l, loc=<2.2074 -0.3625 1.3638>, rot=<-0.0284 0.0722 -0.0458 -0.9959>}
 ring_01_l {parent=ring_metacarpal_l, loc=<7.0295 -0.6197 -0.1446>, rot=<-0.0489 -0.0252 0.1892 -0.9804>}
 ring_02_l {parent=ring_01_l, loc=<3.6174 0.4647 -0.1027>, rot=<0 0 0.2806 -0.9598>}
 ring_03_l {parent=ring_02_l, loc=<2.4152 0.114 0.0199>, rot=<0 0 0.1189 -0.9929>}
 pinky_metacarpal_l {parent=hand_l, loc=<1.7616 -0.9753 3.3814>, rot=<-0.0397 0.0814 -0.069 -0.9935>}
 pinky_01_l {parent=pinky_metacarpal_l, loc=<7.1709 -0.4831 -0.1094>, rot=<-0.0654 -0.0035 0.2242 -0.9723>}
 pinky_02_l {parent=pinky_01_l, loc=<2.8274 0.0893 -0.1043>, rot=<0.0037 0.0011 0.2854 -0.9584>}
 pinky_03_l {parent=pinky_02_l, loc=<2.2046 0.0307 -0.0078>, rot=<-0.0002 0.0098 0.1445 -0.9895>}
 hand_attach_l {parent=hand_l, loc=<8.5 -3 0>, rot=<0.7071 0.7071 0 0>}
 deform_wrist_dn_base_l {parent=hand_l, loc=<0 -2.2342 0>, rot=<0 0 0 -1>}
 deform_wrist_dn_l {parent=deform_wrist_dn_base_l, loc=<0 0 0>, rot=<0 0 0 1>}
 deform_wrist_up_base_l {parent=hand_l, loc=<0 2.2682 0>, rot=<0 0 0 -1>}
 deform_wrist_up_l {parent=deform_wrist_up_base_l, loc=<0 0 0>, rot=<0 0 0 1>}
 lowerarm_twist_02_l {parent=lowerarm_l, loc=<9.6199 0 0>, rot=<0 0 0 -1>}
 lowerarm_twist_01_l {parent=lowerarm_l, loc=<17.8549 0 0>, rot=<0 0 0 -1>}
 deform_elbow_in_base_l {parent=lowerarm_l, loc=<0.2112 5.259 0.8274>, rot=<0 0 0 1>}
 deform_elbow_in_l {parent=deform_elbow_in_base_l, loc=<0 0 0>, rot=<0 0 0 1>}
 deform_elbow_out_base_l {parent=lowerarm_l, loc=<0 -4 0>, rot=<0 0 0 1>}
 deform_elbow_out_l {parent=deform_elbow_out_base_l, loc=<0 0 0>, rot=<0 0 0 1>}
 upperarm_twist_01_l {parent=upperarm_l, loc=<0.0041 0 0>, rot=<0 0 0 -1>}
 deform_upperarm_bk_base_l {parent=upperarm_twist_01_l, loc=<1.0813 -7.2193 -1.9453>, rot=<0 0 0 1>}
 deform_upperarm_bk_l {parent=deform_upperarm_bk_base_l, loc=<0 0 0>, rot=<0 0 0 1>}
 deform_upperarm_fr_base_l {parent=upperarm_twist_01_l, loc=<-0.3691 7.0506 -2.765>, rot=<0 0 0 1>}
 deform_upperarm_fr_l {parent=deform_upperarm_fr_base_l, loc=<0 0 0>, rot=<0 0 0 1>}
 deform_upperarm_up_base_l {parent=upperarm_twist_01_l, loc=<-3.2216 -1.1391 3.3311>, rot=<0 0 0 1>}
 deform_upperarm_up_l {parent=deform_upperarm_up_base_l, loc=<0 0 0>, rot=<0 0 0 1>}
 upperarm_twist_02_l {parent=upperarm_l, loc=<13.7941 0 0>, rot=<0 0 0 -1>}
 clavicle_r {parent=spine_05, loc=<4.3575 -0.1466 2.3223>, rot=<0.7289 0.0465 -0.6829 0.009>}
 upperarm_r {parent=clavicle_r, loc=<-15.5111 0 0>, rot=<0.0226 -0.3234 -0.0088 -0.9459>}
 lowerarm_r {parent=upperarm_r, loc=<-26.6514 0 0>, rot=<0.0391 -0.0099 -0.2453 -0.9686>}
 hand_r {parent=lowerarm_r, loc=<-22.3491 0 0>, rot=<-0.7081 -0.0507 0.0975 -0.6975>}
 weapon_r {parent=hand_r, loc=<0.1974 3.4111 -0.3811>, rot=<0 0 0 1>}
 thumb_01_r {parent=hand_r, loc=<-1.3118 0.3191 2.459>, rot=<0.6546 -0.4042 0.0664 -0.6354>}
 thumb_02_r {parent=thumb_01_r, loc=<-5.9336 0.0448 -0.0401>, rot=<-0.0037 0.0213 0.298 -0.9543>}
 thumb_03_r {parent=thumb_02_r, loc=<-2.2223 -0.3682 0.0907>, rot=<0.0118 -0.013 0.0066 -0.9998>}
 index_metacarpal_r {parent=hand_r, loc=<-1.6535 -0.0183 2.96>, rot=<0.0158 -0.0124 -0.0173 -0.9996>}
 index_01_r {parent=index_metacarpal_r, loc=<-8.3008 -0.0357 0.0524>, rot=<-0.0044 0.0485 0.0924 -0.9945>}
 index_02_r {parent=index_01_r, loc=<-3.3587 0.3004 -0.0452>, rot=<-0.002 0.0027 0.3225 -0.9466>}
 index_03_r {parent=index_02_r, loc=<-2.5578 0.123 0.0252>, rot=<-0.002 0.0067 0.2558 -0.9667>}
 middle_metacarpal_r {parent=hand_r, loc=<-2.1856 0.1725 0.5984>, rot=<0.006 0.0168 -0.0329 -0.9993>}
 middle_01_r {parent=middle_metacarpal_r, loc=<-7.3275 0.2447 -0.0129>, rot=<-0.0208 0.0042 0.0858 -0.9961>}
 middle_02_r {parent=middle_01_r, loc=<-4.0123 -0.1401 -0.1175>, rot=<-0.0218 0.0258 0.3512 -0.9357>}
 middle_03_r {parent=middle_02_r, loc=<-2.6077 -0.0534 -0.1373>, rot=<0.0082 0.0188 0.1708 -0.9851>}
 ring_metacarpal_r {parent=hand_r, loc=<-2.2074 0.3625 -1.3638>, rot=<-0.0284 0.0722 -0.0458 -0.9959>}
 ring_01_r {parent=ring_metacarpal_r, loc=<-7.0295 0.6197 0.1446>, rot=<-0.0489 -0.0252 0.1892 -0.9804>}
 ring_02_r {parent=ring_01_r, loc=<-3.6174 -0.4647 0.1027>, rot=<0 0 0.2806 -0.9598>}
 ring_03_r {parent=ring_02_r, loc=<-2.4152 -0.114 -0.0199>, rot=<0 0 0.1189 -0.9929>}
 pinky_metacarpal_r {parent=hand_r, loc=<-1.7616 0.9753 -3.3814>, rot=<-0.0397 0.0814 -0.069 -0.9935>}
 pinky_01_r {parent=pinky_metacarpal_r, loc=<-7.1709 0.4831 0.1094>, rot=<-0.0654 -0.0035 0.2242 -0.9723>}
 pinky_02_r {parent=pinky_01_r, loc=<-2.8274 -0.0893 0.1043>, rot=<0.0037 0.0011 0.2854 -0.9584>}
 pinky_03_r {parent=pinky_02_r, loc=<-2.2046 -0.0307 0.0078>, rot=<-0.0002 0.0098 0.1445 -0.9895>}
 hand_attach_r {parent=hand_r, loc=<-8.5001 2.9999 0>, rot=<0 0 -0.7071 -0.7071>}
 deform_wrist_dn_base_r {parent=hand_r, loc=<-0.0004 2.2345 -0.0001>, rot=<0 0 0 -1>}
 deform_wrist_dn_r {parent=deform_wrist_dn_base_r, loc=<0 0 0>, rot=<0 0 0 1>}
 deform_wrist_up_base_r {parent=hand_r, loc=<0.0003 -2.2684 0.0001>, rot=<0 0 0 -1>}
 deform_wrist_up_r {parent=deform_wrist_up_base_r, loc=<0 0 0>, rot=<0 0 0 1>}
 lowerarm_twist_02_r {parent=lowerarm_r, loc=<-9.6199 0 0>, rot=<0 0 0 -1>}
 lowerarm_twist_01_r {parent=lowerarm_r, loc=<-17.8549 0 0>, rot=<0 0 0 -1>}
 deform_elbow_in_base_r {parent=lowerarm_r, loc=<-0.2111 -5.2591 -0.8275>, rot=<0 0 0 1>}
 deform_elbow_in_r {parent=deform_elbow_in_base_r, loc=<-0.0001 0 0.0002>, rot=<0 0 0 1>}
 deform_elbow_out_base_r {parent=lowerarm_r, loc=<0 4 -0.0001>, rot=<0 0 0 1>}
 deform_elbow_out_r {parent=deform_elbow_out_base_r, loc=<0 0 0>, rot=<0 0 0 1>}
 upperarm_twist_01_r {parent=upperarm_r, loc=<-0.0041 0 0>, rot=<0 0 0 -1>}
 deform_upperarm_bk_base_r {parent=upperarm_twist_01_r, loc=<-1.0814 7.2192 1.9455>, rot=<0 0 0 1>}
 deform_upperarm_bk_r {parent=deform_upperarm_bk_base_r, loc=<0 0 0>, rot=<0 0 0 1>}
 deform_upperarm_fr_base_r {parent=upperarm_twist_01_r, loc=<0.3689 -7.0506 2.7652>, rot=<0 0 0 1>}
 deform_upperarm_fr_r {parent=deform_upperarm_fr_base_r, loc=<0 0 0>, rot=<0 0 0 1>}
 deform_upperarm_up_base_r {parent=upperarm_twist_01_r, loc=<3.2214 1.1391 -3.3308>, rot=<0 0 0 1>}
 deform_upperarm_up_r {parent=deform_upperarm_up_base_r, loc=<0 0 0>, rot=<0 0 0 -1>}
 upperarm_twist_02_r {parent=upperarm_r, loc=<-13.7941 0 0>, rot=<0 0 0 -1>}
 head_Spine05Delta_null {parent=spine_05, type=Null, loc=<20.6573 3.5489 0>, rot=<0 0 -0.036 -0.9994>}
 clavicle_r_SpineDelta_null {parent=spine_05, type=Null, loc=<4.3575 -0.1466 2.3223>, rot=<0.7289 0.0465 -0.6829 0.009>}
 attach_cape {parent=spine_05, loc=<9.017 -11.714 0>, rot=<0.0616 -0.7044 -0.0616 -0.7044>}
 attach_backpack {parent=spine_05, loc=<-0.123 -13.473 0>, rot=<-0.0093 -0.707 0.0093 -0.707>}
 clavicle_l_SpineDelta_null {parent=spine_05, type=Null, loc=<4.3575 -0.1466 -2.3223>, rot=<0.0465 -0.7289 -0.009 -0.6829>}
 deform_pec_aim_base_l {parent=spine_04, loc=<7.8732 12.6388 -2.4123>, rot=<0.0102 -0.4602 0.3979 -0.7936>}
 deform_pec_aim_l {parent=deform_pec_aim_base_l, loc=<0 0 0>, rot=<0 0 0 1>}
 deform_pec_base_l {parent=deform_pec_aim_l, loc=<10.766 0 0>, rot=<0 0 0 1>}
 deform_pec_l {parent=deform_pec_base_l, loc=<0.055 0 0>, rot=<0 0 0 1>}
 deform_pec_aim_base_r {parent=spine_04, loc=<7.8727 12.6388 2.4123>, rot=<-0.4602 -0.0102 0.7936 0.3979>}
 deform_pec_aim_r {parent=deform_pec_aim_base_r, loc=<0 0 0>, rot=<0 0 0 1>}
 deform_pec_base_r {parent=deform_pec_aim_r, loc=<-10.7662 -0.0006 -0.0007>, rot=<0 0 0 1>}
 deform_pec_r {parent=deform_pec_base_r, loc=<0 0 0>, rot=<0 0 0 1>}
 thigh_l {parent=pelvis, loc=<-2.5693 -0.0109 -7.7501>, rot=<0.0012 0.0572 0.0155 -0.9982>}
 calf_l {parent=thigh_l, loc=<-39.0484 0 0>, rot=<0.0006 -0.0161 -0.0391 -0.9991>}
 foot_l {parent=calf_l, loc=<-42.561 0.1077 0.4292>, rot=<0.0418 -0.0455 0.0343 -0.9975>}
 ball_l {parent=foot_l, loc=<-6.1599 13.8264 0.6806>, rot=<0.0001 -0.0001 -0.7063 -0.7079>}
 calf_twist_02_l {parent=calf_l, loc=<-18.8064 -0.0967 0.0397>, rot=<0.0237 0.0095 -0.0035 -0.9997>}
 calf_twist_01_l {parent=calf_l, loc=<-35.813 -0.0967 0.0397>, rot=<0.0237 0.0095 -0.0035 -0.9997>}
 deform_knee_in_base_l {parent=calf_l, loc=<1.298 -5.2957 -0.0496>, rot=<0 0 0 -1>}
 deform_knee_in_l {parent=deform_knee_in_base_l, loc=<0 0 0>, rot=<0 0 0 1>}
 deform_knee_out_base_l {parent=calf_l, loc=<0 4 0>, rot=<0 0 0 -1>}
 deform_knee_out_l {parent=deform_knee_out_base_l, loc=<0 0 0>, rot=<0 0 0 1>}
 thigh_twist_01_l {parent=thigh_l, loc=<-21.6303 0 0>, rot=<0.0013 0 0 -1>}
 deform_groin_base_l {parent=thigh_l, loc=<-2.1908 8.1142 -1.7442>, rot=<0 0 0 -1>}
 deform_groin_l {parent=deform_groin_base_l, loc=<0 0 0>, rot=<0 0 0 1>}
 thigh_r {parent=pelvis, loc=<-2.5693 -0.0109 7.7501>, rot=<0.0572 -0.0012 0.9982 0.0155>}
 calf_r {parent=thigh_r, loc=<39.0484 0 0>, rot=<0.0006 -0.0161 -0.0391 -0.9991>}
 foot_r {parent=calf_r, loc=<42.561 -0.1077 -0.4292>, rot=<0.0418 -0.0455 0.0343 -0.9975>}
 ball_r {parent=foot_r, loc=<6.1599 -13.8264 -0.6806>, rot=<0.0001 -0.0001 -0.7063 -0.7079>}
 calf_twist_02_r {parent=calf_r, loc=<18.8064 0.0967 -0.0397>, rot=<0.0237 0.0095 -0.0035 -0.9997>}
 calf_twist_01_r {parent=calf_r, loc=<35.813 0.0967 -0.0397>, rot=<0.0237 0.0095 -0.0035 -0.9997>}
 deform_knee_in_base_r {parent=calf_r, loc=<-1.298 5.2957 0.0496>, rot=<0 0 0 1>}
 deform_knee_in_r {parent=deform_knee_in_base_r, loc=<0 0 0>, rot=<0 0 0 1>}
 deform_knee_out_base_r {parent=calf_r, loc=<0 -4 0>, rot=<0 0 0 -1>}
 deform_knee_out_r {parent=deform_knee_out_base_r, loc=<0 0 0>, rot=<0 0 0 1>}
 thigh_twist_01_r {parent=thigh_r, loc=<21.6303 0 0>, rot=<0.0013 0 0 -1>}
 deform_groin_base_r {parent=thigh_r, loc=<2.1908 -8.1142 1.7442>, rot=<0 0 0 -1>}
 deform_groin_r {parent=deform_groin_base_r, loc=<0 0 0>, rot=<0 0 0 1>}
 deform_glute_base_l {parent=pelvis, loc=<-8.8064 -8.8222 -8.4287>, rot=<0.0275 0.0501 -0.4472 -0.8926>}
 deform_glute_l {parent=deform_glute_base_l, loc=<0 0 0>, rot=<0 0 0 1>}
 deform_glute_base_r {parent=pelvis, loc=<-8.8062 -8.8222 8.4287>, rot=<0.0501 -0.0275 0.8926 -0.4472>}
 deform_glute_r {parent=deform_glute_base_r, loc=<0 0 0>, rot=<0 0 0 1>}
 base_C_Belt_01 {parent=pelvis, loc=<-0.8896 12.6648 0>, rot=<-0.0179 -0.9998 0 0.0015>}
 dyn_C_Belt_01 {parent=base_C_Belt_01, loc=<0 0 0>, rot=<0 0 0 1>}
 base_BeltStrap_01 {parent=dyn_C_Belt_01, loc=<-0.5634 0.5542 -3.8583>, rot=<0.6098 -0.3464 -0.3605 -0.615>}
 dyn_BeltStrap_01 {parent=base_BeltStrap_01, loc=<0 0 0>, rot=<0 0 0 1>}
 dyn_BeltStrap_02 {parent=dyn_BeltStrap_01, loc=<3.5657 0 0>, rot=<0.0046 -0.0039 0.2733 -0.9619>}
 dyn_BeltStrap_03 {parent=dyn_BeltStrap_02, loc=<4.0477 0.1207 -0.0198>, rot=<0.0025 0.0162 0.1295 -0.9914>}
 dyn_BeltStrap_04 {parent=dyn_BeltStrap_03, loc=<5.4974 0 0>, rot=<0 0 0 1>}
 dyn_BeltStrap_End {parent=dyn_BeltStrap_04, loc=<4.5462 0 0>, rot=<0 0 0 1>}
 base_L_Belt_01 {parent=pelvis, loc=<-0.3426 9.7155 -9.9567>, rot=<-0.0168 -0.904 -0.4271 -0.0063>}
 dyn_L_Belt_01 {parent=base_L_Belt_01, loc=<0 0 0>, rot=<0 0 0 1>}
 base_L_BulletsMain {parent=dyn_L_Belt_01, loc=<-0.6967 -2.691 -0.1902>, rot=<-0.0361 0.0173 0.0694 -0.9968>}
 dyn_L_BulletsMain {parent=base_L_BulletsMain, loc=<0 0 0>, rot=<0 0 0 1>}
 dyn_L_Bullet_01 {parent=dyn_L_BulletsMain, loc=<-0.0613 2.8452 0.2894>, rot=<0.1138 0.0654 -0.0215 -0.9911>}
 dyn_L_Bullet_01_end {parent=dyn_L_Bullet_01, loc=<2.808 0 0>, rot=<0 0 0 1>}
 dyn_L_Bullet_02 {parent=dyn_L_BulletsMain, loc=<-0.0105 -0.1855 0.5709>, rot=<-0.0175 0.0482 -0.0079 -0.9987>}
 dyn_L_Bullet_02_end {parent=dyn_L_Bullet_02, loc=<2.9203 0 0>, rot=<0 0 0 1>}
 dyn_L_Bullet_03 {parent=dyn_L_BulletsMain, loc=<0.0107 -3.377 0.357>, rot=<-0.0892 0.0442 0.0039 -0.995>}
 dyn_L_Bullet_03_end {parent=dyn_L_Bullet_03, loc=<2.8474 0 0>, rot=<0 0 0 1>}
 base_L_Belt_02 {parent=pelvis, loc=<1.5206 0.1373 -16.2979>, rot=<-0.0179 -0.9998 0.0014 0.0015>}
 dyn_L_Belt_02 {parent=base_L_Belt_02, loc=<0 0 0>, rot=<0 0 0 1>}
 base_L_Belt_03 {parent=pelvis, loc=<3.6842 -10.4495 -9.8847>, rot=<-0.0143 -0.8457 0.5333 0.0108>}
 dyn_L_Belt_03 {parent=base_L_Belt_03, loc=<0 0 0>, rot=<0 0 0 1>}
 base_B_Belt_01 {parent=pelvis, loc=<4.0242 -12.197 0>, rot=<-0.0118 -0.7162 0.6977 0.0136>}
 dyn_B_Belt_01 {parent=base_B_Belt_01, loc=<0 0 0>, rot=<0 0 0 1>}
 base_R_Belt_03 {parent=pelvis, loc=<3.6842 -10.4495 9.8847>, rot=<-0.0073 -0.484 0.8749 0.0164>}
 dyn_R_Belt_03 {parent=base_R_Belt_03, loc=<0 0 0>, rot=<0 0 0 1>}
 base_R_Belt_02 {parent=pelvis, loc=<1.5206 0.1373 16.2979>, rot=<0.0014 -0.0047 0.9998 0.0179>}
 dyn_R_Belt_02 {parent=base_R_Belt_02, loc=<0 0 0>, rot=<0 0 0 1>}
 base_R_Belt_01 {parent=pelvis, loc=<-0.3426 9.7155 9.9567>, rot=<0.0102 0.4942 0.8691 0.0148>}
 dyn_R_Belt_01 {parent=base_R_Belt_01, loc=<0 0 0>, rot=<0 0 0 1>}
 base_R_Granade01 {parent=pelvis, loc=<0.584 6.5194 14.3409>, rot=<0.3711 0.5843 -0.1098 -0.7133>}
 dyn_R_Granade01 {parent=base_R_Granade01, loc=<0 0 0>, rot=<0 0 0 1>}
 Dyn_R_Granade01_End {parent=dyn_R_Granade01, loc=<10.7619 0 0>, rot=<0 0 0 1>}
 base_R_Granade01_Ring {parent=dyn_R_Granade01, loc=<4.0875 0.5349 -5.7292>, rot=<-0.3947 0.3287 -0.546 -0.6618>}
 dyn_R_Granade01_Ring {parent=base_R_Granade01_Ring, loc=<0 0 0>, rot=<0 0 0 1>}
 dyn_R_GranadeRing01_End {parent=dyn_R_Granade01_Ring, loc=<2.4721 0 0>, rot=<0 0 0 1>}
 base_R_Granade02 {parent=pelvis, loc=<1.3527 -0.1475 16.9706>, rot=<0.0017 0.6612 0.0462 -0.7488>}
 dyn_R_Granade02 {parent=base_R_Granade02, loc=<0 0 0>, rot=<0 0 0 1>}
 dyn_R_Granade02_End {parent=dyn_R_Granade02, loc=<11.1254 0 0>, rot=<0 0 0 1>}
 base_R_Granade02_Ring {parent=dyn_R_Granade02, loc=<3.5257 -0.4741 -5.3824>, rot=<0.5337 -0.1032 0.7653 0.3448>}
 dyn_R_Granade02_Ring {parent=base_R_Granade02_Ring, loc=<0 0 0>, rot=<0 0 0 1>}
 dyn_R_Granade02_Ring_End {parent=dyn_R_Granade02_Ring, loc=<2.4445 0 0>, rot=<0 0 0 1>}
 base_Tail {parent=pelvis, loc=<-2.0049 -12.1154 -0.0087>, rot=<0.8771 -0.4803 0 0>}
 dyn_Tail {parent=base_Tail, loc=<0 0 0>, rot=<0 0 0 1>}
 Tail_End {parent=dyn_Tail, loc=<10.806 0 0>, rot=<0 0 0 1>}
 attach {parent=root, loc=<0 0 0>, rot=<0 0 0 1>}
 ik_hand_root {parent=root, loc=<0 0 0>, rot=<0 0 0 1>}
 ik_hand_gun {parent=ik_hand_root, loc=<-51.9727 4.0883 102.6935>, rot=<0.6084 0.1793 0.4613 -0.6205>}
 ik_hand_r {parent=ik_hand_gun, loc=<0 0 0>, rot=<0 0 0 1>}
 ik_hand_l {parent=ik_hand_gun, loc=<53.0285 82.1781 35.2051>, rot=<-0.5102 -0.7906 -0.3387 0>}
 ik_foot_root {parent=root, loc=<0 0 0>, rot=<0 0 0 1>}
 ik_foot_l {parent=ik_foot_root, loc=<15.2608 -4.4176 9.4432>, rot=<0.0357 0.7045 0.0284 -0.7082>}
 ik_foot_r {parent=ik_foot_root, loc=<-15.2608 -4.4176 9.4432>, rot=<0.7082 0.0284 -0.7045 0.0357>}
 root_fp {parent=root, loc=<0 0 0>, rot=<0 0 0 1>}
 attach_fp {parent=root_fp, loc=<0 0 0>, rot=<0 0 0 1>}
 camera_root {parent=root_fp, loc=<0 0 0>, rot=<0 0 0 1>}
 camera {parent=camera_root, loc=<0 0 0>, rot=<0 0 0 1>}
 attach_emote_01 {parent=root, loc=<0 0 0>, rot=<0 0 0 1>}
 attach_emote_02 {parent=root, loc=<0 0 0>, rot=<0 0 0 1>}
 attach_emote_03 {parent=root, loc=<0 0 0>, rot=<0 0 0 1>}
 warp_point_01 {parent=root, loc=<0 0 0>, rot=<0 0 0 1>}
 warp_point_02 {parent=root, loc=<0 0 0>, rot=<0 0 0 1>}
 warp_point_03 {parent=root, loc=<0 0 0>, rot=<0 0 0 1>}
 spine_04_ctrl {controlType=EULER_TRANSFORM, type=Control, loc=<0 0 0>, rot=<0 0 0 1>}
 spine_05_ctrl {controlType=EULER_TRANSFORM, parent=spine_04_ctrl, type=Control, loc=<0 0 0>, rot=<0 0 0 1>}
 neck_01_ctrl {controlType=EULER_TRANSFORM, parent=spine_05_ctrl, type=Control, loc=<0 0 0>, rot=<0 0 0 1>}
 neck_02_ctrl {controlType=EULER_TRANSFORM, parent=neck_01_ctrl, type=Control, loc=<0 0 0>, rot=<0 0 0 1>}
 head_ctrl {controlType=EULER_TRANSFORM, parent=neck_02_ctrl, type=Control, loc=<0 0 0>, rot=<0 0 0 1>}
 base_hood_c_ctrl {controlType=EULER_TRANSFORM, parent=head_ctrl, type=Control, loc=<0 0 0>, rot=<0 0 0 1>}
 base_hood_r_ctrl {controlType=EULER_TRANSFORM, parent=head_ctrl, type=Control, loc=<0 0 0>, rot=<0 0 0 1>}
 base_hood_l_ctrl {controlType=EULER_TRANSFORM, parent=head_ctrl, type=Control, loc=<0 0 0>, rot=<0 0 0 1>}
 R_ArmAim_Null {type=Null, loc=<-12.7 -14 136.2>, rot=<0 0 0 1>}
 R_SpineAim_null {type=Null, loc=<-6.5 -16 135>, rot=<0 0 0 1>}
 dyn_R_Ear_02_c_jnt {loc=<-7.4767 -8.9531 164.4936>, rot=<-0.0566 0.7607 -0.1586 0.6268>}
 dyn_R_Ear_02_Ik_ctrl {controlType=EULER_TRANSFORM, parent=dyn_R_Ear_02_c_jnt, type=Control, loc=<149.4169 66.5579 11.9783>, rot=<0.1165 -0.709 0.1353 0.6823>}
 dyn_R_Ear_03_c_jnt {loc=<-9.5209 -11.7572 155.2856>, rot=<-0.0566 0.7607 -0.1586 0.6268>}
 dyn_R_Ear_03_Ik_ctrl {controlType=EULER_TRANSFORM, parent=dyn_R_Ear_03_c_jnt, type=Control, loc=<-17.9819 -11.7572 155.2856>, rot=<-0.0566 0.7607 -0.1586 0.6268>}
 dyn_R_Ear_04_c_jnt {loc=<-11.2295 -14.0256 147.5027>, rot=<0.0196 -0.7632 0.1176 -0.6351>}
 dyn_R_Ear_04_Ik_ctrl {controlType=EULER_TRANSFORM, parent=dyn_R_Ear_04_c_jnt, type=Control, loc=<-19.8382 -14.2595 147.3229>, rot=<0.0276 0.7648 -0.1464 0.6268>}
 dyn_R_Ear_05_c_jnt {loc=<-12.8816 -16.1332 138.2173>, rot=<0.0162 0.7474 -0.0883 0.6583>}
 dyn_R_Ear_05_Ik_ctrl {controlType=EULER_TRANSFORM, parent=dyn_R_Ear_05_c_jnt, type=Control, loc=<-39.8914 -12.0873 134.3203>, rot=<0.0815 0.7062 -0.1346 0.6903>}
 dyn_R_Ear_06_c_jnt {loc=<-14.2504 -16.8746 129.8821>, rot=<-0.0104 -0.7024 0.0891 -0.7061>}
 dyn_R_Ear_06_Ik_ctrl {controlType=EULER_TRANSFORM, parent=dyn_R_Ear_06_c_jnt, type=Control, loc=<-8.2157 -16.8746 132.1738>, rot=<0.0045 0.7474 -0.0896 0.6583>}
 dyn_L_Ear_02_c_jnt {loc=<7.4736 -8.9521 164.4927>, rot=<-0.1454 0.6212 -0.069 0.767>}
 dyn_L_Ear_02_Ik_ctrl {controlType=EULER_TRANSFORM, parent=dyn_L_Ear_02_c_jnt, type=Control, loc=<149.1597 66.2867 -16.0126>, rot=<-0.1227 0.6763 -0.1297 -0.7147>}
 dyn_L_Ear_03_c_jnt {loc=<9.6269 -11.7708 155.3144>, rot=<-0.1454 0.6212 -0.069 0.767>}
 dyn_L_Ear_03_Ik_ctrl {controlType=EULER_TRANSFORM, parent=dyn_L_Ear_03_c_jnt, type=Control, loc=<-8.4606 0 0>, rot=<0 0 0 -1>}
 dyn_L_Ear_04_c_jnt {loc=<11.4129 -14.0533 147.5519>, rot=<0.1125 -0.6371 0.0255 -0.7621>}
 dyn_L_Ear_04_Ik_ctrl {controlType=EULER_TRANSFORM, parent=dyn_L_Ear_04_c_jnt, type=Control, loc=<-7.3555 0.825 -0.1625>, rot=<0.0035 -0.0112 -0.0558 0.9984>}
 dyn_L_Ear_05_c_jnt {loc=<12.9981 -16.1831 138.2593>, rot=<-0.0786 0.6591 0.0068 0.7479>}
 dyn_L_Ear_05_Ik_ctrl {controlType=EULER_TRANSFORM, parent=dyn_L_Ear_05_c_jnt, type=Control, loc=<-9.2522 1.8262 1.3372>, rot=<-0.0099 0.0573 -0.0767 0.9954>}
 dyn_L_Ear_06_c_jnt {loc=<14.4164 -16.9367 129.9217>, rot=<0.0789 -0.7183 -0.0003 -0.6912>}
 dyn_L_Ear_06_Ik_ctrl {controlType=EULER_TRANSFORM, parent=dyn_L_Ear_06_c_jnt, type=Control, loc=<-11.0937 0 -1.9269>, rot=<0 -0.0822 0 0.9966>}
]
variables [
 IsAimingDownSights {type=bool, default=False}
 IsCrouchWalking {type=float, default=0}
 IsCrouching {type=float, default=0}
 IsHeadLookingDown {type=double, default=0}
 IsHeadLookingUp {type=double, default=0}
 IsJumping {type=double, default=0}
 IsLookingLeft {type=double, default=0}
 IsLookingRight {type=double, default=0}
 IsLookingUpExtreme {type=float, default=0}
 IsPistolOut {type=bool, default=False}
 IsSwingingPickaxe {type=bool, default=False}
 IsWearingBackBling {type=bool, default=False}
 SkydiveDiveMode {type=double, default=0}
 bIsCrouching {type=bool, default=False}
 bIsSMGOut {type=bool, default=False}
 bisPickaxeOut {type=bool, default=False}
 isDBNO {type=bool, default=False}
 isRbanEnabled {type=bool, default=False}
 isSkyDiving {type=bool, default=False}
 isTacticalJumpingF {type=double, default=0}
 isTargetingF {type=double, default=0}
]
