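# Physics asset for the HappyHopper male medium character's EarsB face accessory, providing jointed collision bodies for the two long ears.
# Each ear is a capsule chain from a base bone on the head through six dyn bones, with paired capsules on middle segments for width.
# Ear swing limits are tight near the root and widen toward the tip, with little or no twist.
# Head, neck, upper spine, clavicle and upper-arm capsules form a freely constrained body volume.
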PhysicsAsset /BRCosmetics/Characters/Player/Male/Medium/Bodies/M_MED_HappyHopper/Meshes/Parts/M_MED_HappyHopper_FaceAcc_EarsB_Physics
bodies {
 base_L_Ear_01 [{shapeType=Capsule, length=6, radius=3, shapeName=base_L_Ear_01_capsule, center=<1 0 0>, rotation=<rot 90 0 0>}]
 base_R_Ear_01 [{shapeType=Capsule, length=6, radius=3, shapeName=base_R_Ear_01_capsule, center=<1 0 0>, rotation=<rot 90 0 0>}]
 clavicle_l [{
   shapeType=Capsule
   length=11.961
   radius=9.5
   shapeName=clavicle_l_capsule
   center=<7.9302 -1.4349 -2.8671>
   rotation=<rot -11.9705 -81.4446 -78.6097>
  }]
 clavicle_r [{
   shapeType=Capsule
   length=8.219
   radius=9.6419
   shapeName=clavicle_l_capsule
   center=<-9.3137 0.862 3.0899>
   rotation=<rot -11.9705 -81.4446 101.3903>
  }]
 dyn_L_Ear_01 [{shapeType=Capsule, length=6, radius=3, shapeName=dyn_L_Ear_01_capsule, center=<1 0 0>, rotation=<rot 90 0 0>}]
 dyn_L_Ear_02 [
  {
   shapeType=Capsule
   length=7.6677
   radius=1.9569
   shapeName=dyn_L_Ear_02_capsule
   center=<3.9796 -1.4935 -2.0215>
   rotation=<rot 69.2816 75.4899 75.2914>
  }
  {
   shapeType=Capsule
   length=7.8808
   radius=1.9569
   shapeName=dyn_L_Ear_02_capsule
   center=<4.2453 0.3174 1.2685>
   rotation=<rot 69.019 106.8282 106.6943>
  }
 ]
 dyn_L_Ear_03 [
  {
   shapeType=Capsule
   length=7.1623
   radius=2.2407
   shapeName=dyn_L_Ear_03_capsule
   center=<2.5001 -1.5103 -2.689>
   rotation=<rot 75.9395 61.1413 59.8502>
  }
  {
   shapeType=Capsule
   length=7.4031
   radius=2
   shapeName=dyn_L_Ear_03_capsule
   center=<2.5752 0.594 1.7526>
   rotation=<rot 58.5787 101.9132 96.2578>
  }
 ]
 dyn_L_Ear_04 [
  {
   shapeType=Capsule
   length=7.7961
   radius=2.7644
   shapeName=dyn_L_Ear_04_capsule
   center=<4.0617 0.2008 3.0266>
   rotation=<rot 71.2512 112.7529 111.4461>
  }
  {
   shapeType=Capsule
   length=7.7755
   radius=2.785
   shapeName=dyn_L_Ear_04_capsule
   center=<4.0651 -1.8852 -3.0933>
   rotation=<rot 75.5922 77.6122 75.7646>
  }
 ]
 dyn_L_Ear_05 [
  {
   shapeType=Capsule
   length=8.2596
   radius=2.6138
   shapeName=dyn_L_Ear_05_capsule
   center=<3.2336 -0.1277 4.1644>
   rotation=<rot 76.7673 72.9527 74.5231>
  }
  {
   shapeType=Capsule
   length=8.2596
   radius=2.6138
   shapeName=dyn_L_Ear_05_capsule
   center=<3.0947 -1.2846 -2.9028>
   rotation=<rot 80.4376 122.6625 122.7428>
  }
 ]
 dyn_L_Ear_06 [{shapeType=Capsule, length=13.425, radius=3, shapeName=dyn_L_Ear_06_capsule, center=<3.7096 -0.247 0>, rotation=<rot 90 179.7016 177.7755>}]
 dyn_R_Ear_01 [{shapeType=Capsule, length=6, radius=3, shapeName=dyn_R_Ear_01_capsule, center=<1 0 0>, rotation=<rot 90 0 0>}]
 dyn_R_Ear_02 [
  {
   shapeType=Capsule
   length=7.0718
   radius=1.9569
   shapeName=dyn_R_Ear_02_capsule
   center=<4.4088 0.2941 -1.4685>
   rotation=<rot 60.9718 -75.5508 -76.9513>
  }
  {
   shapeType=Capsule
   length=7.0718
   radius=1.9569
   shapeName=dyn_R_Ear_02_capsule
   center=<4.5271 -1.3392 1.9397>
   rotation=<rot 60.6318 -109.5263 -108.0398>
  }
 ]
 dyn_R_Ear_03 [
  {
   shapeType=Capsule
   length=6.2079
   radius=2.2814
   shapeName=dyn_R_Ear_03_capsule
   center=<2.6087 -1.5937 2.6795>
   rotation=<rot 74.9302 -112.7397 -114.8816>
  }
  {
   shapeType=Capsule
   length=6.1252
   radius=2.3641
   shapeName=dyn_R_Ear_03_capsule
   center=<2.5866 0.2489 -2.1162>
   rotation=<rot 58.0848 -75.9245 -77.6567>
  }
 ]
 dyn_R_Ear_04 [
  {
   shapeType=Capsule
   length=8.2429
   radius=2.7054
   shapeName=dyn_R_Ear_04_capsule
   center=<3.9312 -0.0996 -3.1065>
   rotation=<rot 66.579 -70.4203 -69.4361>
  }
  {
   shapeType=Capsule
   length=8.0715
   radius=2.7465
   shapeName=dyn_R_Ear_04_capsule
   center=<4.0666 -1.9056 3.0207>
   rotation=<rot 74.6183 -100.2512 -99.879>
  }
 ]
 dyn_R_Ear_05 [
  {
   shapeType=Capsule
   length=8.5999
   radius=2.8976
   shapeName=dyn_R_Ear_05_capsule
   center=<2.5019 -0.4738 -4.8864>
   rotation=<rot 63.0032 -93.2142 -90.1051>
  }
  {
   shapeType=Capsule
   length=8.5703
   radius=2.9273
   shapeName=dyn_R_Ear_05_capsule
   center=<2.584 -1.609 2.8199>
   rotation=<rot 80.8316 -72.1448 -72.4335>
  }
 ]
 dyn_R_Ear_06 [{shapeType=Capsule, length=12.5344, radius=3, shapeName=dyn_R_Ear_06_capsule, center=<3.7096 -0.247 0>, rotation=<rot 90 180 178.0739>}]
 head [{
   shapeType=Capsule
   length=5.9853
   radius=7.7296
   shapeName=head_capsule
   center=<7.1364 0.2045 -0.2375>
   rotation=<rot 0.7313 -165.089 -88.9647>
  }]
 neck_01 [{shapeType=Capsule, length=14.6138, radius=5.0217, shapeName=neck_01_capsule, center=<-0.6214 -4.2834 0>}]
 neck_02 [
  {
   shapeType=Capsule
   length=6.8068
   radius=8.4854
   shapeName=neck_02_capsule
   center=<0.5827 -2.9901 -0.016>
   rotation=<rot 0 -166.7272 0.5597>
  }
  {
   shapeType=Capsule
   length=1.9141
   radius=9.1208
   shapeName=neck_02_capsule
   center=<5.0974 -2.9289 -0.0255>
   rotation=<rot 0 -166.7274 -89.441>
  }
 ]
 spine_03 [{
   shapeType=Capsule
   length=15.6083
   radius=9.8741
   shapeName=spine_03_capsule
   center=<7.1721 -3.8873 -0.0038>
   rotation=<rot -0.0021 -90 -0.0206>
  }]
 spine_04 [{
   shapeType=Capsule
   length=23.8485
   radius=9.6745
   shapeName=spine_04_capsule
   center=<6.8537 -3.1083 0.1303>
   rotation=<rot 0.58 -90 -0.4676>
  }]
 spine_05 [{shapeType=Capsule, length=21.6256, radius=11.339, shapeName=spine_05_capsule, center=<0.8228 -0.2728 0.1053>}]
 upperarm_l [{
   shapeType=Capsule
   length=18.4081
   radius=8.5062
   shapeName=upperarm_l_capsule
   center=<10.9967 0.151 -1.9194>
   rotation=<rot 0.0781 -88.3108 87.3194>
  }]
 upperarm_r [{
   shapeType=Capsule
   length=16.9613
   radius=8.3661
   shapeName=upperarm_l_capsule
   center=<-10.9906 -0.3072 1.9197>
   rotation=<rot 0.0982 -87.8912 -92.6799>
  }]
}
constraints [
 {
  bone1Name=head
  bone2Name=neck_02
  swing1LimitDegrees=45
  swing1Motion=Free
  swing2LimitDegrees=45
  swing2Motion=Free
  twistLimitDegrees=45
  twistMotion=Free
 }
 {
  bone1Name=neck_02
  bone2Name=neck_01
  swing1LimitDegrees=45
  swing1Motion=Free
  swing2LimitDegrees=45
  swing2Motion=Free
  twistLimitDegrees=45
  twistMotion=Free
 }
 {
  bone1Name=neck_01
  bone2Name=spine_05
  swing1LimitDegrees=45
  swing1Motion=Free
  swing2LimitDegrees=45
  swing2Motion=Free
  twistLimitDegrees=45
  twistMotion=Free
 }
 {
  bone1Name=clavicle_l
  bone2Name=spine_05
  swing1LimitDegrees=45
  swing1Motion=Free
  swing2LimitDegrees=45
  swing2Motion=Free
  twistLimitDegrees=45
  twistMotion=Free
 }
 {
  bone1Name=clavicle_r
  bone2Name=spine_05
  swing1LimitDegrees=45
  swing1Motion=Free
  swing2LimitDegrees=45
  swing2Motion=Free
  twistLimitDegrees=45
  twistMotion=Free
 }
 {
  bone1Name=spine_05
  bone2Name=spine_04
  swing1LimitDegrees=45
  swing1Motion=Free
  swing2LimitDegrees=45
  swing2Motion=Free
  twistLimitDegrees=45
  twistMotion=Free
 }
 {
  bone1Name=spine_04
  bone2Name=spine_03
  swing1LimitDegrees=45
  swing1Motion=Free
  swing2LimitDegrees=45
  swing2Motion=Free
  twistLimitDegrees=45
  twistMotion=Free
 }
 {
  bone1Name=upperarm_l
  bone2Name=clavicle_l
  swing1LimitDegrees=45
  swing1Motion=Free
  swing2LimitDegrees=45
  swing2Motion=Free
  twistLimitDegrees=45
  twistMotion=Free
 }
 {
  bone1Name=upperarm_r
  bone2Name=clavicle_r
  swing1LimitDegrees=45
  swing1Motion=Free
  swing2LimitDegrees=45
  swing2Motion=Free
  twistLimitDegrees=45
  twistMotion=Free
 }
 {bone1Name=base_L_Ear_01, bone2Name=head, swing1LimitDegrees=45, swing2LimitDegrees=45, twistLimitDegrees=0}
 {bone1Name=dyn_L_Ear_01, bone2Name=base_L_Ear_01, swing1LimitDegrees=45, swing2LimitDegrees=45, twistLimitDegrees=0}
 {bone1Name=dyn_L_Ear_02, bone2Name=dyn_L_Ear_01, swing1LimitDegrees=20, swing2LimitDegrees=5, twistLimitDegrees=0}
 {bone1Name=dyn_L_Ear_03, bone2Name=dyn_L_Ear_02, swing1LimitDegrees=25, swing2LimitDegrees=5, twistLimitDegrees=0}
 {bone1Name=dyn_L_Ear_04, bone2Name=dyn_L_Ear_03, swing1LimitDegrees=30, swing2LimitDegrees=7.5, twistLimitDegrees=5}
 {bone1Name=dyn_L_Ear_05, bone2Name=dyn_L_Ear_04, swing1LimitDegrees=35, swing2LimitDegrees=15, twistLimitDegrees=0}
 {bone1Name=dyn_L_Ear_06, bone2Name=dyn_L_Ear_05, swing1LimitDegrees=40, swing2LimitDegrees=15, twistLimitDegrees=0}
 {bone1Name=base_R_Ear_01, bone2Name=head, swing1LimitDegrees=45, swing2LimitDegrees=45, twistLimitDegrees=0}
 {bone1Name=dyn_R_Ear_01, bone2Name=base_R_Ear_01, swing1LimitDegrees=45, swing2LimitDegrees=45, twistLimitDegrees=0}
 {bone1Name=dyn_R_Ear_02, bone2Name=dyn_R_Ear_01, swing1LimitDegrees=20, swing2LimitDegrees=5, twistLimitDegrees=0}
 {bone1Name=dyn_R_Ear_03, bone2Name=dyn_R_Ear_02, swing1LimitDegrees=25, swing2LimitDegrees=5, twistLimitDegrees=0}
 {bone1Name=dyn_R_Ear_04, bone2Name=dyn_R_Ear_03, swing1LimitDegrees=30, swing2LimitDegrees=7.5, twistLimitDegrees=5}
 {bone1Name=dyn_R_Ear_05, bone2Name=dyn_R_Ear_04, swing1LimitDegrees=35, swing2LimitDegrees=15, twistLimitDegrees=0}
 {bone1Name=dyn_R_Ear_06, bone2Name=dyn_R_Ear_05, swing1LimitDegrees=40, swing2LimitDegrees=15, twistLimitDegrees=0}
]
