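# Physics asset for the HappyHopper male cosmetic's ear accessory, apparently a crouch variant, defining bodies for two floppy ear chains and the upper body.
# Each ear chains from the head through a base and six dyn bones, using paired capsules on outer segments to approximate its width.
# Ear swing limits tighten after the root, then widen toward the tip, with the right ear looser.
# Capsules cover spine, neck, head and arms, linked by free-motion constraints.
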PhysicsAsset /BRCosmetics/Characters/Player/Male/Medium/Bodies/M_MED_HappyHopper/Meshes/Parts/M_MED_HappyHopper_FaceAcc_Ears_Crouch_Physics
bodies {
 base_L_Ear_01 [{shapeType=Capsule, length=6, radius=3, shapeName=base_L_Ear_01_capsule, center=<1 0 0>, rotation=<rot 90 0 0>}]
 base_R_Ear_01 [{shapeType=Capsule, length=6, radius=3, shapeName=base_R_Ear_01_capsule, center=<1 0 0>, rotation=<rot 90 0 0>}]
 clavicle_l [{
   shapeType=Capsule
   length=10.2686
   radius=9.5
   shapeName=clavicle_l_capsule
   center=<7.7503 1.4863 -2.8053>
   rotation=<rot -11.9705 -81.4446 -78.6097>
  }]
 clavicle_r [{
   shapeType=Capsule
   length=8.219
   radius=9.6419
   shapeName=clavicle_l_capsule
   center=<-9.1833 -1.7299 3.0987>
   rotation=<rot -11.9705 -81.4446 101.3903>
  }]
 dyn_L_Ear_01 [{shapeType=Capsule, length=6, radius=3, shapeName=dyn_L_Ear_01_capsule, center=<1 0 0>, rotation=<rot 90 0 0>}]
 dyn_L_Ear_02 [
  {
   shapeType=Capsule
   length=7.6677
   radius=1.9569
   shapeName=dyn_L_Ear_02_capsule
   center=<3.9796 -1.4935 -2.0215>
   rotation=<rot 69.2816 75.4899 75.2914>
  }
  {
   shapeType=Capsule
   length=7.8808
   radius=1.9569
   shapeName=dyn_L_Ear_02_capsule
   center=<4.2453 0.3174 1.2685>
   rotation=<rot 69.019 106.8282 106.6943>
  }
 ]
 dyn_L_Ear_03 [
  {
   shapeType=Capsule
   length=7.1623
   radius=2.2407
   shapeName=dyn_L_Ear_03_capsule
   center=<2.5001 -1.5103 -2.689>
   rotation=<rot 75.9395 61.1413 59.8502>
  }
  {
   shapeType=Capsule
   length=7.4031
   radius=2
   shapeName=dyn_L_Ear_03_capsule
   center=<2.5752 0.594 1.7526>
   rotation=<rot 58.5787 101.9132 96.2578>
  }
 ]
 dyn_L_Ear_04 [
  {
   shapeType=Capsule
   length=7.3045
   radius=2.2978
   shapeName=dyn_L_Ear_04_capsule
   center=<3.6487 0.2001 2.9726>
   rotation=<rot 71.2627 112.6976 115.8813>
  }
  {
   shapeType=Capsule
   length=7.2838
   radius=2.3185
   shapeName=dyn_L_Ear_04_capsule
   center=<3.6517 -1.3043 -3.2163>
   rotation=<rot 75.5945 77.6259 79.8382>
  }
 ]
 dyn_L_Ear_05 [
  {
   shapeType=Capsule
   length=7.0166
   radius=1.9431
   shapeName=dyn_L_Ear_05_capsule
   center=<3.2212 -0.198 4.1811>
   rotation=<rot 76.7673 72.9528 75.3583>
  }
  {
   shapeType=Capsule
   length=7.0166
   radius=1.9431
   shapeName=dyn_L_Ear_05_capsule
   center=<3.0833 -1.3564 -2.8937>
   rotation=<rot 80.4376 122.6633 123.8334>
  }
 ]
 dyn_L_Ear_06 [
  {
   shapeType=Capsule
   length=9.9391
   radius=1.7528
   shapeName=dyn_L_Ear_06_capsule
   center=<5.502 0.3315 -1.1322>
   rotation=<rot 69.2619 -170.914 -177.9071>
  }
  {
   shapeType=Capsule
   length=9.9391
   radius=1.7528
   shapeName=dyn_L_Ear_06_capsule
   center=<5.5732 0.6417 2.0619>
   rotation=<rot 75.736 9.3097 4.6639>
  }
 ]
 dyn_R_Ear_01 [{shapeType=Capsule, length=6, radius=3, shapeName=dyn_R_Ear_01_capsule, center=<1 0 0>, rotation=<rot 90 0 0>}]
 dyn_R_Ear_02 [
  {
   shapeType=Capsule
   length=7.0718
   radius=1.9569
   shapeName=dyn_R_Ear_02_capsule
   center=<4.4088 0.2941 -1.4685>
   rotation=<rot 60.9718 -75.5508 -76.9513>
  }
  {
   shapeType=Capsule
   length=7.0718
   radius=1.9569
   shapeName=dyn_R_Ear_02_capsule
   center=<4.5271 -1.3392 1.9397>
   rotation=<rot 60.6318 -109.5263 -108.0398>
  }
 ]
 dyn_R_Ear_03 [
  {
   shapeType=Capsule
   length=6.2079
   radius=2.2814
   shapeName=dyn_R_Ear_03_capsule
   center=<2.6115 -1.649 2.6661>
   rotation=<rot 74.9302 -112.7397 -114.8816>
  }
  {
   shapeType=Capsule
   length=6.1252
   radius=2.3641
   shapeName=dyn_R_Ear_03_capsule
   center=<2.5846 -0.0431 -2.2923>
   rotation=<rot 58.0848 -75.9245 -77.6567>
  }
 ]
 dyn_R_Ear_04 [
  {
   shapeType=Capsule
   length=7.0987
   radius=2.7054
   shapeName=dyn_R_Ear_04_capsule
   center=<3.3457 -0.3522 -3.1246>
   rotation=<rot 66.579 -70.4202 -66.5003>
  }
  {
   shapeType=Capsule
   length=6.6913
   radius=2.7465
   shapeName=dyn_R_Ear_04_capsule
   center=<3.4873 -2.1552 2.9248>
   rotation=<rot 74.6183 -100.2511 -96.9432>
  }
 ]
 dyn_R_Ear_05 [
  {
   shapeType=Capsule
   length=7.9497
   radius=2.4036
   shapeName=dyn_R_Ear_05_capsule
   center=<2.5121 -0.2853 -4.7624>
   rotation=<rot 63.0032 -93.2142 -90.1051>
  }
  {
   shapeType=Capsule
   length=7.6841
   radius=2.4333
   shapeName=dyn_R_Ear_05_capsule
   center=<2.6075 -1.5905 3.0993>
   rotation=<rot 79.4388 -55.4436 -55.9747>
  }
 ]
 dyn_R_Ear_06 [
  {
   shapeType=Capsule
   length=9.9391
   radius=1.7528
   shapeName=dyn_L_Ear_06_capsule
   center=<5.8102 0.3808 -1.9564>
   rotation=<rot 69.2619 -170.914 -177.9071>
  }
  {
   shapeType=Capsule
   length=9.9391
   radius=1.7528
   shapeName=dyn_L_Ear_06_capsule
   center=<5.3904 0.6118 1.3334>
   rotation=<rot 75.736 9.3097 4.6639>
  }
 ]
 head [{
   shapeType=Capsule
   length=5.9853
   radius=7.7296
   shapeName=head_capsule
   center=<7.1364 0.2045 -0.2375>
   rotation=<rot 0.7313 -165.089 -88.9647>
  }]
 lowerarm_l [{
   shapeType=Capsule
   length=22.2635
   radius=7.3217
   shapeName=upperarm_l_capsule
   center=<10.8504 0.343 -0.8607>
   rotation=<rot 0.0939 -90.4562 90.5946>
  }]
 lowerarm_r [{
   shapeType=Capsule
   length=19.7321
   radius=6.6132
   shapeName=upperarm_l_capsule
   center=<-9.5827 -0.258 0.3926>
   rotation=<rot 0.0981 -87.8911 -87.6152>
  }]
 neck_01 [{shapeType=Capsule, length=12.7345, radius=5.0217, shapeName=neck_01_capsule, center=<-0.6214 -4.2834 0>}]
 neck_02 [
  {
   shapeType=Capsule
   length=6.8068
   radius=8.4854
   shapeName=neck_02_capsule
   center=<0.5827 -2.9901 -0.016>
   rotation=<rot 0 -166.7272 0.5597>
  }
  {
   shapeType=Capsule
   length=1.9141
   radius=9.1208
   shapeName=neck_02_capsule
   center=<5.0974 -2.9289 -0.0255>
   rotation=<rot 0 -166.7274 -89.441>
  }
 ]
 spine_01 [{shapeType=Capsule, length=11.457, radius=10.6817, shapeName=spine_01_capsule, center=<0 -1.5638 0>, rotation=<rot 0 -90 0>}]
 spine_02 [{
   shapeType=Capsule
   length=9.7699
   radius=10.7316
   shapeName=spine_03_capsule
   center=<4.7082 -2.1578 -0.0046>
   rotation=<rot -0.0021 -90 -0.0206>
  }]
 spine_03 [{
   shapeType=Capsule
   length=15.6083
   radius=9.8741
   shapeName=spine_03_capsule
   center=<7.1721 -3.8873 -0.0038>
   rotation=<rot -0.0021 -90 -0.0206>
  }]
 spine_04 [
  {
   shapeType=Capsule
   length=22.3848
   radius=9.6745
   shapeName=spine_04_capsule
   center=<6.8537 -3.1083 0.1303>
   rotation=<rot 0.58 -90 -0.4676>
  }
  {
   shapeType=Capsule
   length=12.9511
   radius=9.6745
   shapeName=spine_04_capsule
   center=<6.8537 -4.5683 0.1451>
   rotation=<rot 0.58 -90 -0.4676>
  }
 ]
 spine_05 [
  {shapeType=Capsule, length=19.9149, radius=11.339, shapeName=spine_05_capsule, center=<0.8228 -0.2728 0.1053>}
  {shapeType=Capsule, length=8.1044, radius=11.1469, shapeName=spine_05_capsule, center=<0.6431 -2.8153 0.1053>}
 ]
 upperarm_l [{
   shapeType=Capsule
   length=18.4081
   radius=8.5062
   shapeName=upperarm_l_capsule
   center=<10.8588 0.1469 -1.9982>
   rotation=<rot 0.0715 -88.305 85.7036>
  }]
 upperarm_r [{
   shapeType=Capsule
   length=16.9613
   radius=8.3661
   shapeName=upperarm_l_capsule
   center=<-10.9906 -0.3072 1.9197>
   rotation=<rot 0.0982 -87.8912 -92.6799>
  }]
}
constraints [
 {
  bone1Name=head
  bone2Name=neck_02
  swing1LimitDegrees=45
  swing1Motion=Free
  swing2LimitDegrees=45
  swing2Motion=Free
  twistLimitDegrees=45
  twistMotion=Free
 }
 {
  bone1Name=neck_02
  bone2Name=neck_01
  swing1LimitDegrees=45
  swing1Motion=Free
  swing2LimitDegrees=45
  swing2Motion=Free
  twistLimitDegrees=45
  twistMotion=Free
 }
 {
  bone1Name=neck_01
  bone2Name=spine_05
  swing1LimitDegrees=45
  swing1Motion=Free
  swing2LimitDegrees=45
  swing2Motion=Free
  twistLimitDegrees=45
  twistMotion=Free
 }
 {
  bone1Name=clavicle_l
  bone2Name=spine_05
  swing1LimitDegrees=45
  swing1Motion=Free
  swing2LimitDegrees=45
  swing2Motion=Free
  twistLimitDegrees=45
  twistMotion=Free
 }
 {
  bone1Name=clavicle_r
  bone2Name=spine_05
  swing1LimitDegrees=45
  swing1Motion=Free
  swing2LimitDegrees=45
  swing2Motion=Free
  twistLimitDegrees=45
  twistMotion=Free
 }
 {
  bone1Name=spine_05
  bone2Name=spine_04
  swing1LimitDegrees=45
  swing1Motion=Free
  swing2LimitDegrees=45
  swing2Motion=Free
  twistLimitDegrees=45
  twistMotion=Free
 }
 {
  bone1Name=spine_04
  bone2Name=spine_03
  swing1LimitDegrees=45
  swing1Motion=Free
  swing2LimitDegrees=45
  swing2Motion=Free
  twistLimitDegrees=45
  twistMotion=Free
 }
 {
  bone1Name=upperarm_l
  bone2Name=clavicle_l
  swing1LimitDegrees=45
  swing1Motion=Free
  swing2LimitDegrees=45
  swing2Motion=Free
  twistLimitDegrees=45
  twistMotion=Free
 }
 {
  bone1Name=upperarm_r
  bone2Name=clavicle_r
  swing1LimitDegrees=45
  swing1Motion=Free
  swing2LimitDegrees=45
  swing2Motion=Free
  twistLimitDegrees=45
  twistMotion=Free
 }
 {bone1Name=base_L_Ear_01, bone2Name=head, swing1LimitDegrees=45, swing2LimitDegrees=45, twistLimitDegrees=0}
 {bone1Name=dyn_L_Ear_01, bone2Name=base_L_Ear_01, swing1LimitDegrees=45, swing2LimitDegrees=45, twistLimitDegrees=0}
 {bone1Name=dyn_L_Ear_02, bone2Name=dyn_L_Ear_01, swing1LimitDegrees=20, swing2LimitDegrees=2.5, twistLimitDegrees=0}
 {bone1Name=dyn_L_Ear_03, bone2Name=dyn_L_Ear_02, swing1LimitDegrees=25, swing2LimitDegrees=2.5, twistLimitDegrees=0}
 {bone1Name=dyn_L_Ear_04, bone2Name=dyn_L_Ear_03, swing1LimitDegrees=30, swing2LimitDegrees=5, twistLimitDegrees=5}
 {bone1Name=dyn_L_Ear_05, bone2Name=dyn_L_Ear_04, swing1LimitDegrees=35, swing2LimitDegrees=7, twistLimitDegrees=5}
 {bone1Name=dyn_L_Ear_06, bone2Name=dyn_L_Ear_05, swing1LimitDegrees=40, swing2LimitDegrees=7, twistLimitDegrees=5}
 {bone1Name=base_R_Ear_01, bone2Name=head, swing1LimitDegrees=45, swing2LimitDegrees=45, twistLimitDegrees=0}
 {bone1Name=dyn_R_Ear_01, bone2Name=base_R_Ear_01, swing1LimitDegrees=45, swing2LimitDegrees=45, twistLimitDegrees=0}
 {bone1Name=dyn_R_Ear_02, bone2Name=dyn_R_Ear_01, swing1LimitDegrees=40, swing2LimitDegrees=2.5, twistLimitDegrees=0}
 {bone1Name=dyn_R_Ear_03, bone2Name=dyn_R_Ear_02, swing1LimitDegrees=40, swing2LimitDegrees=5, twistLimitDegrees=7}
 {bone1Name=dyn_R_Ear_04, bone2Name=dyn_R_Ear_03, swing1LimitDegrees=40, swing2LimitDegrees=10, twistLimitDegrees=5}
 {bone1Name=dyn_R_Ear_05, bone2Name=dyn_R_Ear_04, swing1LimitDegrees=45, swing2LimitDegrees=7, twistLimitDegrees=5}
 {bone1Name=dyn_R_Ear_06, bone2Name=dyn_R_Ear_05, swing1LimitDegrees=50, swing2LimitDegrees=7, twistLimitDegrees=5}
 {
  bone1Name=spine_03
  bone2Name=spine_02
  swing1LimitDegrees=45
  swing1Motion=Free
  swing2LimitDegrees=45
  swing2Motion=Free
  twistLimitDegrees=45
  twistMotion=Free
 }
 {
  bone1Name=lowerarm_l
  bone2Name=upperarm_l
  swing1LimitDegrees=45
  swing1Motion=Free
  swing2LimitDegrees=45
  swing2Motion=Free
  twistLimitDegrees=45
  twistMotion=Free
 }
 {
  bone1Name=lowerarm_r
  bone2Name=upperarm_r
  swing1LimitDegrees=45
  swing1Motion=Free
  swing2LimitDegrees=45
  swing2Motion=Free
  twistLimitDegrees=45
  twistMotion=Free
 }
 {
  bone1Name=spine_02
  bone2Name=spine_01
  swing1LimitDegrees=45
  swing1Motion=Free
  swing2LimitDegrees=45
  swing2Motion=Free
  twistLimitDegrees=45
  twistMotion=Free
 }
]
