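# Animation Blueprint for the Happy Hopper character's face accessory part, posing its ears and zipper.
# Copied parent pose, with arms layered from the body mesh, feeds a Control Rig driven by weapon, movement and backbling state from the event graph.
# Cached stages then apply ear offsets, two blended ear RigidBody setups (one for crouch), and zipper AnimDynamics.
# RigidBody switches off for backbling, DBNO or disabled physics; neck-correction nodes sit unconnected.
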AnimBlueprint /BRCosmetics/Characters/Player/Male/Medium/Bodies/M_MED_HappyHopper/Meshes/Parts/M_MED_HappyHopper_FaceAcc_AnimBP
target_skeleton=&/M_MED_HappyHopper_FaceAcc_Skeleton
graphs [
 AnimGraph {nodes [
   n1 = "Transform (Modify) Bone - Bone: neck_01"(ComponentPose=Local To Component(), Alpha=1, Translation="0,0,0", Rotation="0,-10,0.000", Scale="1,1,1") {
    rotationMode=BMM_Additive
    rotationSpace=BCS_BoneSpace
    alphaBoolBlend {blendOption=Linear}
    alphaScaleBiasClamp {interpSpeedDecreasing=5, interpSpeedIncreasing=20}
    boneToModify {boneName=neck_01}
    rotation {yaw=-10}
   }
   n2 = "Transform (Modify) Bone - Bone: neck_02"(ComponentPose=@n1, Alpha=1, Translation="0,0,0", Rotation="0,-10,0.000", Scale="1,1,1") {
    rotationMode=BMM_Additive
    rotationSpace=BCS_BoneSpace
    alphaBoolBlend {blendOption=Linear}
    alphaScaleBiasClamp {interpSpeedDecreasing=5, interpSpeedIncreasing=20}
    boneToModify {boneName=neck_02}
    rotation {yaw=-10}
   }
   n3 = "Transform (Modify) Bone - Bone: head"(ComponentPose=@n2, Alpha=1, Translation="0,0,0", Rotation="0,-10,0.000", Scale="1,1,1") {
    rotationMode=BMM_Additive
    rotationSpace=BCS_BoneSpace
    alphaBoolBlend {blendOption=Linear}
    alphaScaleBiasClamp {interpSpeedDecreasing=5, interpSpeedIncreasing=20}
    boneToModify {boneName=head}
    rotation {yaw=-10}
   }
   Component To Local(ComponentPose=@n3)
   Output Pose(Result=Use cached pose 'DynamicsCache'()) {layerGroup=None}
   n4 = NOT Boolean(A=Get NonDynamicSetupNeeded() -> (NonDynamicSetupNeeded:bool))
   n5 = RigidBody!(
    ComponentPose=Local To Component(LocalPose=Use cached pose 'RIgidBodyCache'())
    bAlphaBoolEnabled:bool=@n4
    OverrideWorldGravity:Vector="0,0,0"
   ) -> (Pose) {
    alphaInputType=Bool
    bOverrideWorldGravity=true
    overridePhysicsAsset=&/M_MED_HappyHopper_FaceAcc_EarsB_Physics
    simulationSpace=BaseBoneSpace
    alphaBoolBlend {blendOption=Linear}
    baseBoneRef {boneName=spine_03}
    simSpaceSettings {
     maxAngularAcceleration=1
     maxAngularVelocity=3
     maxLinearAcceleration=1000
     maxLinearVelocity=500
     velocityScaleZ=0.5
     worldAlpha=1
     externalLinearDragV {x=0.2, z=0.2}
    }
   }
   n6 = RigidBody(ComponentPose=Local To Component(LocalPose=Use cached pose 'RIgidBodyCache'()), bAlphaBoolEnabled=@n4, OverrideWorldGravity="0,0,0") {
    alphaInputType=Bool
    bOverrideWorldGravity=true
    overridePhysicsAsset=&/M_MED_HappyHopper_FaceAcc_Ears_Crouch_Physics
    simulationSpace=BaseBoneSpace
    alphaBoolBlend {blendOption=Linear}
    baseBoneRef {boneName=spine_03}
    simSpaceSettings {
     maxAngularAcceleration=1
     maxAngularVelocity=3
     maxLinearAcceleration=1000
     maxLinearVelocity=500
     velocityScaleZ=0.5
     worldAlpha=1
     externalLinearDragV {x=0.2, z=0.2}
    }
   }
   Save cached pose 'CRCache_Input'(Pose=Blend Poses by bool(BlendPose_0=Component To Local(ComponentPose=@n6), BlendPose_1=Component To Local(ComponentPose=@n5)) {blendType=Linear}) {cacheName=CRCache_Input}
   n7 = Anim Dynamics - Bone: dyn_Zipper!(ComponentPose=Local To Component(LocalPose=Use cached pose 'CRCache_Input'()), Alpha=1, GravityOverride:Vector="0,0,-980") -> (Pose) {
    angularDampingOverride=0.9
    angularSpringConstant=20000
    bOverrideAngularDamping=true
    bUseGravityOverride=true
    gravityScale=0.35
    simulationSpace=RootRelative
    alphaBoolBlend {blendOption=Linear}
    boundBone {boneName=dyn_Zipper}
    componentAppliedLinearAccClamp=<10000 10000 10000>
    componentLinearAccScale=<1 1 1>
    componentLinearVelScale=<1 1 1>
    gravityOverride {z=-980}
    physicsBodyDefinitions [{
      boundBone {boneName=dyn_Zipper}
      boxExtents=<11 32 9>
      constraintSetup {angularLimitsMax=<5 45 138.6036>, angularLimitsMin=<-5 -45 -10>, angularTarget=<0 0 0>}
      localJointOffset=<1.25 0 0>
     }]
   }
   Save cached pose 'DynamicsCache'(Pose=Component To Local(ComponentPose=@n7)) {cacheName=DynamicsCache}
   n8 = Get BodyMesh() -> (BodyMesh:Skeletal Mesh Component Object Reference)
   n9 = Control Rig!(
    Source=Layered blend per bone(BasePose=Copy Pose From Mesh() {bUseAttachedParent=true}, BlendPoses_0=Copy Pose From Mesh(SourceMeshComponent=@n8)) {layerSetup [{branchFilters [{boneName=upperarm_r}, {boneName=upperarm_l}]}]}
    Alpha
    IsWearingBackBling:bool=Get isWearingPackBool() -> (isWearingPackBool:bool)
    IsJumping:double=Get IsJumping() -> (IsJumping:double)
    isSkyDiving:bool=Get Is SkyDiving() -> (Is SkyDiving:bool)
    IsPistolOut:bool=Get PistolOut() -> (PistolOut:bool)
    bIsCrouching:bool=Get bIsCrouching()
    IsSwingingPickaxe:bool=Get IsSwingingPIckaxe() -> (IsSwingingPIckaxe:bool)
    bisPickaxeOut:bool=Get PickAxeOut() -> (PickAxeOut:bool)
    IsAimingDownSights:bool=Get AimingDown() -> (AimingDown:bool)
    bIsSMGOut:bool=Get IsSMGOut() -> (IsSMGOut:bool)
    isTargetingF:double=Get isTargeting float() -> (isTargeting float:double)
    SkydiveDiveMode:double=Get isDiveModeFloat() -> (isDiveModeFloat:double)
    isRbanEnabled:bool=Get bRigidBodyEnabled()
    isDBNO:bool=Get bIsDBNO()
    isTacticalJumpingF:double=Get isTacticalJumpingF() -> (isTacticalJumpingF:double)
   ) -> (Pose) {
    alphaBoolBlend {blendOption=Linear}
    controlRigAssetReference {blueprintRigClass=&/M_MED_HappyHopper_FaceAcc_CtrlRig.M_MED_HappyHopper_FaceAcc_CtrlRig_C}
   }
   Save cached pose 'PreSim_Cache'(Pose=@n9) {cacheName=PreSim_Cache}
   n10 = "Transform (Modify) Bone - Bone: base_L_Ear_01"!(ComponentPose=Local To Component(LocalPose=Use cached pose 'PreSim_Cache'()), bAlphaBoolEnabled:bool=True, Rotation="0,20,60") -> (Pose) {
    alphaInputType=Bool
    rotationMode=BMM_Additive
    alphaBoolBlend {blendInTime=0.1, blendOption=Linear, blendOutTime=0.1}
    boneToModify {boneName=base_L_Ear_01}
    rotation {roll=60, yaw=20}
   }
   n11 = "Transform (Modify) Bone - Bone: base_R_Ear_01"!(ComponentPose=@n10, bAlphaBoolEnabled:bool=True, Rotation="10,-20,50") -> (Pose) {
    alphaInputType=Bool
    rotationMode=BMM_Additive
    alphaBoolBlend {blendInTime=0.1, blendOption=Linear, blendOutTime=0.1}
    boneToModify {boneName=base_R_Ear_01}
    rotation=<rot 10 -20 50>
   }
   n12 = Get bIsParachuteOpened()
   n13 = "Transform (Modify) Bone - Bone: base_L_Ear_01"(ComponentPose=@n11, bAlphaBoolEnabled=@n12, Rotation="0,10,30") {
    alphaInputType=Bool
    rotationMode=BMM_Additive
    alphaBoolBlend {blendInTime=0.1, blendOption=Linear, blendOutTime=0.1}
    boneToModify {boneName=base_L_Ear_01}
    rotation {roll=30, yaw=10}
   }
   n14 = "Transform (Modify) Bone - Bone: base_R_Ear_01"(ComponentPose=@n13, bAlphaBoolEnabled=@n12, Rotation="5,-10,25") {
    alphaInputType=Bool
    rotationMode=BMM_Additive
    alphaBoolBlend {blendInTime=0.1, blendOption=Linear, blendOutTime=0.1}
    boneToModify {boneName=base_R_Ear_01}
    rotation=<rot 5 -10 25>
   }
   Save cached pose 'RIgidBodyCache'(Pose=Blend Poses by bool(BlendPose_0=Use cached pose 'PreSim_Cache'(), BlendPose_1=Component To Local(ComponentPose=@n14)) {blendType=Linear}) {cacheName=RIgidBodyCache}
  ]}
 EventGraph {nodes [
   n1 = Scale Joint Rotation Function!(
    execute
    Scale:Rotator="1,1,0.1"
    TargetBone:Name=head
    ReferenceBone:Name=neck_02
    SkelMeshComponent:Skinned Mesh Component Object Reference=GetPartSkeletalMeshComponent()
   ) -> (then, FinalRotation:Rotator, OriginalRotation:Rotator)
   > Set head rot(head rot:Rotator=@n1) -> (Output_Get:Rotator)
   Event OnBeginJumping()
   > Set isTacticalJumping(isTacticalJumping:bool=Get bIsTacticalSprinting()) -> (Output_Get:bool)
   Event OnBeginPickaxeSwing() {is=title K2Node_Event}
   > Set IsSwingingPIckaxe(IsSwingingPIckaxe:bool=true) -> (Output_Get:bool)
   n2 = Event BlueprintUpdateAnimation()
   > n3 = Sequence()
   > n4 = Gravity Override State Function DEPRECATED(CharacterPartAnimInstance=Self-Reference(), GameTimeInSeconds=GetGameTimeInSeconds(), GravityOverrideParams=Get Gravity Override Params())
   > Set GravityOverride(GravityOverride=@n4)
   n5 = OR Boolean(A=Get bIsSkydiveDiveMode(), B=Get bIsParachuteOpened())
   > Set OnAir(OnAir:bool=AND Boolean(A=Get Is SkyDiving(), B=NOT Boolean(A=@n5))) -> (Output_Get:bool)
   n6 = Get bRigidBodyEnabled()
   > Set NonDynamicSetupNeeded(NonDynamicSetupNeeded:bool=OR Boolean!(A=Get isWearingPackBool(), B=NOT Boolean(A=@n6), C:bool=Get bIsDBNO()) -> (ReturnValue)) -> (Output_Get:bool)
   n7 = Get bIsTargeting()
   Set isTargeting float(
    execute=@n3.then_1
    isTargeting float:double=FInterpTo(Current=Get isTargeting float(), Target="To Float (Boolean)"(InBool=@n7), DeltaTime=@n2.DeltaTimeX, InterpSpeed=15)
   ) -> (Output_Get:double)
   n8 = Get bIsSkydiveDiveMode()
   > Set isDiveModeFloat(isDiveModeFloat:double=FInterpTo(Current=Get isDiveModeFloat(), Target="To Float (Boolean)"(InBool=@n8), DeltaTime=@n2.DeltaTimeX, InterpSpeed=5)) -> (Output_Get:double)
   n9 = Get isTacticalJumping() -> (isTacticalJumping:bool)
   > Set isTacticalJumpingF(isTacticalJumpingF:double=FInterpTo(Current=Get isTacticalJumpingF(), Target="To Float (Boolean)"(InBool=@n9), DeltaTime=@n2.DeltaTimeX, InterpSpeed=5)) -> (Output_Get:double)
   Event OnEndPickaxeSwing() {is=title K2Node_Event}
   > Set IsSwingingPIckaxe(IsSwingingPIckaxe=false)
   Event OnWeaponChanged()
   > Set PistolOut(PistolOut:bool=false) -> (Output_Get:bool)
   > Set PickAxeOut(PickAxeOut:bool=false) -> (Output_Get:bool)
   > Set IsSMGOut(IsSMGOut:bool=false) -> (Output_Get:bool)
   > n10 = Switch on EFortWeaponCoreAnimation!(execute, Selection:EFortWeaponCoreAnimation Enum=Get WeaponCoreAnim!() -> (WeaponCoreAnim)) -> (
    Melee:Exec
    Pistol:Exec
    Shotgun:Exec
    PaperBlueprint:Exec
    Rifle:Exec
    MeleeOneHand:Exec
    MachinePistol:Exec
    RocketLauncher:Exec
    GrenadeLauncher:Exec
    GoingCommando:Exec
    AssaultRifle:Exec
    TacticalShotgun:Exec
    SniperRifle:Exec
    TrapPlacement:Exec
    ShoulderLauncher:Exec
    AbilityDecoTool:Exec
    Crossbow:Exec
    C4:Exec
    RemoteControl:Exec
    DualWield:Exec
    AR_BullPup:Exec
    AR_ForwardGrip:Exec
    MedPackPaddles:Exec
    SMG_P90:Exec
    AR_DrumGun:Exec
    Consumable_Small:Exec
    Consumable_Large:Exec
    Balloon:Exec
    MountedTurret:Exec
    CreativeTool:Exec
    ExplosiveBow:Exec
    AshtonIndigo:Exec
    AshtonChicago:Exec
    MeleeDualWield:Exec
    Unarmed:Exec
   )
   > Set PickAxeOut(PickAxeOut=true)
   Set PistolOut(execute=@n10.Pistol, PistolOut=true)
   Set IsSMGOut(execute=@n10.SMG_P90, IsSMGOut=true)
   Event OnBeginTargeting() {is=title K2Node_Event}
   > Set AimingDown(AimingDown:bool=true) -> (Output_Get:bool)
   Event OnBeginSkydiving()
   > Set Is SkyDiving(Is SkyDiving:bool=true) -> (Output_Get:bool)
   n11 = Event OnCharacterCustomizationLoaded()
   > Set BodyMesh(BodyMesh=Get Array Item with EFort Custom Part Type(Array=@n11.CustomMeshParts))
   n12 = CheckForBackbling()
   > Set isWearingPackBool(isWearingPackBool:bool=OR Boolean(A=@n12, B=@n12.bIsBackblingACape, C=@n12.bIsBackblingLong)) -> (Output_Get:bool)
   n13 = OR Boolean!(A=@n12.bIsBackblingACape, B=@n12.bIsBackblingLong) -> (ReturnValue)
   > Set is_f_WearingCape(is_f_WearingCape:double="To Float (Boolean)"(InBool=@n13)) -> (Output_Get:double)
   > Set isWearingBP(isWearingBP:bool=AND Boolean(A=@n12, B=NOT Boolean(A=@n13))) -> (Output_Get:bool)
   Event OnEndSkydiving()
   > Set Is SkyDiving(Is SkyDiving=false)
   Event OnEndTargeting() {is=title K2Node_Event}
   > Set AimingDown(AimingDown=false)
   Event OnEndJumping()
   > Set isTacticalJumping(isTacticalJumping=false)
   Branch(Condition=Get bIsCrouching())
   n14 = BreakRotator(InRot=GetDeltaTransformFromRefPose!(self=GetOwnerSkeletalMeshComponent(), BoneName=head, BaseName=spine_05) -> (ReturnValue_Location:Vector, ReturnValue_Rotation:Rotator, ReturnValue_Scale:Vector).ReturnValue_Rotation)
   > Set TwistCompensatorAlpha(TwistCompensatorAlpha:double=Remap Float(Input="Absolute (Float)"(A=@n14), OldRange=(X=14,Y=40), NewRange=(X=0,Y=1)) -> (Output=~)) -> (Output_Get:double)
   n15 = BreakRotator(InRot=GetDeltaTransformFromRefPose(self=GetOwnerSkeletalMeshComponent(), BoneName=neck_01, BaseName=spine_05).ReturnValue_Rotation)
   n16 = Get X Percentage() -> (X Percentage:double)
   n17 = float + float(A=Get TwistCompensatorAlpha() -> (TwistCompensatorAlpha:double), B=1)
   n18 = Get Y Percentage() -> (Y Percentage:double)
   n19 = Get Z Percentage() -> (Z Percentage:double)
   > Set neck01 rot(neck01 rot:Rotator=MakeRotator(
     Roll=float * float!(A=@n15, B=-1, C:double=@n16) -> (ReturnValue) {is=title K2Node_PromotableOperator}
     Pitch=float * float!(A=@n15.Pitch, B=-1, C=@n18, D:double=@n17) -> (ReturnValue)
     Yaw=float * float!(A=@n15.Yaw, B=-1, C=@n19) -> (ReturnValue)
    )) -> (Output_Get:Rotator)
   n20 = BreakRotator(InRot=GetDeltaTransformFromRefPose(self=GetOwnerSkeletalMeshComponent(), BoneName=neck_02, BaseName=neck_01).ReturnValue_Rotation)
   n21 = Get Z Percentage()
   n22 = Get X Percentage()
   n23 = float + float(A=Get TwistCompensatorAlpha(), B=1)
   n24 = Get Y Percentage()
   > Set neck02 rot(neck02 rot:Rotator=MakeRotator(
     Roll=float * float(A=@n20, B=-1, C=@n22)
     Pitch=float * float!(A=@n20.Pitch, B=-1, C=@n24, D:double=@n23) -> (ReturnValue)
     Yaw=float * float!(A=@n20.Yaw, B=-1, C=@n21) -> (ReturnValue)
    )) -> (Output_Get:Rotator)
   n25 = BreakRotator(InRot=GetDeltaTransformFromRefPose(self=GetOwnerSkeletalMeshComponent(), BoneName=head, BaseName=neck_02).ReturnValue_Rotation)
   n26 = float + float(A=Get TwistCompensatorAlpha(), B=1)
   n27 = Get Y Percentage()
   n28 = Get Z Percentage()
   n29 = Get X Percentage()
   > Set head rot(head rot=MakeRotator(
     Roll=float * float(A=@n25, B=-1, C=@n29)
     Pitch=float * float!(A=@n25.Pitch, B=-1, C=@n27, D:double=@n26) -> (ReturnValue)
     Yaw=float * float!(A=@n25.Yaw, B=-1, C=@n28) -> (ReturnValue)
    ))
  ]}
]
variables [
 AimingDown {type=boolean}
 BodyMesh {type=SkeletalMeshComponent}
 clavicle_l_location {type=Vector}
 clavicle_l_rotation {type=Rotator}
 clavicle_r_location {type=Vector}
 clavicle_r_rotation {type=Rotator}
 fixEarsWhenFalling {type=boolean}
 Gravity Override Params {type=GravityOverrideParamsStruct}
 GravityOverride {type=Vector}
 head rot {type=Rotator}
 Is SkyDiving {type=boolean}
 Is Using Hang Gliding in Parachute {type=boolean}
 is_f_WearingCape {type=number}
 isBodyAttached {type=boolean}
 isDiveModeFloat {type=number}
 IsJumping {type=number}
 IsSMGOut {type=boolean}
 IsSwingingPIckaxe {type=boolean}
 isTacticalJumping {type=boolean}
 isTacticalJumpingF {type=number}
 isTargeting float {type=number}
 isWearingBP {type=boolean}
 isWearingPackBool {type=boolean}
 lowerarm_l_location {type=Vector}
 lowerarm_l_rotation {type=Rotator}
 lowerarm_r_location {type=Vector}
 lowerarm_r_rotation {type=Rotator}
 neck01 rot {type=Rotator}
 neck02 rot {type=Rotator}
 NonDynamicSetupNeeded {type=boolean}
 OnAir {type=boolean}
 PickAxeOut {type=boolean, default=true}
 PistolOut {type=boolean}
 TwistCompensatorAlpha {type=number, default=1}
 upperarm_l_location {type=Vector}
 upperarm_l_rotation {type=Rotator}
 upperarm_r_location {type=Vector}
 upperarm_r_rotation {type=Rotator}
 X Percentage {type=number, default=1}
 Y Percentage {type=number, default=0.1}
 Z Percentage {type=number, default=0.1}
]
